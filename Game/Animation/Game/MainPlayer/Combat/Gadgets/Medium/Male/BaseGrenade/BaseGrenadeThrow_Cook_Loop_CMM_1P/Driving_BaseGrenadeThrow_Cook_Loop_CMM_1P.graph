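# Level Sequence that drives the first-person medium male animation for the looping cook phase of the base grenade throw.
# A skeletal mesh actor is animated through a Fortnite_Biped_CtrlRig parameter track, while its skeletal animation track holds no sections.
# Legs are switched to IK and arms, spine and neck to FK.
# Subtle keyed sway spans frames 0 to 43, with end values returning to about the starting pose so it loops.
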LevelSequence /Game/Animation/Game/MainPlayer/Combat/Gadgets/Medium/Male/BaseGrenade/BaseGrenadeThrow_Cook_Loop_CMM_1P/Driving_BaseGrenadeThrow_Cook_Loop_CMM_1P
bindings ["BaseGrenadeThrow_Cook_Loop_CMM_1P --> BaseGrenadeThrow_Cook_Loop_CMM_1P" {
  id=#1
  possessed_class=SkeletalMeshActor
  tracks [
   {
    type=MovieScene3DTransformTrack
    sections [{
      type=MovieScene3DTransformSection
      channels [
       Location.X {type=MovieSceneScriptingDoubleChannel, default=0}
       Location.Y {type=MovieSceneScriptingDoubleChannel, default=0}
       Location.Z {type=MovieSceneScriptingDoubleChannel, default=0}
       Rotation.X {type=MovieSceneScriptingDoubleChannel, default=0}
       Rotation.Y {type=MovieSceneScriptingDoubleChannel, default=0}
       Rotation.Z {type=MovieSceneScriptingDoubleChannel, default=0}
       Scale.X {type=MovieSceneScriptingDoubleChannel, default=1}
       Scale.Y {type=MovieSceneScriptingDoubleChannel, default=1}
       Scale.Z {type=MovieSceneScriptingDoubleChannel, default=1}
      ]
     }]
   }
   {type=MovieSceneSkeletalAnimationTrack, sections []}
   {type=MovieSceneSpawnTrack, sections [{type=MovieSceneSpawnSection, channels [None {type=MovieSceneScriptingBoolChannel, default=true}]}]}
   {
    type=MovieSceneControlRigParameterTrack
    display_name=Fortnite_Biped_CtrlRig
    sections [
     {
      type=MovieSceneControlRigParameterSection
      channels [
       leg_l_stretch_switch {type=MovieSceneScriptingBoolChannel, default=false, const [false,44]}
       leg_r_stretch_switch {type=MovieSceneScriptingBoolChannel, default=false, const [false,44]}
       arm_l_fk_local {type=MovieSceneScriptingBoolChannel, default=true, const [true,44]}
       Auto Interpolate_3 {type=MovieSceneScriptingBoolChannel, default=true, const [true,44]}
       Auto Interpolate {type=MovieSceneScriptingBoolChannel, default=true, const [true,44]}
       arm_l_stretch_switch {type=MovieSceneScriptingBoolChannel, default=false, const [false,44]}
       Keep Aligned {type=MovieSceneScriptingBoolChannel, default=false, const [false,44]}
       arm_r_fk_local {type=MovieSceneScriptingBoolChannel, default=true, const [true,44]}
       Auto Interpolate_4 {type=MovieSceneScriptingBoolChannel, default=true, const [true,44]}
       Auto Interpolate_2 {type=MovieSceneScriptingBoolChannel, default=true, const [true,44]}
       arm_r_stretch_switch {type=MovieSceneScriptingBoolChannel, default=false, const [false,44]}
       Keep Aligned_2 {type=MovieSceneScriptingBoolChannel, default=false, const [false,44]}
       head_local {type=MovieSceneScriptingBoolChannel, default=true, const [true,44]}
       Stretch {type=MovieSceneScriptingBoolChannel, default=false, const [false,44]}
       head_stretch {type=MovieSceneScriptingBoolChannel, default=false, const [false,44]}
       spine_stretch_switch {type=MovieSceneScriptingBoolChannel, default=false, const [false,44]}
       arm_l_fk_ik_switch {type=MovieSceneScriptingBoolChannel, default=false, const [false,44]}
       leg_l_fk_ik_switch {type=MovieSceneScriptingBoolChannel, default=true, const [true,44]}
       arm_r_fk_ik_switch {type=MovieSceneScriptingBoolChannel, default=false, const [false,44]}
       leg_r_fk_ik_switch {type=MovieSceneScriptingBoolChannel, default=true, const [true,44]}
       ShowBodyControls {type=MovieSceneScriptingBoolChannel, default=true, const [true,44]}
       spine_fk_ik_switch {type=MovieSceneScriptingBoolChannel, default=false, const [false,44]}
       neck_fk_ik_switch {type=MovieSceneScriptingBoolChannel, default=false, const [false,44]}
       space {type=MovieSceneScriptingByteChannel, default=1}
       global_ctrl.Location.X {type=MovieSceneScriptingFloatChannel, default=0, const [0,44]}
       global_ctrl.Location.Y {type=MovieSceneScriptingFloatChannel, default=0, const [0,44]}
       global_ctrl.Location.Z {type=MovieSceneScriptingFloatChannel, default=0, const [0,44]}
       global_ctrl.Rotation.X {type=MovieSceneScriptingFloatChannel, default=0, const [0,44]}
       global_ctrl.Rotation.Y {type=MovieSceneScriptingFloatChannel, default=0, const [0,44]}
       global_ctrl.Rotation.Z {type=MovieSceneScriptingFloatChannel, default=0, const [0,44]}
       root_ctrl.Location.X {type=MovieSceneScriptingFloatChannel, default=0, const [0,44]}
       root_ctrl.Location.Y {type=MovieSceneScriptingFloatChannel, default=0, const [0,44]}
       root_ctrl.Location.Z {type=MovieSceneScriptingFloatChannel, default=0, const [0,44]}
       root_ctrl.Rotation.X {type=MovieSceneScriptingFloatChannel, default=0, const [0,44]}
       root_ctrl.Rotation.Y {type=MovieSceneScriptingFloatChannel, default=0, const [0,44]}
       root_ctrl.Rotation.Z {type=MovieSceneScriptingFloatChannel, default=0, const [0,44]}
       root_ctrl.Scale.X {type=MovieSceneScriptingFloatChannel, default=1, const [1,44]}
       root_ctrl.Scale.Y {type=MovieSceneScriptingFloatChannel, default=1, const [1,44]}
       root_ctrl.Scale.Z {type=MovieSceneScriptingFloatChannel, default=1, const [1,44]}
       body_offset_ctrl.Location.X {type=MovieSceneScriptingFloatChannel, default=0, const [0,44]}
       body_offset_ctrl.Location.Y {type=MovieSceneScriptingFloatChannel, default=0, const [0,44]}
       body_offset_ctrl.Location.Z {type=MovieSceneScriptingFloatChannel, default=0, const [0,44]}
       body_offset_ctrl.Rotation.X {type=MovieSceneScriptingFloatChannel, default=0, const [0,44]}
       body_offset_ctrl.Rotation.Y {type=MovieSceneScriptingFloatChannel, default=0, const [0,44]}
       body_offset_ctrl.Rotation.Z {type=MovieSceneScriptingFloatChannel, default=0, const [0,44]}
       body_offset_ctrl.Scale.X {type=MovieSceneScriptingFloatChannel, default=1, const [1,44]}
       body_offset_ctrl.Scale.Y {type=MovieSceneScriptingFloatChannel, default=1, const [1,44]}
       body_offset_ctrl.Scale.Z {type=MovieSceneScriptingFloatChannel, default=1, const [1,44]}
       thigh_l_fk_ctrl.Location.X {type=MovieSceneScriptingFloatChannel, default=0, thinned_from=44, keys [[0,0],[43,0]]}
       thigh_l_fk_ctrl.Location.Y {type=MovieSceneScriptingFloatChannel, default=0, thinned_from=44, keys [[0,0],[43,0]]}
       thigh_l_fk_ctrl.Location.Z {type=MovieSceneScriptingFloatChannel, default=0, thinned_from=44, keys [[0,0],[43,0]]}
       thigh_l_fk_ctrl.Rotation.X {type=MovieSceneScriptingFloatChannel, default=20.0182, thinned_from=44, keys [[0,20.0182],[23,19.7478],[43,20.0182]]}
       thigh_l_fk_ctrl.Rotation.Y {
        type=MovieSceneScriptingFloatChannel
        default=8.8032
        thinned_from=44
        keys [[0,8.8032],[2,8.8487],[5,8.9911],[15,9.6513],[18,9.7861],[20,9.8325],[22,9.8348],[24,9.7865],[27,9.6362],[36,9.0091],[39,8.8617],[41,8.8096],[43,8.8032]]
       }
       thigh_l_fk_ctrl.Rotation.Z {
        type=MovieSceneScriptingFloatChannel
        default=-16.8459
        thinned_from=44
        keys [[0,-16.8459],[2,-16.8712],[5,-17.0436],[17,-18.4282],[20,-18.6522],[22,-18.7124],[24,-18.6834],[27,-18.4912],[38,-17.1514],[40,-16.9732],[42,-16.8669],[43,-16.8459]]
       }
       thigh_l_fk_ctrl.Scale.X {type=MovieSceneScriptingFloatChannel, default=1, const [1,44]}
       thigh_l_fk_ctrl.Scale.Y {type=MovieSceneScriptingFloatChannel, default=1, const [1,44]}
       thigh_l_fk_ctrl.Scale.Z {type=MovieSceneScriptingFloatChannel, default=1, const [1,44]}
       calf_l_fk_ctrl.Location.X {type=MovieSceneScriptingFloatChannel, default=0.1088, thinned_from=44, keys [[0,0.1088],[7,0.1087],[24,0.1118],[33,0.111],[43,0.1088]]}
       calf_l_fk_ctrl.Location.Y {type=MovieSceneScriptingFloatChannel, default=0.0077, thinned_from=44, keys [[0,0.0077],[43,0.0077]]}
       calf_l_fk_ctrl.Location.Z {type=MovieSceneScriptingFloatChannel, default=0.0037, thinned_from=44, keys [[0,0.0037],[43,0.0037]]}
       calf_l_fk_ctrl.Rotation.X {type=MovieSceneScriptingFloatChannel, default=0.9238, thinned_from=44, keys [[0,0.9238],[8,0.9239],[24,0.9468],[43,0.9238]]}
       calf_l_fk_ctrl.Rotation.Y {type=MovieSceneScriptingFloatChannel, default=0.212, thinned_from=44, keys [[0,0.212],[43,0.212]]}
       calf_l_fk_ctrl.Rotation.Z {
        type=MovieSceneScriptingFloatChannel
        default=26.0629
        thinned_from=44
        keys [[0,26.0629],[7,26.0373],[24,26.8226],[32,26.6722],[43,26.0629]]
       }
       calf_l_fk_ctrl.Scale.X {type=MovieSceneScriptingFloatChannel, default=1, const [1,44]}
       calf_l_fk_ctrl.Scale.Y {type=MovieSceneScriptingFloatChannel, default=1, const [1,44]}
       calf_l_fk_ctrl.Scale.Z {type=MovieSceneScriptingFloatChannel, default=1, const [1,44]}
       foot_l_fk_ctrl.Location.X {type=MovieSceneScriptingFloatChannel, default=0, thinned_from=44, keys [[0,0],[43,0]]}
       foot_l_fk_ctrl.Location.Y {type=MovieSceneScriptingFloatChannel, default=0, thinned_from=44, keys [[0,0],[43,0]]}
       foot_l_fk_ctrl.Location.Z {type=MovieSceneScriptingFloatChannel, default=0, thinned_from=44, keys [[0,0],[43,0]]}
       foot_l_fk_ctrl.Rotation.X {type=MovieSceneScriptingFloatChannel, default=-4.6038, thinned_from=44, keys [[0,-4.6038],[34,-4.5908],[43,-4.6038]]}
       foot_l_fk_ctrl.Rotation.Y {
        type=MovieSceneScriptingFloatChannel
        default=-4.3537
        thinned_from=44
        keys [[0,-4.3537],[6,-4.4291],[12,-4.4421],[13,-4.4303],[14,-4.4303],[28,-4.2537],[34,-4.2418],[35,-4.2577],[36,-4.2577],[43,-4.3537]]
       }
       foot_l_fk_ctrl.Rotation.Z {
        type=MovieSceneScriptingFloatChannel
        default=-0.6751
        thinned_from=44
        keys [[0,-0.6751],[2,-0.6171],[3,-0.5901],[4,-0.5649],[5,-0.5419],[6,-0.5215],[7,-0.504],[8,-0.4896],[9,-0.4896],[10,-0.47],[12,-0.47],[13,-0.4647],[14,-0.4647],[15,-0.4783],[16,-0.4906],[17,-0.5066],[18,-0.5265],[19,-0.5506],[20,-0.5782],[23,-0.6696],[24,-0.6989],[25,-0.7267],[26,-0.7526],[27,-0.7763],[28,-0.7974],[29,-0.8156],[30,-0.8304],[31,-0.8304],[32,-0.8492],[36,-0.8481],[37,-0.8266],[38,-0.8093],[39,-0.7875],[40,-0.7606],[42,-0.7004],[43,-0.6751]]
       }
       foot_l_fk_ctrl.Scale.X {type=MovieSceneScriptingFloatChannel, default=1, const [1,44]}
       foot_l_fk_ctrl.Scale.Y {type=MovieSceneScriptingFloatChannel, default=1, const [1,44]}
       foot_l_fk_ctrl.Scale.Z {type=MovieSceneScriptingFloatChannel, default=1, const [1,44]}
       ball_l_fk_ctrl.Location.X {type=MovieSceneScriptingFloatChannel, default=0, const [0,44]}
       ball_l_fk_ctrl.Location.Y {type=MovieSceneScriptingFloatChannel, default=0, const [0,44]}
       ball_l_fk_ctrl.Location.Z {type=MovieSceneScriptingFloatChannel, default=0, const [0,44]}
       ball_l_fk_ctrl.Rotation.X {type=MovieSceneScriptingFloatChannel, default=0, const [0,44]}
       ball_l_fk_ctrl.Rotation.Y {type=MovieSceneScriptingFloatChannel, default=0, const [0,44]}
       ball_l_fk_ctrl.Rotation.Z {type=MovieSceneScriptingFloatChannel, default=0, const [0,44]}
       ball_l_fk_ctrl.Scale.X {type=MovieSceneScriptingFloatChannel, default=1, const [1,44]}
       ball_l_fk_ctrl.Scale.Y {type=MovieSceneScriptingFloatChannel, default=1, const [1,44]}
       ball_l_fk_ctrl.Scale.Z {type=MovieSceneScriptingFloatChannel, default=1, const [1,44]}
       foot_l_ik_ctrl.Location.X {type=MovieSceneScriptingFloatChannel, default=6.416, thinned_from=44, keys [[0,6.416],[43,6.416]]}
       foot_l_ik_ctrl.Location.Y {type=MovieSceneScriptingFloatChannel, default=21.9206, thinned_from=44, keys [[0,21.9206],[43,21.9206]]}
       foot_l_ik_ctrl.Location.Z {
        type=MovieSceneScriptingFloatChannel
        default=-0.0002
        thinned_from=44
        keys [[0,-0.0002],[3,-0.0001],[4,0.0001],[5,-0.0002],[24,-0.0001],[25,-0.0008],[27,0.0003],[28,-0.0001],[43,-0.0002]]
       }
       foot_l_ik_ctrl.Rotation.X {
        type=MovieSceneScriptingFloatChannel
        default=0.0269
        thinned_from=44
        keys [[0,0.0269],[3,0.0269],[4,0.0304],[5,0.0258],[6,0.0269],[8,0.0269],[9,0.016],[10,0.0269],[11,0.0219],[12,0.0199],[13,0.0269],[14,0.0313],[15,0.0269],[24,0.0269],[25,0.0164],[26,0.027],[27,0.034],[28,0.027],[30,0.0269],[31,0.0377],[32,0.0269],[33,0.0307],[34,0.0306],[35,0.0269],[36,0.0191],[37,0.027],[43,0.0269]]
       }
       foot_l_ik_ctrl.Rotation.Y {
        type=MovieSceneScriptingFloatChannel
        default=0.1387
        thinned_from=44
        keys [[0,0.1387],[3,0.1388],[4,0.1387],[8,0.1388],[9,0.1415],[10,0.1388],[11,0.1369],[12,0.133],[13,0.1388],[14,0.1311],[15,0.1389],[24,0.1388],[25,0.1395],[26,0.1387],[27,0.1384],[28,0.1388],[30,0.1388],[31,0.1366],[32,0.1389],[33,0.1419],[34,0.1474],[35,0.1389],[36,0.1484],[37,0.1387],[43,0.1387]]
       }
       foot_l_ik_ctrl.Rotation.Z {type=MovieSceneScriptingFloatChannel, default=10.5353, thinned_from=44, keys [[0,10.5353],[43,10.5353]]}
       foot_l_ik_ctrl.Scale.X {type=MovieSceneScriptingFloatChannel, default=1, const [1,44]}
       foot_l_ik_ctrl.Scale.Y {type=MovieSceneScriptingFloatChannel, default=1, const [1,44]}
       foot_l_ik_ctrl.Scale.Z {type=MovieSceneScriptingFloatChannel, default=1, const [1,44]}
       ball_l_ik_ctrl.Location.X {type=MovieSceneScriptingFloatChannel, default=0, const [0,44]}
       ball_l_ik_ctrl.Location.Y {type=MovieSceneScriptingFloatChannel, default=0, const [0,44]}
       ball_l_ik_ctrl.Location.Z {type=MovieSceneScriptingFloatChannel, default=0, const [0,44]}
       ball_l_ik_ctrl.Rotation.X {type=MovieSceneScriptingFloatChannel, default=0, const [0,44]}
       ball_l_ik_ctrl.Rotation.Y {type=MovieSceneScriptingFloatChannel, default=0, const [0,44]}
       ball_l_ik_ctrl.Rotation.Z {type=MovieSceneScriptingFloatChannel, default=0, const [0,44]}
       ball_l_ik_ctrl.Scale.X {type=MovieSceneScriptingFloatChannel, default=1, const [1,44]}
       ball_l_ik_ctrl.Scale.Y {type=MovieSceneScriptingFloatChannel, default=1, const [1,44]}
       ball_l_ik_ctrl.Scale.Z {type=MovieSceneScriptingFloatChannel, default=1, const [1,44]}
       foot_bk1_l_ctrl.Location.X {type=MovieSceneScriptingFloatChannel, default=0, const [0,44]}
       foot_bk1_l_ctrl.Location.Y {type=MovieSceneScriptingFloatChannel, default=0, const [0,44]}
       foot_bk1_l_ctrl.Location.Z {type=MovieSceneScriptingFloatChannel, default=0, const [0,44]}
       foot_bk1_l_ctrl.Rotation.X {type=MovieSceneScriptingFloatChannel, default=0, const [0,44]}
       foot_bk1_l_ctrl.Rotation.Y {type=MovieSceneScriptingFloatChannel, default=0, const [0,44]}
       foot_bk1_l_ctrl.Rotation.Z {type=MovieSceneScriptingFloatChannel, default=0, const [0,44]}
       foot_bk1_l_ctrl.Scale.X {type=MovieSceneScriptingFloatChannel, default=1, const [1,44]}
       foot_bk1_l_ctrl.Scale.Y {type=MovieSceneScriptingFloatChannel, default=1, const [1,44]}
       foot_bk1_l_ctrl.Scale.Z {type=MovieSceneScriptingFloatChannel, default=1, const [1,44]}
       foot_roll_l_ctrl.Location.X {type=MovieSceneScriptingFloatChannel, default=0, const [0,44]}
       foot_roll_l_ctrl.Location.Y {type=MovieSceneScriptingFloatChannel, default=0, const [0,44]}
       foot_roll_l_ctrl.Location.Z {type=MovieSceneScriptingFloatChannel, default=0, const [0,44]}
       foot_roll_l_ctrl.Rotation.X {type=MovieSceneScriptingFloatChannel, default=0, const [0,44]}
       foot_roll_l_ctrl.Rotation.Y {type=MovieSceneScriptingFloatChannel, default=0, const [0,44]}
       foot_roll_l_ctrl.Rotation.Z {type=MovieSceneScriptingFloatChannel, default=0, const [0,44]}
       foot_roll_l_ctrl.Scale.X {type=MovieSceneScriptingFloatChannel, default=1, const [1,44]}
       foot_roll_l_ctrl.Scale.Y {type=MovieSceneScriptingFloatChannel, default=1, const [1,44]}
       foot_roll_l_ctrl.Scale.Z {type=MovieSceneScriptingFloatChannel, default=1, const [1,44]}
       roll_blend_l_ctrl {type=MovieSceneScriptingFloatChannel, default=1, const [1,44]}
       tip_l_ctrl.Location.X {type=MovieSceneScriptingFloatChannel, default=0, const [0,44]}
       tip_l_ctrl.Location.Y {type=MovieSceneScriptingFloatChannel, default=0, const [0,44]}
       tip_l_ctrl.Location.Z {type=MovieSceneScriptingFloatChannel, default=0, const [0,44]}
       tip_l_ctrl.Rotation.X {type=MovieSceneScriptingFloatChannel, default=0, const [0,44]}
       tip_l_ctrl.Rotation.Y {type=MovieSceneScriptingFloatChannel, default=0, const [0,44]}
       tip_l_ctrl.Rotation.Z {type=MovieSceneScriptingFloatChannel, default=0, const [0,44]}
       tip_l_ctrl.Scale.X {type=MovieSceneScriptingFloatChannel, default=1, const [1,44]}
       tip_l_ctrl.Scale.Y {type=MovieSceneScriptingFloatChannel, default=1, const [1,44]}
       tip_l_ctrl.Scale.Z {type=MovieSceneScriptingFloatChannel, default=1, const [1,44]}
       heel_l_ctrl.Location.X {type=MovieSceneScriptingFloatChannel, default=0, const [0,44]}
       heel_l_ctrl.Location.Y {type=MovieSceneScriptingFloatChannel, default=0, const [0,44]}
       heel_l_ctrl.Location.Z {type=MovieSceneScriptingFloatChannel, default=0, const [0,44]}
       heel_l_ctrl.Rotation.X {type=MovieSceneScriptingFloatChannel, default=0, const [0,44]}
       heel_l_ctrl.Rotation.Y {type=MovieSceneScriptingFloatChannel, default=0, const [0,44]}
       heel_l_ctrl.Rotation.Z {type=MovieSceneScriptingFloatChannel, default=0, const [0,44]}
       heel_l_ctrl.Scale.X {type=MovieSceneScriptingFloatChannel, default=1, const [1,44]}
       heel_l_ctrl.Scale.Y {type=MovieSceneScriptingFloatChannel, default=1, const [1,44]}
       heel_l_ctrl.Scale.Z {type=MovieSceneScriptingFloatChannel, default=1, const [1,44]}
       leg_l_softness {type=MovieSceneScriptingFloatChannel, default=0, const [0,44]}
       leg_l_pv_ik_ctrl.Location.X {
        type=MovieSceneScriptingFloatChannel
        default=0.3036
        thinned_from=44
        keys [[0,0.3036],[1,0.3001],[2,0.2953],[3,0.2894],[5,0.2753],[8,0.2508],[9,0.2459],[10,0.2342],[11,0.2279],[12,0.221],[13,0.2115],[14,0.2039],[16,0.1941],[17,0.1901],[18,0.1873],[19,0.1856],[20,0.1853],[21,0.1863],[22,0.1887],[23,0.1924],[24,0.1976],[25,0.2052],[26,0.2109],[27,0.2178],[28,0.2272],[30,0.2446],[31,0.2498],[32,0.2617],[33,0.2685],[34,0.2759],[36,0.2926],[37,0.2961],[38,0.3004],[39,0.3037],[40,0.3056],[41,0.3062],[42,0.3054],[43,0.3036]]
       }
       leg_l_pv_ik_ctrl.Location.Y {type=MovieSceneScriptingFloatChannel, default=8.7711, thinned_from=44, keys [[0,8.7711],[8,8.7745],[24,8.9923],[43,8.7711]]}
       leg_l_pv_ik_ctrl.Location.Z {type=MovieSceneScriptingFloatChannel, default=10.7791, thinned_from=44, keys [[0,10.7791],[6,10.7542],[24,11.113],[30,11.0875],[43,10.7791]]}
       leg_l_pv_ik_ctrl.Rotation.X {
        type=MovieSceneScriptingFloatChannel
        default=-13.6926
        thinned_from=44
        keys [[0,-13.6926],[7,-13.6744],[24,-14.0799],[32,-14.0077],[43,-13.6926]]
       }
       leg_l_pv_ik_ctrl.Rotation.Y {
        type=MovieSceneScriptingFloatChannel
        default=0.1216
        thinned_from=44
        keys [[0,0.1216],[5,0.1195],[8,0.119],[9,0.1192],[10,0.1192],[14,0.1202],[24,0.1258],[27,0.127],[30,0.1273],[31,0.1271],[32,0.1271],[36,0.1258],[43,0.1216]]
       }
       leg_l_pv_ik_ctrl.Rotation.Z {
        type=MovieSceneScriptingFloatChannel
        default=-0.2762
        thinned_from=44
        keys [[0,-0.2762],[1,-0.2724],[2,-0.2672],[3,-0.2606],[4,-0.2524],[5,-0.2446],[8,-0.2163],[9,-0.2105],[10,-0.1969],[11,-0.1894],[12,-0.1813],[13,-0.1701],[14,-0.1611],[16,-0.1491],[17,-0.1442],[18,-0.1405],[19,-0.1382],[20,-0.1375],[21,-0.1383],[22,-0.1406],[23,-0.1447],[24,-0.1503],[25,-0.159],[26,-0.1653],[27,-0.173],[28,-0.1838],[30,-0.2038],[31,-0.2101],[32,-0.2238],[33,-0.2319],[34,-0.2407],[36,-0.2605],[37,-0.265],[38,-0.2704],[39,-0.2746],[40,-0.2772],[41,-0.2784],[42,-0.2779],[43,-0.2762]]
       }
       leg_l_pv_ik_ctrl.Scale.X {type=MovieSceneScriptingFloatChannel, default=1, const [1,44]}
       leg_l_pv_ik_ctrl.Scale.Y {type=MovieSceneScriptingFloatChannel, default=1, const [1,44]}
       leg_l_pv_ik_ctrl.Scale.Z {type=MovieSceneScriptingFloatChannel, default=1, const [1,44]}
       thigh_r_fk_ctrl.Location.X {type=MovieSceneScriptingFloatChannel, default=0, thinned_from=44, keys [[0,0],[43,0]]}
       thigh_r_fk_ctrl.Location.Y {type=MovieSceneScriptingFloatChannel, default=0, thinned_from=44, keys [[0,0],[43,0]]}
       thigh_r_fk_ctrl.Location.Z {type=MovieSceneScriptingFloatChannel, default=0, thinned_from=44, keys [[0,0],[43,0]]}
       thigh_r_fk_ctrl.Rotation.X {type=MovieSceneScriptingFloatChannel, default=6.2177, thinned_from=44, keys [[0,6.2177],[24,6.3501],[41,6.2268],[43,6.2268]]}
       thigh_r_fk_ctrl.Rotation.Y {
        type=MovieSceneScriptingFloatChannel
        default=15.1659
        thinned_from=44
        keys [[0,15.1659],[5,14.9736],[15,14.4029],[19,14.283],[22,14.2858],[26,14.4316],[37,15.0743],[41,15.1733],[43,15.1733]]
       }
       thigh_r_fk_ctrl.Rotation.Z {
        type=MovieSceneScriptingFloatChannel
        default=-5.26
        thinned_from=44
        keys [[0,-5.26],[1,-5.2965],[2,-5.3494],[3,-5.4166],[5,-5.587],[8,-5.9029],[13,-6.4618],[15,-6.6515],[17,-6.7988],[18,-6.8517],[19,-6.888],[20,-6.9066],[21,-6.9066],[22,-6.8886],[23,-6.8511],[24,-6.7954],[25,-6.7239],[27,-6.5419],[34,-5.7327],[36,-5.5272],[37,-5.4403],[38,-5.3669],[39,-5.3093],[40,-5.2706],[41,-5.2509],[43,-5.2509]]
       }
       thigh_r_fk_ctrl.Scale.X {type=MovieSceneScriptingFloatChannel, default=1, const [1,44]}
       thigh_r_fk_ctrl.Scale.Y {type=MovieSceneScriptingFloatChannel, default=1, const [1,44]}
       thigh_r_fk_ctrl.Scale.Z {type=MovieSceneScriptingFloatChannel, default=1, const [1,44]}
       calf_r_fk_ctrl.Location.X {type=MovieSceneScriptingFloatChannel, default=0.1146, thinned_from=44, keys [[0,0.1146],[14,0.1167],[39,0.1142],[43,0.1146]]}
       calf_r_fk_ctrl.Location.Y {type=MovieSceneScriptingFloatChannel, default=0.0095, thinned_from=44, keys [[0,0.0095],[43,0.0095]]}
       calf_r_fk_ctrl.Location.Z {type=MovieSceneScriptingFloatChannel, default=0.0038, thinned_from=44, keys [[0,0.0038],[43,0.0038]]}
       calf_r_fk_ctrl.Rotation.X {type=MovieSceneScriptingFloatChannel, default=0.9558, thinned_from=44, keys [[0,0.9558],[14,0.9713],[39,0.9525],[43,0.9558]]}
       calf_r_fk_ctrl.Rotation.Y {type=MovieSceneScriptingFloatChannel, default=0.2158, thinned_from=44, keys [[0,0.2158],[43,0.2158]]}
       calf_r_fk_ctrl.Rotation.Z {type=MovieSceneScriptingFloatChannel, default=27.5166, thinned_from=44, keys [[0,27.5166],[14,28.0464],[36,27.403],[43,27.5166]]}
       calf_r_fk_ctrl.Scale.X {type=MovieSceneScriptingFloatChannel, default=1, const [1,44]}
       calf_r_fk_ctrl.Scale.Y {type=MovieSceneScriptingFloatChannel, default=1, const [1,44]}
       calf_r_fk_ctrl.Scale.Z {type=MovieSceneScriptingFloatChannel, default=1, const [1,44]}
       foot_r_fk_ctrl.Location.X {type=MovieSceneScriptingFloatChannel, default=0, thinned_from=44, keys [[0,0],[43,0]]}
       foot_r_fk_ctrl.Location.Y {type=MovieSceneScriptingFloatChannel, default=0, thinned_from=44, keys [[0,0],[43,0]]}
       foot_r_fk_ctrl.Location.Z {type=MovieSceneScriptingFloatChannel, default=0, thinned_from=44, keys [[0,0],[43,0]]}
       foot_r_fk_ctrl.Rotation.X {type=MovieSceneScriptingFloatChannel, default=-2.9896, thinned_from=44, keys [[0,-2.9896],[43,-2.9896]]}
       foot_r_fk_ctrl.Rotation.Y {
        type=MovieSceneScriptingFloatChannel
        default=-9.7343
        thinned_from=44
        keys [[0,-9.7343],[9,-9.5655],[10,-9.5655],[11,-9.5445],[19,-9.5577],[31,-9.7775],[32,-9.7775],[38,-9.7994],[43,-9.7343]]
       }
       foot_r_fk_ctrl.Rotation.Z {type=MovieSceneScriptingFloatChannel, default=-15.9769, thinned_from=44, keys [[0,-15.9769],[10,-16.0386],[30,-15.7926],[43,-15.9769]]}
       foot_r_fk_ctrl.Scale.X {type=MovieSceneScriptingFloatChannel, default=1, const [1,44]}
       foot_r_fk_ctrl.Scale.Y {type=MovieSceneScriptingFloatChannel, default=1, const [1,44]}
       foot_r_fk_ctrl.Scale.Z {type=MovieSceneScriptingFloatChannel, default=1, const [1,44]}
       ball_r_fk_ctrl.Location.X {type=MovieSceneScriptingFloatChannel, default=0, const [0,44]}
       ball_r_fk_ctrl.Location.Y {type=MovieSceneScriptingFloatChannel, default=0, const [0,44]}
       ball_r_fk_ctrl.Location.Z {type=MovieSceneScriptingFloatChannel, default=0, const [0,44]}
       ball_r_fk_ctrl.Rotation.X {type=MovieSceneScriptingFloatChannel, default=0, const [0,44]}
       ball_r_fk_ctrl.Rotation.Y {type=MovieSceneScriptingFloatChannel, default=0, const [0,44]}
       ball_r_fk_ctrl.Rotation.Z {type=MovieSceneScriptingFloatChannel, default=0, const [0,44]}
       ball_r_fk_ctrl.Scale.X {type=MovieSceneScriptingFloatChannel, default=1, const [1,44]}
       ball_r_fk_ctrl.Scale.Y {type=MovieSceneScriptingFloatChannel, default=1, const [1,44]}
       ball_r_fk_ctrl.Scale.Z {type=MovieSceneScriptingFloatChannel, default=1, const [1,44]}
       foot_r_ik_ctrl.Location.X {
        type=MovieSceneScriptingFloatChannel
        default=-0.8486
        thinned_from=44
        keys [[0,-0.8486],[14,-0.848],[15,-0.8434],[16,-0.8478],[17,-0.8492],[36,-0.8484],[37,-0.8511],[38,-0.8505],[39,-0.8458],[40,-0.8486],[43,-0.8491]]
       }
       foot_r_ik_ctrl.Location.Y {type=MovieSceneScriptingFloatChannel, default=-19.6649, thinned_from=44, keys [[0,-19.6649],[43,-19.6779]]}
       foot_r_ik_ctrl.Location.Z {
        type=MovieSceneScriptingFloatChannel
        default=0.2137
        thinned_from=44
        keys [[0,0.2137],[14,0.2128],[15,0.2101],[16,0.2127],[17,0.2135],[18,0.2126],[20,0.2125],[21,0.2128],[36,0.2134],[37,0.2151],[38,0.2146],[39,0.2121],[40,0.2136],[41,0.2136],[42,0.2121],[43,0.217]]
       }
       foot_r_ik_ctrl.Rotation.X {
        type=MovieSceneScriptingFloatChannel
        default=0.9359
        thinned_from=44
        keys [[0,0.9359],[9,0.934],[10,0.9262],[11,0.9332],[12,0.9215],[13,0.9325],[14,0.9321],[15,0.9224],[16,0.9315],[17,0.9341],[18,0.931],[20,0.9306],[21,0.9324],[22,0.9305],[31,0.9326],[32,0.9413],[33,0.9334],[36,0.9345],[37,0.9404],[38,0.939],[39,0.9299],[40,0.9356],[41,0.9358],[42,0.9317],[43,0.9414]]
       }
       foot_r_ik_ctrl.Rotation.Y {
        type=MovieSceneScriptingFloatChannel
        default=0.5216
        thinned_from=44
        keys [[0,0.5216],[9,0.5197],[10,0.5082],[11,0.5192],[12,0.5121],[13,0.5186],[31,0.5185],[32,0.5297],[33,0.5194],[41,0.5213],[42,0.5189],[43,0.5284]]
       }
       foot_r_ik_ctrl.Rotation.Z {type=MovieSceneScriptingFloatChannel, default=-36.751, thinned_from=44, keys [[0,-36.751],[43,-36.761]]}
       foot_r_ik_ctrl.Scale.X {type=MovieSceneScriptingFloatChannel, default=1, const [1,44]}
       foot_r_ik_ctrl.Scale.Y {type=MovieSceneScriptingFloatChannel, default=1, const [1,44]}
       foot_r_ik_ctrl.Scale.Z {type=MovieSceneScriptingFloatChannel, default=1, const [1,44]}
       ball_r_ik_ctrl.Location.X {type=MovieSceneScriptingFloatChannel, default=0, const [0,44]}
       ball_r_ik_ctrl.Location.Y {type=MovieSceneScriptingFloatChannel, default=0, const [0,44]}
       ball_r_ik_ctrl.Location.Z {type=MovieSceneScriptingFloatChannel, default=0, const [0,44]}
       ball_r_ik_ctrl.Rotation.X {type=MovieSceneScriptingFloatChannel, default=0, const [0,44]}
       ball_r_ik_ctrl.Rotation.Y {type=MovieSceneScriptingFloatChannel, default=0, const [0,44]}
       ball_r_ik_ctrl.Rotation.Z {type=MovieSceneScriptingFloatChannel, default=0, const [0,44]}
       ball_r_ik_ctrl.Scale.X {type=MovieSceneScriptingFloatChannel, default=1, const [1,44]}
       ball_r_ik_ctrl.Scale.Y {type=MovieSceneScriptingFloatChannel, default=1, const [1,44]}
       ball_r_ik_ctrl.Scale.Z {type=MovieSceneScriptingFloatChannel, default=1, const [1,44]}
       foot_bk1_r_ctrl.Location.X {type=MovieSceneScriptingFloatChannel, default=0, const [0,44]}
       foot_bk1_r_ctrl.Location.Y {type=MovieSceneScriptingFloatChannel, default=0, const [0,44]}
       foot_bk1_r_ctrl.Location.Z {type=MovieSceneScriptingFloatChannel, default=0, const [0,44]}
       foot_bk1_r_ctrl.Rotation.X {type=MovieSceneScriptingFloatChannel, default=0, const [0,44]}
       foot_bk1_r_ctrl.Rotation.Y {type=MovieSceneScriptingFloatChannel, default=0, const [0,44]}
       foot_bk1_r_ctrl.Rotation.Z {type=MovieSceneScriptingFloatChannel, default=0, const [0,44]}
       foot_bk1_r_ctrl.Scale.X {type=MovieSceneScriptingFloatChannel, default=1, const [1,44]}
       foot_bk1_r_ctrl.Scale.Y {type=MovieSceneScriptingFloatChannel, default=1, const [1,44]}
       foot_bk1_r_ctrl.Scale.Z {type=MovieSceneScriptingFloatChannel, default=1, const [1,44]}
       foot_roll_r_ctrl.Location.X {type=MovieSceneScriptingFloatChannel, default=0, const [0,44]}
       foot_roll_r_ctrl.Location.Y {type=MovieSceneScriptingFloatChannel, default=0, const [0,44]}
       foot_roll_r_ctrl.Location.Z {type=MovieSceneScriptingFloatChannel, default=0, const [0,44]}
       foot_roll_r_ctrl.Rotation.X {type=MovieSceneScriptingFloatChannel, default=0, const [0,44]}
       foot_roll_r_ctrl.Rotation.Y {type=MovieSceneScriptingFloatChannel, default=0, const [0,44]}
       foot_roll_r_ctrl.Rotation.Z {type=MovieSceneScriptingFloatChannel, default=0, const [0,44]}
       foot_roll_r_ctrl.Scale.X {type=MovieSceneScriptingFloatChannel, default=1, const [1,44]}
       foot_roll_r_ctrl.Scale.Y {type=MovieSceneScriptingFloatChannel, default=1, const [1,44]}
       foot_roll_r_ctrl.Scale.Z {type=MovieSceneScriptingFloatChannel, default=1, const [1,44]}
       roll_blend_r_ctrl {type=MovieSceneScriptingFloatChannel, default=1, const [1,44]}
       tip_r_ctrl.Location.X {type=MovieSceneScriptingFloatChannel, default=0, const [0,44]}
       tip_r_ctrl.Location.Y {type=MovieSceneScriptingFloatChannel, default=0, const [0,44]}
       tip_r_ctrl.Location.Z {type=MovieSceneScriptingFloatChannel, default=0, const [0,44]}
       tip_r_ctrl.Rotation.X {type=MovieSceneScriptingFloatChannel, default=0, const [0,44]}
       tip_r_ctrl.Rotation.Y {type=MovieSceneScriptingFloatChannel, default=0, const [0,44]}
       tip_r_ctrl.Rotation.Z {type=MovieSceneScriptingFloatChannel, default=0, const [0,44]}
       tip_r_ctrl.Scale.X {type=MovieSceneScriptingFloatChannel, default=1, const [1,44]}
       tip_r_ctrl.Scale.Y {type=MovieSceneScriptingFloatChannel, default=1, const [1,44]}
       tip_r_ctrl.Scale.Z {type=MovieSceneScriptingFloatChannel, default=1, const [1,44]}
       heel_r_ctrl.Location.X {type=MovieSceneScriptingFloatChannel, default=0, const [0,44]}
       heel_r_ctrl.Location.Y {type=MovieSceneScriptingFloatChannel, default=0, const [0,44]}
       heel_r_ctrl.Location.Z {type=MovieSceneScriptingFloatChannel, default=0, const [0,44]}
       heel_r_ctrl.Rotation.X {type=MovieSceneScriptingFloatChannel, default=0, const [0,44]}
       heel_r_ctrl.Rotation.Y {type=MovieSceneScriptingFloatChannel, default=0, const [0,44]}
       heel_r_ctrl.Rotation.Z {type=MovieSceneScriptingFloatChannel, default=0, const [0,44]}
       heel_r_ctrl.Scale.X {type=MovieSceneScriptingFloatChannel, default=1, const [1,44]}
       heel_r_ctrl.Scale.Y {type=MovieSceneScriptingFloatChannel, default=1, const [1,44]}
       heel_r_ctrl.Scale.Z {type=MovieSceneScriptingFloatChannel, default=1, const [1,44]}
       leg_r_softness {type=MovieSceneScriptingFloatChannel, default=0, const [0,44]}
       leg_r_pv_ik_ctrl.Location.X {
        type=MovieSceneScriptingFloatChannel
        default=1.2089
        thinned_from=44
        keys [[0,1.2089],[4,1.185],[9,1.1466],[10,1.1409],[11,1.1323],[12,1.128],[13,1.1198],[14,1.1145],[15,1.1117],[17,1.1038],[19,1.1029],[21,1.1067],[24,1.1211],[31,1.1754],[32,1.1812],[33,1.1903],[36,1.2074],[39,1.2165],[43,1.2089]]
       }
       leg_r_pv_ik_ctrl.Location.Y {type=MovieSceneScriptingFloatChannel, default=9.1525, thinned_from=44, keys [[0,9.1525],[14,9.3066],[39,9.1201],[43,9.1525]]}
       leg_r_pv_ik_ctrl.Location.Z {type=MovieSceneScriptingFloatChannel, default=11.5253, thinned_from=44, keys [[0,11.5253],[14,11.7583],[36,11.4734],[43,11.5253]]}
       leg_r_pv_ik_ctrl.Rotation.X {type=MovieSceneScriptingFloatChannel, default=-14.5496, thinned_from=44, keys [[0,-14.5496],[14,-14.8174],[36,-14.4899],[43,-14.5496]]}
       leg_r_pv_ik_ctrl.Rotation.Y {type=MovieSceneScriptingFloatChannel, default=0.184, thinned_from=44, keys [[0,0.184],[10,0.1864],[32,0.1824],[43,0.184]]}
       leg_r_pv_ik_ctrl.Rotation.Z {
        type=MovieSceneScriptingFloatChannel
        default=-1.2815
        thinned_from=44
        keys [[0,-1.2815],[4,-1.2522],[9,-1.2062],[10,-1.1994],[11,-1.1893],[12,-1.1842],[13,-1.1747],[14,-1.1686],[15,-1.1654],[16,-1.1596],[17,-1.1563],[19,-1.1557],[21,-1.1608],[23,-1.1714],[31,-1.2437],[32,-1.2508],[33,-1.2614],[36,-1.2814],[39,-1.2916],[43,-1.2815]]
       }
       leg_r_pv_ik_ctrl.Scale.X {type=MovieSceneScriptingFloatChannel, default=1, const [1,44]}
       leg_r_pv_ik_ctrl.Scale.Y {type=MovieSceneScriptingFloatChannel, default=1, const [1,44]}
       leg_r_pv_ik_ctrl.Scale.Z {type=MovieSceneScriptingFloatChannel, default=1, const [1,44]}
       upperarm_l_fk_ctrl.Location.X {type=MovieSceneScriptingFloatChannel, default=0, thinned_from=44, keys [[0,0],[43,0]]}
       upperarm_l_fk_ctrl.Location.Y {type=MovieSceneScriptingFloatChannel, default=0, thinned_from=44, keys [[0,0],[43,0]]}
       upperarm_l_fk_ctrl.Location.Z {type=MovieSceneScriptingFloatChannel, default=0, thinned_from=44, keys [[0,0],[43,0]]}
       upperarm_l_fk_ctrl.Rotation.X {
        type=MovieSceneScriptingFloatChannel
        default=38.0713
        thinned_from=44
        keys [[0,38.0713],[3,38.8044],[7,39.3943],[12,39.7334],[16,39.7008],[19,39.4475],[21,39.1651],[30,37.0322],[32,36.7316],[34,36.6309],[36,36.7515],[43,38.0713]]
       }
       upperarm_l_fk_ctrl.Rotation.Y {
        type=MovieSceneScriptingFloatChannel
        default=27.3335
        thinned_from=44
        keys [[0,27.3335],[3,27.3663],[8,27.7564],[17,28.5752],[21,28.7465],[25,28.578],[36,27.5285],[41,27.3205],[43,27.3335]]
       }
       upperarm_l_fk_ctrl.Rotation.Z {
        type=MovieSceneScriptingFloatChannel
        default=37.9475
        thinned_from=44
        keys [[0,37.9475],[16,39.1785],[27,38.9471],[39,37.9588],[43,37.9475]]
       }
       upperarm_l_fk_ctrl.Scale.X {type=MovieSceneScriptingFloatChannel, default=1, const [1,44]}
       upperarm_l_fk_ctrl.Scale.Y {type=MovieSceneScriptingFloatChannel, default=1, const [1,44]}
       upperarm_l_fk_ctrl.Scale.Z {type=MovieSceneScriptingFloatChannel, default=1, const [1,44]}
       lowerarm_l_fk_ctrl.Location.X {type=MovieSceneScriptingFloatChannel, default=0, thinned_from=44, keys [[0,0],[43,0]]}
       lowerarm_l_fk_ctrl.Location.Y {type=MovieSceneScriptingFloatChannel, default=0, thinned_from=44, keys [[0,0],[43,0]]}
       lowerarm_l_fk_ctrl.Location.Z {type=MovieSceneScriptingFloatChannel, default=0, thinned_from=44, keys [[0,0],[43,0]]}
       lowerarm_l_fk_ctrl.Rotation.X {type=MovieSceneScriptingFloatChannel, default=-3.3373, thinned_from=44, keys [[0,-3.3373],[35,-3.3137],[43,-3.3373]]}
       lowerarm_l_fk_ctrl.Rotation.Y {
        type=MovieSceneScriptingFloatChannel
        default=1.6523
        thinned_from=44
        keys [[0,1.6523],[6,1.6575],[10,1.6471],[20,1.5839],[23,1.5737],[27,1.5748],[36,1.6124],[43,1.6523]]
       }
       lowerarm_l_fk_ctrl.Rotation.Z {
        type=MovieSceneScriptingFloatChannel
        default=47.7692
        thinned_from=44
        keys [[0,47.7692],[3,47.4554],[6,47.4676],[9,47.8521],[11,48.3478],[19,51.373],[21,51.9405],[23,52.2562],[26,52.2819],[30,51.6957],[35,50.395],[42,48.02],[43,47.7692]]
       }
       lowerarm_l_fk_ctrl.Scale.X {type=MovieSceneScriptingFloatChannel, default=1, const [1,44]}
       lowerarm_l_fk_ctrl.Scale.Y {type=MovieSceneScriptingFloatChannel, default=1, const [1,44]}
       lowerarm_l_fk_ctrl.Scale.Z {type=MovieSceneScriptingFloatChannel, default=1, const [1,44]}
       hand_l_fk_ctrl.Location.X {type=MovieSceneScriptingFloatChannel, default=0, thinned_from=44, keys [[0,0],[43,0]]}
       hand_l_fk_ctrl.Location.Y {type=MovieSceneScriptingFloatChannel, default=0, thinned_from=44, keys [[0,0],[43,0]]}
       hand_l_fk_ctrl.Location.Z {type=MovieSceneScriptingFloatChannel, default=0, thinned_from=44, keys [[0,0],[43,0]]}
       hand_l_fk_ctrl.Rotation.X {
        type=MovieSceneScriptingFloatChannel
        default=-7.5742
        thinned_from=44
        keys [[0,-7.5742],[2,-7.6826],[4,-7.7454],[6,-7.7611],[8,-7.7305],[10,-7.6567],[12,-7.54],[14,-7.3815],[21,-6.697],[23,-6.5391],[25,-6.4286],[26,-6.3923],[27,-6.3688],[28,-6.3586],[29,-6.3621],[30,-6.3799],[31,-6.4126],[32,-6.4602],[33,-6.523],[34,-6.6028],[36,-6.8082],[41,-7.4079],[42,-7.5038],[43,-7.5742]]
       }
       hand_l_fk_ctrl.Rotation.Y {
        type=MovieSceneScriptingFloatChannel
        default=7.2316
        thinned_from=44
        keys [[0,7.2316],[2,7.9599],[3,8.3027],[4,8.6173],[5,8.8929],[6,9.1191],[7,9.2851],[8,9.3811],[9,9.3978],[10,9.3253],[11,9.1542],[12,8.9051],[13,8.6089],[14,8.2737],[15,7.907],[16,7.5166],[20,5.8679],[23,4.6956],[24,4.3123],[25,3.9464],[26,3.6082],[27,3.3079],[28,3.0555],[29,2.8611],[30,2.7353],[31,2.6878],[32,2.7282],[33,2.8662],[34,3.0992],[35,3.4105],[36,3.7879],[37,4.2194],[38,4.6928],[39,5.196],[41,6.2439],[42,6.7546],[43,7.2316]]
       }
       hand_l_fk_ctrl.Rotation.Z {
        type=MovieSceneScriptingFloatChannel
        default=-9.6446
        thinned_from=44
        keys [[0,-9.6446],[1,-9.7657],[2,-9.8489],[3,-9.9026],[8,-10.0139],[9,-10.0622],[10,-10.1384],[11,-10.2509],[15,-10.7673],[17,-10.9726],[19,-11.1068],[20,-11.1382],[21,-11.1404],[23,-11.0735],[25,-10.9351],[27,-10.7385],[30,-10.3645],[33,-9.9405],[34,-9.8171],[35,-9.7297],[36,-9.6728],[37,-9.6409],[43,-9.6446]]
       }
       hand_l_fk_ctrl.Scale.X {type=MovieSceneScriptingFloatChannel, default=1, const [1,44]}
       hand_l_fk_ctrl.Scale.Y {type=MovieSceneScriptingFloatChannel, default=1, const [1,44]}
       hand_l_fk_ctrl.Scale.Z {type=MovieSceneScriptingFloatChannel, default=1, const [1,44]}
       thumb_01_l_ctrl.Location.X {type=MovieSceneScriptingFloatChannel, default=0, thinned_from=44, keys [[0,0],[43,0]]}
       thumb_01_l_ctrl.Location.Y {type=MovieSceneScriptingFloatChannel, default=0, const [0,44]}
       thumb_01_l_ctrl.Location.Z {type=MovieSceneScriptingFloatChannel, default=0, thinned_from=44, keys [[0,0],[43,0]]}
       thumb_01_l_ctrl.Rotation.X {
        type=MovieSceneScriptingFloatChannel
        default=5.4807
        thinned_from=44
        keys [[0,5.4807],[1,5.4673],[2,5.4284],[3,5.3668],[4,5.2849],[5,5.1852],[6,5.0703],[7,4.9428],[9,4.6598],[13,4.0534],[15,3.7704],[16,3.6429],[17,3.528],[18,3.4283],[19,3.3464],[20,3.2848],[21,3.2459],[22,3.2323],[23,3.2472],[24,3.2897],[25,3.3569],[26,3.446],[27,3.554],[28,3.6781],[29,3.8153],[31,4.1173],[35,4.7505],[36,4.8978],[37,5.035],[38,5.1591],[39,5.2671],[40,5.3562],[41,5.4235],[42,5.466],[43,5.4808]]
       }
       thumb_01_l_ctrl.Rotation.Y {
        type=MovieSceneScriptingFloatChannel
        default=0.8396
        thinned_from=44
        keys [[0,0.8396],[1,0.8345],[2,0.82],[3,0.797],[4,0.7665],[5,0.7294],[6,0.6866],[7,0.6391],[8,0.5878],[9,0.5337],[12,0.3639],[13,0.3079],[14,0.2538],[15,0.2025],[16,0.155],[17,0.1122],[18,0.075],[19,0.0445],[20,0.0216],[21,0.0071],[22,0.0021],[23,0.0076],[24,0.0234],[25,0.0485],[26,0.0816],[27,0.1219],[28,0.1681],[29,0.2192],[30,0.2741],[31,0.3317],[34,0.5099],[35,0.5675],[36,0.6224],[37,0.6735],[38,0.7197],[39,0.76],[40,0.7931],[41,0.8182],[42,0.834],[43,0.8395]]
       }
       thumb_01_l_ctrl.Rotation.Z {
        type=MovieSceneScriptingFloatChannel
        default=-0.1513
        thinned_from=44
        keys [[0,-0.1513],[1,-0.1713],[2,-0.2293],[3,-0.3213],[4,-0.4435],[5,-0.5922],[6,-0.7636],[7,-0.954],[8,-1.1594],[9,-1.3761],[10,-1.6005],[12,-2.0567],[13,-2.281],[14,-2.4978],[15,-2.7032],[16,-2.8935],[17,-3.0649],[18,-3.2136],[19,-3.3359],[20,-3.4279],[21,-3.4859],[22,-3.5061],[23,-3.4839],[24,-3.4205],[25,-3.3202],[26,-3.1873],[27,-3.0261],[28,-2.841],[29,-2.6363],[30,-2.4164],[31,-2.1857],[34,-1.4717],[35,-1.2409],[36,-1.021],[37,-0.8163],[38,-0.6312],[39,-0.47],[40,-0.3371],[41,-0.2367],[42,-0.1733],[43,-0.1512]]
       }
       thumb_01_l_ctrl.Scale.X {type=MovieSceneScriptingFloatChannel, default=1, const [1,44]}
       thumb_01_l_ctrl.Scale.Y {type=MovieSceneScriptingFloatChannel, default=1, const [1,44]}
       thumb_01_l_ctrl.Scale.Z {type=MovieSceneScriptingFloatChannel, default=1, const [1,44]}
       thumb_02_l_ctrl.Location.X {type=MovieSceneScriptingFloatChannel, default=0, thinned_from=44, keys [[0,0],[43,0]]}
       thumb_02_l_ctrl.Location.Y {type=MovieSceneScriptingFloatChannel, default=0, thinned_from=44, keys [[0,0],[43,0]]}
       thumb_02_l_ctrl.Location.Z {type=MovieSceneScriptingFloatChannel, default=0, thinned_from=44, keys [[0,0],[43,0]]}
       thumb_02_l_ctrl.Rotation.X {type=MovieSceneScriptingFloatChannel, default=0, thinned_from=44, keys [[0,0],[43,0]]}
       thumb_02_l_ctrl.Rotation.Y {type=MovieSceneScriptingFloatChannel, default=0, thinned_from=44, keys [[0,0],[43,0]]}
       thumb_02_l_ctrl.Rotation.Z {
        type=MovieSceneScriptingFloatChannel
        default=17.7895
        thinned_from=44
        keys [[0,17.7895],[1,17.8105],[2,17.7913],[4,17.6485],[6,17.3906],[9,16.8507],[16,15.3809],[18,15.0366],[20,14.7787],[21,14.6911],[22,14.6359],[23,14.6167],[24,14.6377],[25,14.6981],[26,14.7936],[28,15.0736],[30,15.4447],[37,16.9825],[39,17.3536],[41,17.6336],[43,17.7895]]
       }
       thumb_02_l_ctrl.Scale.X {type=MovieSceneScriptingFloatChannel, default=1, const [1,44]}
       thumb_02_l_ctrl.Scale.Y {type=MovieSceneScriptingFloatChannel, default=1, const [1,44]}
       thumb_02_l_ctrl.Scale.Z {type=MovieSceneScriptingFloatChannel, default=1, const [1,44]}
       thumb_03_l_ctrl.Location.X {type=MovieSceneScriptingFloatChannel, default=0, thinned_from=44, keys [[0,0],[43,0]]}
       thumb_03_l_ctrl.Location.Y {type=MovieSceneScriptingFloatChannel, default=0, thinned_from=44, keys [[0,0],[43,0]]}
       thumb_03_l_ctrl.Location.Z {type=MovieSceneScriptingFloatChannel, default=0, thinned_from=44, keys [[0,0],[43,0]]}
       thumb_03_l_ctrl.Rotation.X {type=MovieSceneScriptingFloatChannel, default=0, thinned_from=44, keys [[0,0],[43,0]]}
       thumb_03_l_ctrl.Rotation.Y {type=MovieSceneScriptingFloatChannel, default=0, thinned_from=44, keys [[0,0],[43,0]]}
       thumb_03_l_ctrl.Rotation.Z {
        type=MovieSceneScriptingFloatChannel
        default=-0.0803
        thinned_from=44
        keys [[0,-0.0803],[1,-0.0208],[2,-0.0208],[3,-0.0189],[4,-0.0734],[5,-0.1599],[6,-0.2748],[7,-0.4145],[8,-0.5756],[9,-0.7544],[10,-0.9474],[11,-1.1511],[12,-1.3619],[14,-1.7906],[15,-2.0015],[16,-2.2051],[17,-2.3982],[18,-2.577],[19,-2.7381],[20,-2.8778],[21,-2.9927],[22,-3.0792],[23,-3.1336],[24,-3.1526],[25,-3.1318],[26,-3.0722],[27,-2.9779],[28,-2.853],[29,-2.7015],[30,-2.5276],[31,-2.3352],[32,-2.1286],[33,-1.9118],[36,-1.2408],[37,-1.024],[38,-0.8173],[39,-0.625],[40,-0.451],[41,-0.2996],[42,-0.1746],[43,-0.0803]]
       }
       thumb_03_l_ctrl.Scale.X {type=MovieSceneScriptingFloatChannel, default=1, const [1,44]}
       thumb_03_l_ctrl.Scale.Y {type=MovieSceneScriptingFloatChannel, default=1, const [1,44]}
       thumb_03_l_ctrl.Scale.Z {type=MovieSceneScriptingFloatChannel, default=1, const [1,44]}
       index_metacarpal_l_ctrl.Location.X {type=MovieSceneScriptingFloatChannel, default=0, const [0,44]}
       index_metacarpal_l_ctrl.Location.Y {type=MovieSceneScriptingFloatChannel, default=0, const [0,44]}
       index_metacarpal_l_ctrl.Location.Z {type=MovieSceneScriptingFloatChannel, default=0, const [0,44]}
       index_metacarpal_l_ctrl.Rotation.X {type=MovieSceneScriptingFloatChannel, default=0, const [0,44]}
       index_metacarpal_l_ctrl.Rotation.Y {type=MovieSceneScriptingFloatChannel, default=0, const [0,44]}
       index_metacarpal_l_ctrl.Rotation.Z {type=MovieSceneScriptingFloatChannel, default=0, const [0,44]}
       index_metacarpal_l_ctrl.Scale.X {type=MovieSceneScriptingFloatChannel, default=1, const [1,44]}
       index_metacarpal_l_ctrl.Scale.Y {type=MovieSceneScriptingFloatChannel, default=1, const [1,44]}
       index_metacarpal_l_ctrl.Scale.Z {type=MovieSceneScriptingFloatChannel, default=1, const [1,44]}
       index_01_l_ctrl.Location.X {type=MovieSceneScriptingFloatChannel, default=0, const [0,44]}
       index_01_l_ctrl.Location.Y {type=MovieSceneScriptingFloatChannel, default=0, thinned_from=44, keys [[0,0],[43,0]]}
       index_01_l_ctrl.Location.Z {type=MovieSceneScriptingFloatChannel, default=0, const [0,44]}
       index_01_l_ctrl.Rotation.X {type=MovieSceneScriptingFloatChannel, default=0, thinned_from=44, keys [[0,0],[43,0]]}
       index_01_l_ctrl.Rotation.Y {type=MovieSceneScriptingFloatChannel, default=0, thinned_from=44, keys [[0,0],[43,-0.0001]]}
       index_01_l_ctrl.Rotation.Z {
        type=MovieSceneScriptingFloatChannel
        default=27.5406
        thinned_from=44
        keys [[0,27.5406],[1,27.7644],[2,27.9058],[3,27.9551],[4,27.9101],[5,27.7808],[6,27.5756],[7,27.3029],[9,26.5888],[11,25.7061],[17,22.7206],[19,21.8378],[20,21.4555],[21,21.1238],[22,20.8511],[23,20.6459],[24,20.5165],[25,20.4716],[26,20.5209],[27,20.6623],[28,20.8861],[29,21.1827],[30,21.5423],[32,22.4118],[39,26.0149],[41,26.8844],[42,27.244],[43,27.5405]]
       }
       index_01_l_ctrl.Scale.X {type=MovieSceneScriptingFloatChannel, default=1, const [1,44]}
       index_01_l_ctrl.Scale.Y {type=MovieSceneScriptingFloatChannel, default=1, const [1,44]}
       index_01_l_ctrl.Scale.Z {type=MovieSceneScriptingFloatChannel, default=1, const [1,44]}
       index_02_l_ctrl.Location.X {type=MovieSceneScriptingFloatChannel, default=0, thinned_from=44, keys [[0,0],[43,0]]}
       index_02_l_ctrl.Location.Y {type=MovieSceneScriptingFloatChannel, default=0, thinned_from=44, keys [[0,0],[43,0]]}
       index_02_l_ctrl.Location.Z {type=MovieSceneScriptingFloatChannel, default=0, thinned_from=44, keys [[0,0],[43,0]]}
       index_02_l_ctrl.Rotation.X {type=MovieSceneScriptingFloatChannel, default=0, thinned_from=44, keys [[0,0],[43,0]]}
       index_02_l_ctrl.Rotation.Y {type=MovieSceneScriptingFloatChannel, default=0, thinned_from=44, keys [[0,0],[43,0]]}
       index_02_l_ctrl.Rotation.Z {
        type=MovieSceneScriptingFloatChannel
        default=17.1565
        thinned_from=44
        keys [[0,17.1565],[1,17.3362],[2,17.4718],[3,17.5574],[4,17.5873],[5,17.56],[6,17.4817],[7,17.3574],[8,17.1922],[10,16.7597],[12,16.225],[18,14.4165],[20,13.8818],[21,13.6503],[22,13.4493],[23,13.2841],[24,13.1598],[25,13.0815],[26,13.0543],[27,13.0841],[28,13.1698],[29,13.3053],[30,13.485],[31,13.7028],[33,14.2295],[40,16.412],[42,16.9387],[43,17.1565]]
       }
       index_02_l_ctrl.Scale.X {type=MovieSceneScriptingFloatChannel, default=1, const [1,44]}
       index_02_l_ctrl.Scale.Y {type=MovieSceneScriptingFloatChannel, default=1, const [1,44]}
       index_02_l_ctrl.Scale.Z {type=MovieSceneScriptingFloatChannel, default=1, const [1,44]}
       index_03_l_ctrl.Location.X {type=MovieSceneScriptingFloatChannel, default=-0.0001, thinned_from=44, keys [[0,-0.0001],[43,-0.0001]]}
       index_03_l_ctrl.Location.Y {type=MovieSceneScriptingFloatChannel, default=0.0003, thinned_from=44, keys [[0,0.0003],[43,0.0003]]}
       index_03_l_ctrl.Location.Z {type=MovieSceneScriptingFloatChannel, default=0, thinned_from=44, keys [[0,0],[43,0]]}
       index_03_l_ctrl.Rotation.X {type=MovieSceneScriptingFloatChannel, default=0, thinned_from=44, keys [[0,0],[43,0]]}
       index_03_l_ctrl.Rotation.Y {type=MovieSceneScriptingFloatChannel, default=0, thinned_from=44, keys [[0,0],[43,0]]}
       index_03_l_ctrl.Rotation.Z {
        type=MovieSceneScriptingFloatChannel
        default=-2.5272
        thinned_from=44
        keys [[0,-2.5272],[1,-2.3123],[2,-2.1351],[3,-2.0014],[4,-1.9169],[5,-1.8875],[6,-1.9143],[7,-1.9916],[8,-2.1142],[9,-2.2771],[10,-2.4753],[11,-2.7038],[12,-2.9574],[13,-3.2312],[14,-3.5201],[18,-4.726],[19,-5.0149],[20,-5.2887],[21,-5.5423],[22,-5.7707],[23,-5.9689],[24,-6.1318],[25,-6.2544],[26,-6.3317],[27,-6.3586],[28,-6.3291],[29,-6.2446],[30,-6.1109],[31,-5.9337],[32,-5.7189],[33,-5.4722],[34,-5.1994],[35,-4.9064],[37,-4.2826],[39,-3.6473],[40,-3.3397],[41,-3.0467],[42,-2.7739],[43,-2.5272]]
       }
       index_03_l_ctrl.Scale.X {type=MovieSceneScriptingFloatChannel, default=1, const [1,44]}
       index_03_l_ctrl.Scale.Y {type=MovieSceneScriptingFloatChannel, default=1, const [1,44]}
       index_03_l_ctrl.Scale.Z {type=MovieSceneScriptingFloatChannel, default=1, const [1,44]}
       middle_metacarpal_l_ctrl.Location.X {type=MovieSceneScriptingFloatChannel, default=0, const [0,44]}
       middle_metacarpal_l_ctrl.Location.Y {type=MovieSceneScriptingFloatChannel, default=0, const [0,44]}
       middle_metacarpal_l_ctrl.Location.Z {type=MovieSceneScriptingFloatChannel, default=0, const [0,44]}
       middle_metacarpal_l_ctrl.Rotation.X {type=MovieSceneScriptingFloatChannel, default=0, const [0,44]}
       middle_metacarpal_l_ctrl.Rotation.Y {type=MovieSceneScriptingFloatChannel, default=0, const [0,44]}
       middle_metacarpal_l_ctrl.Rotation.Z {type=MovieSceneScriptingFloatChannel, default=0, const [0,44]}
       middle_metacarpal_l_ctrl.Scale.X {type=MovieSceneScriptingFloatChannel, default=1, const [1,44]}
       middle_metacarpal_l_ctrl.Scale.Y {type=MovieSceneScriptingFloatChannel, default=1, const [1,44]}
       middle_metacarpal_l_ctrl.Scale.Z {type=MovieSceneScriptingFloatChannel, default=1, const [1,44]}
       middle_01_l_ctrl.Location.X {type=MovieSceneScriptingFloatChannel, default=0, thinned_from=44, keys [[0,0],[43,0]]}
       middle_01_l_ctrl.Location.Y {type=MovieSceneScriptingFloatChannel, default=0, const [0,44]}
       middle_01_l_ctrl.Location.Z {type=MovieSceneScriptingFloatChannel, default=0, const [0,44]}
       middle_01_l_ctrl.Rotation.X {
        type=MovieSceneScriptingFloatChannel
        default=4.8669
        thinned_from=44
        keys [[0,4.8669],[2,4.8559],[5,4.8047],[16,4.4802],[19,4.4178],[21,4.3966],[23,4.3969],[25,4.42],[28,4.4876],[37,4.7731],[40,4.8407],[42,4.8638],[43,4.8669]]
       }
       middle_01_l_ctrl.Rotation.Y {
        type=MovieSceneScriptingFloatChannel
        default=3.2659
        thinned_from=44
        keys [[0,3.2659],[1,3.2693],[3,3.2956],[5,3.3431],[8,3.4425],[15,3.7132],[18,3.8027],[20,3.8403],[22,3.854],[24,3.839],[26,3.7981],[29,3.7015],[36,3.4183],[38,3.35],[40,3.2984],[42,3.2697],[43,3.2658]]
       }
       middle_01_l_ctrl.Rotation.Z {
        type=MovieSceneScriptingFloatChannel
        default=8.6317
        thinned_from=44
        keys [[0,8.6317],[1,8.5871],[2,8.459],[3,8.2555],[4,7.9852],[5,7.6564],[6,7.2775],[7,6.8567],[8,6.4025],[9,5.9232],[13,3.9225],[14,3.4432],[15,2.989],[16,2.5682],[17,2.1892],[18,1.8604],[19,1.5901],[20,1.3867],[21,1.2585],[22,1.2139],[23,1.2628],[24,1.403],[25,1.6248],[26,1.9188],[27,2.2752],[28,2.6845],[29,3.1371],[30,3.6233],[31,4.1335],[34,5.7122],[35,6.2224],[36,6.7086],[37,7.1612],[38,7.5704],[39,7.9269],[40,8.2208],[41,8.4427],[42,8.5829],[43,8.6317]]
       }
       middle_01_l_ctrl.Scale.X {type=MovieSceneScriptingFloatChannel, default=1, const [1,44]}
       middle_01_l_ctrl.Scale.Y {type=MovieSceneScriptingFloatChannel, default=1, const [1,44]}
       middle_01_l_ctrl.Scale.Z {type=MovieSceneScriptingFloatChannel, default=1, const [1,44]}
       middle_02_l_ctrl.Location.X {type=MovieSceneScriptingFloatChannel, default=-0.0452, thinned_from=44, keys [[0,-0.0452],[43,-0.0452]]}
       middle_02_l_ctrl.Location.Y {type=MovieSceneScriptingFloatChannel, default=0.0507, thinned_from=44, keys [[0,0.0507],[43,0.0507]]}
       middle_02_l_ctrl.Location.Z {type=MovieSceneScriptingFloatChannel, default=0.1146, thinned_from=44, keys [[0,0.1146],[43,0.1146]]}
       middle_02_l_ctrl.Rotation.X {type=MovieSceneScriptingFloatChannel, default=0, thinned_from=44, keys [[0,0],[43,0]]}
       middle_02_l_ctrl.Rotation.Y {type=MovieSceneScriptingFloatChannel, default=0, thinned_from=44, keys [[0,0],[43,0]]}
       middle_02_l_ctrl.Rotation.Z {
        type=MovieSceneScriptingFloatChannel
        default=4.1758
        thinned_from=44
        keys [[0,4.1758],[1,4.2054],[2,4.1784],[3,4.1008],[4,3.9775],[5,3.8138],[6,3.6145],[7,3.3849],[8,3.13],[9,2.8548],[10,2.5644],[11,2.2639],[13,1.6527],[14,1.3522],[15,1.0618],[16,0.7865],[17,0.5316],[18,0.302],[19,0.1027],[20,-0.061],[21,-0.1843],[22,-0.2619],[23,-0.2889],[24,-0.2593],[25,-0.1744],[26,-0.04],[27,0.1381],[28,0.3541],[29,0.6021],[30,0.8763],[31,1.1709],[32,1.48],[35,2.4366],[36,2.7457],[37,3.0403],[38,3.3145],[39,3.5624],[40,3.7784],[41,3.9565],[42,4.0909],[43,4.1758]]
       }
       middle_02_l_ctrl.Scale.X {type=MovieSceneScriptingFloatChannel, default=1, const [1,44]}
       middle_02_l_ctrl.Scale.Y {type=MovieSceneScriptingFloatChannel, default=1, const [1,44]}
       middle_02_l_ctrl.Scale.Z {type=MovieSceneScriptingFloatChannel, default=1, const [1,44]}
       middle_03_l_ctrl.Location.X {type=MovieSceneScriptingFloatChannel, default=0.0206, thinned_from=44, keys [[0,0.0206],[43,0.0206]]}
       middle_03_l_ctrl.Location.Y {type=MovieSceneScriptingFloatChannel, default=-0.0583, thinned_from=44, keys [[0,-0.0583],[43,-0.0583]]}
       middle_03_l_ctrl.Location.Z {type=MovieSceneScriptingFloatChannel, default=0.1324, const [0.1324,44]}
       middle_03_l_ctrl.Rotation.X {type=MovieSceneScriptingFloatChannel, default=0, thinned_from=44, keys [[0,0],[43,0]]}
       middle_03_l_ctrl.Rotation.Y {type=MovieSceneScriptingFloatChannel, default=0, thinned_from=44, keys [[0,0],[43,0]]}
       middle_03_l_ctrl.Rotation.Z {
        type=MovieSceneScriptingFloatChannel
        default=-0.1142
        thinned_from=44
        keys [[0,-0.1142],[1,-0.0295],[2,0],[3,-0.0269],[4,-0.1044],[5,-0.2273],[6,-0.3906],[7,-0.5893],[8,-0.8183],[9,-1.0725],[10,-1.347],[11,-1.6366],[12,-1.9363],[14,-2.5458],[15,-2.8455],[16,-3.1351],[17,-3.4096],[18,-3.6638],[19,-3.8928],[20,-4.0915],[21,-4.2548],[22,-4.3778],[23,-4.4552],[24,-4.4822],[25,-4.4526],[26,-4.3679],[27,-4.2339],[28,-4.0562],[29,-3.8409],[30,-3.5935],[31,-3.3201],[32,-3.0263],[33,-2.718],[36,-1.7641],[37,-1.4558],[38,-1.162],[39,-0.8886],[40,-0.6413],[41,-0.4259],[42,-0.2483],[43,-0.1142]]
       }
       middle_03_l_ctrl.Scale.X {type=MovieSceneScriptingFloatChannel, default=1, const [1,44]}
       middle_03_l_ctrl.Scale.Y {type=MovieSceneScriptingFloatChannel, default=1, const [1,44]}
       middle_03_l_ctrl.Scale.Z {type=MovieSceneScriptingFloatChannel, default=1, const [1,44]}
       ring_metacarpal_l_ctrl.Location.X {type=MovieSceneScriptingFloatChannel, default=0, const [0,44]}
       ring_metacarpal_l_ctrl.Location.Y {type=MovieSceneScriptingFloatChannel, default=0, const [0,44]}
       ring_metacarpal_l_ctrl.Location.Z {type=MovieSceneScriptingFloatChannel, default=0, const [0,44]}
       ring_metacarpal_l_ctrl.Rotation.X {type=MovieSceneScriptingFloatChannel, default=0, const [0,44]}
       ring_metacarpal_l_ctrl.Rotation.Y {type=MovieSceneScriptingFloatChannel, default=0, const [0,44]}
       ring_metacarpal_l_ctrl.Rotation.Z {type=MovieSceneScriptingFloatChannel, default=0, const [0,44]}
       ring_metacarpal_l_ctrl.Scale.X {type=MovieSceneScriptingFloatChannel, default=1, const [1,44]}
       ring_metacarpal_l_ctrl.Scale.Y {type=MovieSceneScriptingFloatChannel, default=1, const [1,44]}
       ring_metacarpal_l_ctrl.Scale.Z {type=MovieSceneScriptingFloatChannel, default=1, const [1,44]}
       ring_01_l_ctrl.Location.X {type=MovieSceneScriptingFloatChannel, default=0, const [0,44]}
       ring_01_l_ctrl.Location.Y {type=MovieSceneScriptingFloatChannel, default=0, const [0,44]}
       ring_01_l_ctrl.Location.Z {type=MovieSceneScriptingFloatChannel, default=0, const [0,44]}
       ring_01_l_ctrl.Rotation.X {
        type=MovieSceneScriptingFloatChannel
        default=5.7554
        thinned_from=44
        keys [[0,5.7554],[4,5.7707],[19,6.0679],[23,6.0977],[27,6.0652],[40,5.7861],[43,5.7554]]
       }
       ring_01_l_ctrl.Rotation.Y {
        type=MovieSceneScriptingFloatChannel
        default=-3.1751
        thinned_from=44
        keys [[0,-3.1751],[1,-3.1802],[2,-3.1757],[3,-3.1625],[4,-3.1417],[6,-3.0802],[8,-2.9978],[15,-2.6436],[17,-2.5521],[18,-2.5124],[19,-2.4779],[20,-2.4496],[21,-2.4282],[22,-2.4147],[23,-2.41],[24,-2.4152],[25,-2.4299],[26,-2.4532],[27,-2.4841],[28,-2.5214],[30,-2.6116],[37,-2.9825],[39,-3.0712],[40,-3.1079],[41,-3.1381],[42,-3.1608],[43,-3.1752]]
       }
       ring_01_l_ctrl.Rotation.Z {
        type=MovieSceneScriptingFloatChannel
        default=8.3983
        thinned_from=44
        keys [[0,8.3983],[1,8.447],[2,8.4025],[3,8.2745],[4,8.0713],[5,7.8013],[6,7.4728],[7,7.0943],[8,6.674],[9,6.2204],[10,5.7417],[14,3.7437],[15,3.2652],[16,2.8116],[17,2.3915],[18,2.0132],[19,1.6849],[20,1.415],[21,1.2119],[22,1.084],[23,1.0395],[24,1.0882],[25,1.2282],[26,1.4497],[27,1.7432],[28,2.099],[29,2.5077],[30,2.9595],[31,3.445],[32,3.9545],[35,5.531],[36,6.0406],[37,6.5261],[38,6.9781],[39,7.3869],[40,7.743],[41,8.0366],[42,8.2582],[43,8.3982]]
       }
       ring_01_l_ctrl.Scale.X {type=MovieSceneScriptingFloatChannel, default=1, const [1,44]}
       ring_01_l_ctrl.Scale.Y {type=MovieSceneScriptingFloatChannel, default=1, const [1,44]}
       ring_01_l_ctrl.Scale.Z {type=MovieSceneScriptingFloatChannel, default=1, const [1,44]}
       ring_02_l_ctrl.Location.X {type=MovieSceneScriptingFloatChannel, default=0, thinned_from=44, keys [[0,0],[43,0]]}
       ring_02_l_ctrl.Location.Y {type=MovieSceneScriptingFloatChannel, default=0, thinned_from=44, keys [[0,0],[43,0]]}
       ring_02_l_ctrl.Location.Z {type=MovieSceneScriptingFloatChannel, default=0, thinned_from=44, keys [[0,0],[43,0]]}
       ring_02_l_ctrl.Rotation.X {type=MovieSceneScriptingFloatChannel, default=0, thinned_from=44, keys [[0,0],[43,0]]}
       ring_02_l_ctrl.Rotation.Y {type=MovieSceneScriptingFloatChannel, default=0, thinned_from=44, keys [[0,0],[43,0]]}
       ring_02_l_ctrl.Rotation.Z {
        type=MovieSceneScriptingFloatChannel
        default=-3.7544
        thinned_from=44
        keys [[0,-3.7544],[1,-3.7814],[2,-3.8588],[3,-3.9817],[4,-4.1451],[5,-4.3437],[6,-4.5727],[7,-4.827],[8,-5.1014],[10,-5.6907],[13,-6.6],[14,-6.8896],[15,-7.164],[16,-7.4183],[17,-7.6473],[18,-7.846],[19,-8.0093],[20,-8.1322],[21,-8.2097],[22,-8.2366],[23,-8.2071],[24,-8.1224],[25,-7.9883],[26,-7.8107],[27,-7.5953],[28,-7.348],[29,-7.0746],[30,-6.7808],[34,-5.5185],[35,-5.2102],[36,-4.9165],[37,-4.643],[38,-4.3957],[39,-4.1803],[40,-4.0027],[41,-3.8687],[42,-3.784],[43,-3.7544]]
       }
       ring_02_l_ctrl.Scale.X {type=MovieSceneScriptingFloatChannel, default=1, const [1,44]}
       ring_02_l_ctrl.Scale.Y {type=MovieSceneScriptingFloatChannel, default=1, const [1,44]}
       ring_02_l_ctrl.Scale.Z {type=MovieSceneScriptingFloatChannel, default=1, const [1,44]}
       ring_03_l_ctrl.Location.X {type=MovieSceneScriptingFloatChannel, default=0, thinned_from=44, keys [[0,0],[43,0]]}
       ring_03_l_ctrl.Location.Y {type=MovieSceneScriptingFloatChannel, default=0, thinned_from=44, keys [[0,0],[43,0]]}
       ring_03_l_ctrl.Location.Z {type=MovieSceneScriptingFloatChannel, default=0, thinned_from=44, keys [[0,0],[43,0]]}
       ring_03_l_ctrl.Rotation.X {type=MovieSceneScriptingFloatChannel, default=0, thinned_from=44, keys [[0,0],[43,0]]}
       ring_03_l_ctrl.Rotation.Y {type=MovieSceneScriptingFloatChannel, default=0, const [0,44]}
       ring_03_l_ctrl.Rotation.Z {
        type=MovieSceneScriptingFloatChannel
        default=-3.7814
        thinned_from=44
        keys [[0,-3.7814],[1,-3.8589],[2,-3.9818],[3,-4.1451],[4,-4.3438],[5,-4.5728],[6,-4.827],[7,-5.1015],[9,-5.6908],[12,-6.6],[13,-6.8896],[14,-7.1641],[15,-7.4183],[16,-7.6473],[17,-7.846],[18,-8.0093],[19,-8.1322],[20,-8.2097],[21,-8.2366],[22,-8.2071],[23,-8.1224],[24,-7.9883],[25,-7.8107],[26,-7.5953],[27,-7.348],[28,-7.0746],[29,-6.7808],[33,-5.5186],[34,-5.2103],[35,-4.9165],[36,-4.6431],[37,-4.3957],[38,-4.1804],[39,-4.0028],[40,-3.8687],[41,-3.784],[42,-3.7545],[43,-3.7814]]
       }
       ring_03_l_ctrl.Scale.X {type=MovieSceneScriptingFloatChannel, default=1, const [1,44]}
       ring_03_l_ctrl.Scale.Y {type=MovieSceneScriptingFloatChannel, default=1, const [1,44]}
       ring_03_l_ctrl.Scale.Z {type=MovieSceneScriptingFloatChannel, default=1, const [1,44]}
       pinky_metacarpal_l_ctrl.Location.X {type=MovieSceneScriptingFloatChannel, default=0, const [0,44]}
       pinky_metacarpal_l_ctrl.Location.Y {type=MovieSceneScriptingFloatChannel, default=0, const [0,44]}
       pinky_metacarpal_l_ctrl.Location.Z {type=MovieSceneScriptingFloatChannel, default=0, const [0,44]}
       pinky_metacarpal_l_ctrl.Rotation.X {type=MovieSceneScriptingFloatChannel, default=0, const [0,44]}
       pinky_metacarpal_l_ctrl.Rotation.Y {type=MovieSceneScriptingFloatChannel, default=0, const [0,44]}
       pinky_metacarpal_l_ctrl.Rotation.Z {type=MovieSceneScriptingFloatChannel, default=0, const [0,44]}
       pinky_metacarpal_l_ctrl.Scale.X {type=MovieSceneScriptingFloatChannel, default=1, const [1,44]}
       pinky_metacarpal_l_ctrl.Scale.Y {type=MovieSceneScriptingFloatChannel, default=1, const [1,44]}
       pinky_metacarpal_l_ctrl.Scale.Z {type=MovieSceneScriptingFloatChannel, default=1, const [1,44]}
       pinky_01_l_ctrl.Location.X {type=MovieSceneScriptingFloatChannel, default=0, const [0,44]}
       pinky_01_l_ctrl.Location.Y {type=MovieSceneScriptingFloatChannel, default=0, const [0,44]}
       pinky_01_l_ctrl.Location.Z {type=MovieSceneScriptingFloatChannel, default=0, const [0,44]}
       pinky_01_l_ctrl.Rotation.X {type=MovieSceneScriptingFloatChannel, default=0, thinned_from=44, keys [[0,0],[43,0]]}
       pinky_01_l_ctrl.Rotation.Y {type=MovieSceneScriptingFloatChannel, default=0, thinned_from=44, keys [[0,0],[43,-0.0001]]}
       pinky_01_l_ctrl.Rotation.Z {
        type=MovieSceneScriptingFloatChannel
        default=-4.4906
        thinned_from=44
        keys [[0,-4.4906],[1,-4.4419],[2,-4.4864],[3,-4.6144],[4,-4.8176],[5,-5.0875],[6,-5.4158],[7,-5.7943],[8,-6.2145],[9,-6.6681],[10,-7.1467],[14,-9.1446],[15,-9.6232],[16,-10.0768],[17,-10.4969],[18,-10.8754],[19,-11.2037],[20,-11.4736],[21,-11.6768],[22,-11.8048],[23,-11.8493],[24,-11.8005],[25,-11.6605],[26,-11.439],[27,-11.1454],[28,-10.7895],[29,-10.3808],[30,-9.9289],[31,-9.4434],[35,-7.3574],[36,-6.8478],[37,-6.3623],[38,-5.9104],[39,-5.5017],[40,-5.1457],[41,-4.8522],[42,-4.6306],[43,-4.4907]]
       }
       pinky_01_l_ctrl.Scale.X {type=MovieSceneScriptingFloatChannel, default=1, const [1,44]}
       pinky_01_l_ctrl.Scale.Y {type=MovieSceneScriptingFloatChannel, default=1, const [1,44]}
       pinky_01_l_ctrl.Scale.Z {type=MovieSceneScriptingFloatChannel, default=1, const [1,44]}
       pinky_02_l_ctrl.Location.X {type=MovieSceneScriptingFloatChannel, default=0, thinned_from=44, keys [[0,0],[43,0]]}
       pinky_02_l_ctrl.Location.Y {type=MovieSceneScriptingFloatChannel, default=0, thinned_from=44, keys [[0,0],[43,0]]}
       pinky_02_l_ctrl.Location.Z {type=MovieSceneScriptingFloatChannel, default=0, thinned_from=44, keys [[0,0],[43,0]]}
       pinky_02_l_ctrl.Rotation.X {type=MovieSceneScriptingFloatChannel, default=0, thinned_from=44, keys [[0,0],[43,0]]}
       pinky_02_l_ctrl.Rotation.Y {type=MovieSceneScriptingFloatChannel, default=0, thinned_from=44, keys [[0,0],[43,0]]}
       pinky_02_l_ctrl.Rotation.Z {
        type=MovieSceneScriptingFloatChannel
        default=-6.1119
        thinned_from=44
        keys [[0,-6.1119],[1,-6.1388],[2,-6.2163],[3,-6.3392],[4,-6.5025],[5,-6.7012],[6,-6.9302],[7,-7.1844],[8,-7.4589],[14,-9.2471],[15,-9.5216],[16,-9.7758],[17,-10.0048],[18,-10.2035],[19,-10.3668],[20,-10.4897],[21,-10.5671],[22,-10.5941],[23,-10.5646],[24,-10.4799],[25,-10.3458],[26,-10.1682],[27,-9.9528],[28,-9.7055],[29,-9.4321],[31,-8.83],[35,-7.5677],[36,-7.2739],[37,-7.0005],[38,-6.7531],[39,-6.5378],[40,-6.3602],[41,-6.2261],[42,-6.1414],[43,-6.1119]]
       }
       pinky_02_l_ctrl.Scale.X {type=MovieSceneScriptingFloatChannel, default=1, const [1,44]}
       pinky_02_l_ctrl.Scale.Y {type=MovieSceneScriptingFloatChannel, default=1, const [1,44]}
       pinky_02_l_ctrl.Scale.Z {type=MovieSceneScriptingFloatChannel, default=1, const [1,44]}
       pinky_03_l_ctrl.Location.X {type=MovieSceneScriptingFloatChannel, default=0, thinned_from=44, keys [[0,0],[43,0]]}
       pinky_03_l_ctrl.Location.Y {type=MovieSceneScriptingFloatChannel, default=0, thinned_from=44, keys [[0,0],[43,0]]}
       pinky_03_l_ctrl.Location.Z {type=MovieSceneScriptingFloatChannel, default=0, thinned_from=44, keys [[0,0],[43,0]]}
       pinky_03_l_ctrl.Rotation.X {type=MovieSceneScriptingFloatChannel, default=0, thinned_from=44, keys [[0,0],[43,0]]}
       pinky_03_l_ctrl.Rotation.Y {type=MovieSceneScriptingFloatChannel, default=0, thinned_from=44, keys [[0,0],[43,0]]}
       pinky_03_l_ctrl.Rotation.Z {
        type=MovieSceneScriptingFloatChannel
        default=-6.1388
        thinned_from=44
        keys [[0,-6.1388],[1,-6.2163],[2,-6.3392],[3,-6.5025],[4,-6.7012],[5,-6.9302],[6,-7.1844],[7,-7.4589],[13,-9.247],[14,-9.5215],[15,-9.7757],[16,-10.0047],[17,-10.2034],[18,-10.3667],[19,-10.4897],[20,-10.5671],[21,-10.5941],[22,-10.5645],[23,-10.4798],[24,-10.3458],[25,-10.1681],[26,-9.9528],[27,-9.7055],[28,-9.432],[30,-8.8299],[34,-7.5677],[35,-7.2739],[36,-7.0005],[37,-6.7531],[38,-6.5378],[39,-6.3601],[40,-6.2261],[41,-6.1414],[42,-6.1119],[43,-6.1388]]
       }
       pinky_03_l_ctrl.Scale.X {type=MovieSceneScriptingFloatChannel, default=1, const [1,44]}
       pinky_03_l_ctrl.Scale.Y {type=MovieSceneScriptingFloatChannel, default=1, const [1,44]}
       pinky_03_l_ctrl.Scale.Z {type=MovieSceneScriptingFloatChannel, default=1, const [1,44]}
       hand_l_ik_ctrl.Location.X {
        type=MovieSceneScriptingFloatChannel
        default=-12.3899
        thinned_from=44
        keys [[0,-12.3899],[7,-12.1812],[10,-12.1844],[13,-12.2934],[24,-13.0942],[28,-13.2898],[31,-13.3402],[33,-13.305],[35,-13.1952],[38,-12.9058],[41,-12.5661],[43,-12.3899]]
       }
       hand_l_ik_ctrl.Location.Y {
        type=MovieSceneScriptingFloatChannel
        default=23.4205
        thinned_from=44
        keys [[0,23.4205],[8,23.8359],[19,24.0415],[36,23.3788],[43,23.4205]]
       }
       hand_l_ik_ctrl.Location.Z {type=MovieSceneScriptingFloatChannel, default=51.4599, thinned_from=44, keys [[0,51.4599],[20,50.6932],[34,51.6615],[43,51.4599]]}
       hand_l_ik_ctrl.Rotation.X {type=MovieSceneScriptingFloatChannel, default=110.6746, thinned_from=44, keys [[0,110.6746],[22,111.3353],[36,110.154],[43,110.6746]]}
       hand_l_ik_ctrl.Rotation.Y {
        type=MovieSceneScriptingFloatChannel
        default=36.7189
        thinned_from=44
        keys [[0,36.7189],[6,35.2353],[8,34.8734],[10,34.6913],[15,34.9545],[19,35.4883],[22,36.0855],[28,37.8458],[30,38.3004],[32,38.5341],[34,38.4797],[37,38.104],[41,37.2478],[43,36.7189]]
       }
       hand_l_ik_ctrl.Rotation.Z {
        type=MovieSceneScriptingFloatChannel
        default=95.8693
        thinned_from=44
        keys [[0,95.8693],[10,97.9048],[20,97.9631],[32,95.2407],[39,95.2703],[43,95.8693]]
       }
       hand_l_ik_ctrl.Scale.X {type=MovieSceneScriptingFloatChannel, default=1, const [1,44]}
       hand_l_ik_ctrl.Scale.Y {type=MovieSceneScriptingFloatChannel, default=1, const [1,44]}
       hand_l_ik_ctrl.Scale.Z {type=MovieSceneScriptingFloatChannel, default=1, const [1,44]}
       Softness {type=MovieSceneScriptingFloatChannel, default=0, const [0,44]}
       hand_l_ikorient_ctrl.Rotation.Roll {type=MovieSceneScriptingFloatChannel, default=0, const [0,44]}
       hand_l_ikorient_ctrl.Rotation.Pitch {type=MovieSceneScriptingFloatChannel, default=0, const [0,44]}
       hand_l_ikorient_ctrl.Rotation.Yaw {type=MovieSceneScriptingFloatChannel, default=0, const [0,44]}
       arm_l_pv_ik_ctrl.Location.X {
        type=MovieSceneScriptingFloatChannel
        default=10.5411
        thinned_from=44
        keys [[0,10.5411],[2,10.4379],[5,10.3643],[8,10.3691],[15,10.5679],[21,10.7994],[25,10.8322],[42,10.568],[43,10.5411]]
       }
       arm_l_pv_ik_ctrl.Location.Y {type=MovieSceneScriptingFloatChannel, default=63.5104, thinned_from=44, keys [[0,63.5104],[43,63.5104]]}
       arm_l_pv_ik_ctrl.Location.Z {
        type=MovieSceneScriptingFloatChannel
        default=13.7711
        thinned_from=44
        keys [[0,13.7711],[4,13.9424],[9,14.0164],[12,13.9809],[21,13.7148],[28,13.2764],[31,13.1707],[33,13.1593],[36,13.2852],[43,13.7711]]
       }
       arm_l_pv_ik_ctrl.Rotation.X {
        type=MovieSceneScriptingFloatChannel
        default=2.6204
        thinned_from=44
        keys [[0,2.6204],[1,2.8082],[2,2.9856],[3,3.1528],[4,3.3096],[5,3.4553],[6,3.5901],[7,3.7135],[9,3.9343],[11,4.1088],[12,4.1854],[13,4.2485],[15,4.3411],[16,4.3569],[17,4.3521],[18,4.322],[19,4.274],[20,4.1734],[21,4.0628],[22,3.8784],[23,3.6463],[24,3.3952],[25,3.1196],[27,2.5348],[28,2.2463],[29,1.9745],[30,1.7301],[31,1.5233],[32,1.364],[33,1.2596],[34,1.2295],[35,1.2656],[36,1.3573],[37,1.4937],[38,1.6633],[39,1.8554],[40,2.054],[41,2.2387],[42,2.4482],[43,2.6204]]
       }
       arm_l_pv_ik_ctrl.Rotation.Y {type=MovieSceneScriptingFloatChannel, default=37.788, thinned_from=44, keys [[0,37.788],[19,36.8625],[41,37.8056],[43,37.788]]}
       arm_l_pv_ik_ctrl.Rotation.Z {
        type=MovieSceneScriptingFloatChannel
        default=75.8312
        thinned_from=44
        keys [[0,75.8312],[6,77.1388],[15,77.6335],[21,77.0133],[31,74.4611],[34,74.2094],[38,74.6904],[43,75.8312]]
       }
       arm_l_pv_ik_ctrl.Scale.X {type=MovieSceneScriptingFloatChannel, default=1, const [1,44]}
       arm_l_pv_ik_ctrl.Scale.Y {type=MovieSceneScriptingFloatChannel, default=1, const [1,44]}
       arm_l_pv_ik_ctrl.Scale.Z {type=MovieSceneScriptingFloatChannel, default=1, const [1,44]}
       upperarm_r_fk_ctrl.Location.X {type=MovieSceneScriptingFloatChannel, default=0, thinned_from=44, keys [[0,0],[43,0]]}
       upperarm_r_fk_ctrl.Location.Y {type=MovieSceneScriptingFloatChannel, default=0.0001, thinned_from=44, keys [[0,0.0001],[43,0.0001]]}
       upperarm_r_fk_ctrl.Location.Z {type=MovieSceneScriptingFloatChannel, default=-0.0001, thinned_from=44, keys [[0,-0.0001],[43,-0.0001]]}
       upperarm_r_fk_ctrl.Rotation.X {
        type=MovieSceneScriptingFloatChannel
        default=-42.1399
        thinned_from=44
        keys [[0,-42.1399],[6,-42.9577],[9,-43.1019],[11,-43.0278],[13,-42.7736],[15,-42.2682],[21,-40.3108],[24,-39.7909],[27,-39.5706],[30,-39.6376],[33,-39.9599],[42,-41.9771],[43,-42.1399]]
       }
       upperarm_r_fk_ctrl.Rotation.Y {
        type=MovieSceneScriptingFloatChannel
        default=11.3375
        thinned_from=44
        keys [[0,11.3375],[1,11.1913],[3,10.8347],[7,10.0396],[9,9.6924],[10,9.5469],[11,9.4259],[12,9.3421],[13,9.2784],[14,9.2366],[16,9.2475],[17,9.2918],[18,9.3565],[21,9.6426],[23,9.9081],[30,11.046],[32,11.3112],[34,11.5033],[36,11.6223],[37,11.6521],[39,11.6446],[41,11.5403],[43,11.3375]]
       }
       upperarm_r_fk_ctrl.Rotation.Z {
        type=MovieSceneScriptingFloatChannel
        default=-52.0204
        thinned_from=44
        keys [[0,-52.0204],[7,-52.3052],[26,-50.7931],[33,-51.0339],[43,-52.0204]]
       }
       upperarm_r_fk_ctrl.Scale.X {type=MovieSceneScriptingFloatChannel, default=1, const [1,44]}
       upperarm_r_fk_ctrl.Scale.Y {type=MovieSceneScriptingFloatChannel, default=1, const [1,44]}
       upperarm_r_fk_ctrl.Scale.Z {type=MovieSceneScriptingFloatChannel, default=1, const [1,44]}
       lowerarm_r_fk_ctrl.Location.X {type=MovieSceneScriptingFloatChannel, default=0, thinned_from=44, keys [[0,0],[43,0]]}
       lowerarm_r_fk_ctrl.Location.Y {type=MovieSceneScriptingFloatChannel, default=0, thinned_from=44, keys [[0,0],[43,0]]}
       lowerarm_r_fk_ctrl.Location.Z {type=MovieSceneScriptingFloatChannel, default=0, thinned_from=44, keys [[0,0],[43,0]]}
       lowerarm_r_fk_ctrl.Rotation.X {type=MovieSceneScriptingFloatChannel, default=-3.3918, thinned_from=44, keys [[0,-3.3918],[13,-3.3546],[40,-3.3992],[43,-3.3918]]}
       lowerarm_r_fk_ctrl.Rotation.Y {
        type=MovieSceneScriptingFloatChannel
        default=1.742
        thinned_from=44
        keys [[0,1.742],[9,1.6893],[12,1.6825],[17,1.6946],[30,1.7514],[38,1.7569],[43,1.742]]
       }
       lowerarm_r_fk_ctrl.Rotation.Z {
        type=MovieSceneScriptingFloatChannel
        default=42.4065
        thinned_from=44
        keys [[0,42.4065],[7,45.0146],[9,45.5973],[11,45.9419],[13,45.9755],[15,45.7306],[28,42.2236],[31,41.6764],[34,41.4234],[37,41.4225],[39,41.59],[42,42.1526],[43,42.4065]]
       }
       lowerarm_r_fk_ctrl.Scale.X {type=MovieSceneScriptingFloatChannel, default=1, const [1,44]}
       lowerarm_r_fk_ctrl.Scale.Y {type=MovieSceneScriptingFloatChannel, default=1, const [1,44]}
       lowerarm_r_fk_ctrl.Scale.Z {type=MovieSceneScriptingFloatChannel, default=1, const [1,44]}
       hand_r_fk_ctrl.Location.X {type=MovieSceneScriptingFloatChannel, default=0, thinned_from=44, keys [[0,0],[43,0]]}
       hand_r_fk_ctrl.Location.Y {type=MovieSceneScriptingFloatChannel, default=0, thinned_from=44, keys [[0,0],[43,0]]}
       hand_r_fk_ctrl.Location.Z {type=MovieSceneScriptingFloatChannel, default=0, thinned_from=44, keys [[0,0],[43,0]]}
       hand_r_fk_ctrl.Rotation.X {
        type=MovieSceneScriptingFloatChannel
        default=-2.6495
        thinned_from=44
        keys [[0,-2.6495],[1,-2.504],[2,-2.3481],[4,-2.0262],[5,-1.8714],[6,-1.7272],[7,-1.5986],[8,-1.4905],[9,-1.4077],[10,-1.3551],[11,-1.3377],[12,-1.3607],[13,-1.4307],[14,-1.5455],[15,-1.6975],[16,-1.8792],[17,-2.0828],[18,-2.3003],[20,-2.7438],[21,-2.9492],[22,-3.129],[23,-3.284],[24,-3.422],[25,-3.543],[26,-3.6448],[27,-3.7275],[28,-3.7906],[29,-3.8336],[30,-3.8562],[31,-3.8581],[32,-3.8391],[33,-3.7992],[34,-3.7386],[35,-3.6594],[36,-3.5637],[37,-3.4537],[38,-3.3315],[39,-3.1994],[42,-2.7757],[43,-2.6495]]
       }
       hand_r_fk_ctrl.Rotation.Y {
        type=MovieSceneScriptingFloatChannel
        default=1.6536
        thinned_from=44
        keys [[0,1.6536],[1,1.7616],[2,1.8929],[3,2.0383],[5,2.3537],[6,2.5072],[7,2.6488],[8,2.7721],[9,2.8714],[10,2.9408],[11,2.9741],[12,2.9645],[13,2.9286],[14,2.8863],[15,2.8349],[16,2.7719],[17,2.6948],[18,2.6016],[19,2.4905],[20,2.3588],[21,2.2033],[22,2.0205],[23,1.8167],[25,1.3941],[26,1.1879],[27,0.9956],[28,0.8243],[29,0.6812],[30,0.5735],[31,0.508],[32,0.4913],[33,0.5297],[34,0.6098],[35,0.7104],[36,0.8262],[37,0.9518],[39,1.2114],[40,1.3341],[41,1.4479],[42,1.554],[43,1.6536]]
       }
       hand_r_fk_ctrl.Rotation.Z {
        type=MovieSceneScriptingFloatChannel
        default=-26.084
        thinned_from=44
        keys [[0,-26.084],[2,-25.5398],[7,-23.948],[9,-23.4415],[10,-23.2612],[11,-23.1446],[12,-23.1028],[13,-23.1485],[14,-23.2781],[15,-23.4775],[17,-24.0298],[21,-25.3515],[26,-26.7001],[29,-27.2981],[31,-27.5413],[33,-27.6174],[35,-27.513],[38,-27.1052],[43,-26.084]]
       }
       hand_r_fk_ctrl.Scale.X {type=MovieSceneScriptingFloatChannel, default=1, const [1,44]}
       hand_r_fk_ctrl.Scale.Y {type=MovieSceneScriptingFloatChannel, default=1, const [1,44]}
       hand_r_fk_ctrl.Scale.Z {type=MovieSceneScriptingFloatChannel, default=1, const [1,44]}
       thumb_01_r_ctrl.Location.X {type=MovieSceneScriptingFloatChannel, default=0, thinned_from=44, keys [[0,0],[43,0]]}
       thumb_01_r_ctrl.Location.Y {type=MovieSceneScriptingFloatChannel, default=0, const [0,44]}
       thumb_01_r_ctrl.Location.Z {type=MovieSceneScriptingFloatChannel, default=0, thinned_from=44, keys [[0,0],[43,0]]}
       thumb_01_r_ctrl.Rotation.X {type=MovieSceneScriptingFloatChannel, default=15.1666, const [15.1666,44]}
       thumb_01_r_ctrl.Rotation.Y {type=MovieSceneScriptingFloatChannel, default=-25.3344, const [-25.3344,44]}
       thumb_01_r_ctrl.Rotation.Z {type=MovieSceneScriptingFloatChannel, default=-2.159, const [-2.159,44]}
       thumb_01_r_ctrl.Scale.X {type=MovieSceneScriptingFloatChannel, default=1, const [1,44]}
       thumb_01_r_ctrl.Scale.Y {type=MovieSceneScriptingFloatChannel, default=1, const [1,44]}
       thumb_01_r_ctrl.Scale.Z {type=MovieSceneScriptingFloatChannel, default=1, const [1,44]}
       thumb_02_r_ctrl.Location.X {type=MovieSceneScriptingFloatChannel, default=0, const [0,44]}
       thumb_02_r_ctrl.Location.Y {type=MovieSceneScriptingFloatChannel, default=0, const [0,44]}
       thumb_02_r_ctrl.Location.Z {type=MovieSceneScriptingFloatChannel, default=0, const [0,44]}
       thumb_02_r_ctrl.Rotation.X {type=MovieSceneScriptingFloatChannel, default=0, const [0,44]}
       thumb_02_r_ctrl.Rotation.Y {type=MovieSceneScriptingFloatChannel, default=0, const [0,44]}
       thumb_02_r_ctrl.Rotation.Z {type=MovieSceneScriptingFloatChannel, default=0, const [0,44]}
       thumb_02_r_ctrl.Scale.X {type=MovieSceneScriptingFloatChannel, default=1, const [1,44]}
       thumb_02_r_ctrl.Scale.Y {type=MovieSceneScriptingFloatChannel, default=1, const [1,44]}
       thumb_02_r_ctrl.Scale.Z {type=MovieSceneScriptingFloatChannel, default=1, const [1,44]}
       thumb_03_r_ctrl.Location.X {type=MovieSceneScriptingFloatChannel, default=0, const [0,44]}
       thumb_03_r_ctrl.Location.Y {type=MovieSceneScriptingFloatChannel, default=0, thinned_from=44, keys [[0,0],[43,0]]}
       thumb_03_r_ctrl.Location.Z {type=MovieSceneScriptingFloatChannel, default=0, thinned_from=44, keys [[0,0],[43,0]]}
       thumb_03_r_ctrl.Rotation.X {type=MovieSceneScriptingFloatChannel, default=0, const [0,44]}
       thumb_03_r_ctrl.Rotation.Y {type=MovieSceneScriptingFloatChannel, default=0, thinned_from=44, keys [[0,0],[43,0]]}
       thumb_03_r_ctrl.Rotation.Z {type=MovieSceneScriptingFloatChannel, default=-24.6415, const [-24.6415,44]}
       thumb_03_r_ctrl.Scale.X {type=MovieSceneScriptingFloatChannel, default=1, const [1,44]}
       thumb_03_r_ctrl.Scale.Y {type=MovieSceneScriptingFloatChannel, default=1, const [1,44]}
       thumb_03_r_ctrl.Scale.Z {type=MovieSceneScriptingFloatChannel, default=1, const [1,44]}
       index_metacarpal_r_ctrl.Location.X {type=MovieSceneScriptingFloatChannel, default=0, const [0,44]}
       index_metacarpal_r_ctrl.Location.Y {type=MovieSceneScriptingFloatChannel, default=0, const [0,44]}
       index_metacarpal_r_ctrl.Location.Z {type=MovieSceneScriptingFloatChannel, default=0, const [0,44]}
       index_metacarpal_r_ctrl.Rotation.X {type=MovieSceneScriptingFloatChannel, default=0, const [0,44]}
       index_metacarpal_r_ctrl.Rotation.Y {type=MovieSceneScriptingFloatChannel, default=0, const [0,44]}
       index_metacarpal_r_ctrl.Rotation.Z {type=MovieSceneScriptingFloatChannel, default=0, const [0,44]}
       index_metacarpal_r_ctrl.Scale.X {type=MovieSceneScriptingFloatChannel, default=1, const [1,44]}
       index_metacarpal_r_ctrl.Scale.Y {type=MovieSceneScriptingFloatChannel, default=1, const [1,44]}
       index_metacarpal_r_ctrl.Scale.Z {type=MovieSceneScriptingFloatChannel, default=1, const [1,44]}
       index_01_r_ctrl.Location.X {type=MovieSceneScriptingFloatChannel, default=0, const [0,44]}
       index_01_r_ctrl.Location.Y {type=MovieSceneScriptingFloatChannel, default=0, const [0,44]}
       index_01_r_ctrl.Location.Z {type=MovieSceneScriptingFloatChannel, default=0, thinned_from=44, keys [[0,0],[43,0]]}
       index_01_r_ctrl.Rotation.X {type=MovieSceneScriptingFloatChannel, default=-1.6155, const [-1.6155,44]}
       index_01_r_ctrl.Rotation.Y {type=MovieSceneScriptingFloatChannel, default=-0.9102, const [-0.9102,44]}
       index_01_r_ctrl.Rotation.Z {type=MovieSceneScriptingFloatChannel, default=-14.4207, const [-14.4207,44]}
       index_01_r_ctrl.Scale.X {type=MovieSceneScriptingFloatChannel, default=1, const [1,44]}
       index_01_r_ctrl.Scale.Y {type=MovieSceneScriptingFloatChannel, default=1, const [1,44]}
       index_01_r_ctrl.Scale.Z {type=MovieSceneScriptingFloatChannel, default=1, const [1,44]}
       index_02_r_ctrl.Location.X {type=MovieSceneScriptingFloatChannel, default=0, const [0,44]}
       index_02_r_ctrl.Location.Y {type=MovieSceneScriptingFloatChannel, default=0, const [0,44]}
       index_02_r_ctrl.Location.Z {type=MovieSceneScriptingFloatChannel, default=0, thinned_from=44, keys [[0,0],[43,0]]}
       index_02_r_ctrl.Rotation.X {type=MovieSceneScriptingFloatChannel, default=0, const [0,44]}
       index_02_r_ctrl.Rotation.Y {type=MovieSceneScriptingFloatChannel, default=0, const [0,44]}
       index_02_r_ctrl.Rotation.Z {type=MovieSceneScriptingFloatChannel, default=-4.4333, const [-4.4333,44]}
       index_02_r_ctrl.Scale.X {type=MovieSceneScriptingFloatChannel, default=1, const [1,44]}
       index_02_r_ctrl.Scale.Y {type=MovieSceneScriptingFloatChannel, default=1, const [1,44]}
       index_02_r_ctrl.Scale.Z {type=MovieSceneScriptingFloatChannel, default=1, const [1,44]}
       index_03_r_ctrl.Location.X {type=MovieSceneScriptingFloatChannel, default=0, const [0,44]}
       index_03_r_ctrl.Location.Y {type=MovieSceneScriptingFloatChannel, default=0, const [0,44]}
       index_03_r_ctrl.Location.Z {type=MovieSceneScriptingFloatChannel, default=0, thinned_from=44, keys [[0,0],[43,0]]}
       index_03_r_ctrl.Rotation.X {type=MovieSceneScriptingFloatChannel, default=0, const [0,44]}
       index_03_r_ctrl.Rotation.Y {type=MovieSceneScriptingFloatChannel, default=0, const [0,44]}
       index_03_r_ctrl.Rotation.Z {type=MovieSceneScriptingFloatChannel, default=-7.9391, const [-7.9391,44]}
       index_03_r_ctrl.Scale.X {type=MovieSceneScriptingFloatChannel, default=1, const [1,44]}
       index_03_r_ctrl.Scale.Y {type=MovieSceneScriptingFloatChannel, default=1, const [1,44]}
       index_03_r_ctrl.Scale.Z {type=MovieSceneScriptingFloatChannel, default=1, const [1,44]}
       middle_metacarpal_r_ctrl.Location.X {type=MovieSceneScriptingFloatChannel, default=0, const [0,44]}
       middle_metacarpal_r_ctrl.Location.Y {type=MovieSceneScriptingFloatChannel, default=0, const [0,44]}
       middle_metacarpal_r_ctrl.Location.Z {type=MovieSceneScriptingFloatChannel, default=0, const [0,44]}
       middle_metacarpal_r_ctrl.Rotation.X {type=MovieSceneScriptingFloatChannel, default=0, const [0,44]}
       middle_metacarpal_r_ctrl.Rotation.Y {type=MovieSceneScriptingFloatChannel, default=0, const [0,44]}
       middle_metacarpal_r_ctrl.Rotation.Z {type=MovieSceneScriptingFloatChannel, default=0, const [0,44]}
       middle_metacarpal_r_ctrl.Scale.X {type=MovieSceneScriptingFloatChannel, default=1, const [1,44]}
       middle_metacarpal_r_ctrl.Scale.Y {type=MovieSceneScriptingFloatChannel, default=1, const [1,44]}
       middle_metacarpal_r_ctrl.Scale.Z {type=MovieSceneScriptingFloatChannel, default=1, const [1,44]}
       middle_01_r_ctrl.Location.X {type=MovieSceneScriptingFloatChannel, default=0, const [0,44]}
       middle_01_r_ctrl.Location.Y {type=MovieSceneScriptingFloatChannel, default=0, const [0,44]}
       middle_01_r_ctrl.Location.Z {type=MovieSceneScriptingFloatChannel, default=0, const [0,44]}
       middle_01_r_ctrl.Rotation.X {type=MovieSceneScriptingFloatChannel, default=-2.5983, const [-2.5983,44]}
       middle_01_r_ctrl.Rotation.Y {type=MovieSceneScriptingFloatChannel, default=-2.1692, const [-2.1692,44]}
       middle_01_r_ctrl.Rotation.Z {type=MovieSceneScriptingFloatChannel, default=-19.3507, const [-19.3507,44]}
       middle_01_r_ctrl.Scale.X {type=MovieSceneScriptingFloatChannel, default=1, const [1,44]}
       middle_01_r_ctrl.Scale.Y {type=MovieSceneScriptingFloatChannel, default=1, const [1,44]}
       middle_01_r_ctrl.Scale.Z {type=MovieSceneScriptingFloatChannel, default=1, const [1,44]}
       middle_02_r_ctrl.Location.X {type=MovieSceneScriptingFloatChannel, default=0, thinned_from=44, keys [[0,0],[43,0]]}
       middle_02_r_ctrl.Location.Y {type=MovieSceneScriptingFloatChannel, default=0, const [0,44]}
       middle_02_r_ctrl.Location.Z {type=MovieSceneScriptingFloatChannel, default=0, thinned_from=44, keys [[0,0],[43,0]]}
       middle_02_r_ctrl.Rotation.X {type=MovieSceneScriptingFloatChannel, default=-0.0039, const [-0.0039,44]}
       middle_02_r_ctrl.Rotation.Y {type=MovieSceneScriptingFloatChannel, default=-0.013, const [-0.013,44]}
       middle_02_r_ctrl.Rotation.Z {type=MovieSceneScriptingFloatChannel, default=-12.5582, const [-12.5582,44]}
       middle_02_r_ctrl.Scale.X {type=MovieSceneScriptingFloatChannel, default=1, const [1,44]}
       middle_02_r_ctrl.Scale.Y {type=MovieSceneScriptingFloatChannel, default=1, const [1,44]}
       middle_02_r_ctrl.Scale.Z {type=MovieSceneScriptingFloatChannel, default=1, const [1,44]}
       middle_03_r_ctrl.Location.X {type=MovieSceneScriptingFloatChannel, default=0, const [0,44]}
       middle_03_r_ctrl.Location.Y {type=MovieSceneScriptingFloatChannel, default=0, const [0,44]}
       middle_03_r_ctrl.Location.Z {type=MovieSceneScriptingFloatChannel, default=0, thinned_from=44, keys [[0,0],[43,0]]}
       middle_03_r_ctrl.Rotation.X {type=MovieSceneScriptingFloatChannel, default=-0.011, const [-0.011,44]}
       middle_03_r_ctrl.Rotation.Y {type=MovieSceneScriptingFloatChannel, default=0.0014, const [0.0014,44]}
       middle_03_r_ctrl.Rotation.Z {type=MovieSceneScriptingFloatChannel, default=-6.4754, const [-6.4754,44]}
       middle_03_r_ctrl.Scale.X {type=MovieSceneScriptingFloatChannel, default=1, const [1,44]}
       middle_03_r_ctrl.Scale.Y {type=MovieSceneScriptingFloatChannel, default=1, const [1,44]}
       middle_03_r_ctrl.Scale.Z {type=MovieSceneScriptingFloatChannel, default=1, const [1,44]}
       ring_metacarpal_r_ctrl.Location.X {type=MovieSceneScriptingFloatChannel, default=0, const [0,44]}
       ring_metacarpal_r_ctrl.Location.Y {type=MovieSceneScriptingFloatChannel, default=0, const [0,44]}
       ring_metacarpal_r_ctrl.Location.Z {type=MovieSceneScriptingFloatChannel, default=0, const [0,44]}
       ring_metacarpal_r_ctrl.Rotation.X {type=MovieSceneScriptingFloatChannel, default=0, const [0,44]}
       ring_metacarpal_r_ctrl.Rotation.Y {type=MovieSceneScriptingFloatChannel, default=0, const [0,44]}
       ring_metacarpal_r_ctrl.Rotation.Z {type=MovieSceneScriptingFloatChannel, default=0, const [0,44]}
       ring_metacarpal_r_ctrl.Scale.X {type=MovieSceneScriptingFloatChannel, default=1, const [1,44]}
       ring_metacarpal_r_ctrl.Scale.Y {type=MovieSceneScriptingFloatChannel, default=1, const [1,44]}
       ring_metacarpal_r_ctrl.Scale.Z {type=MovieSceneScriptingFloatChannel, default=1, const [1,44]}
       ring_01_r_ctrl.Location.X {type=MovieSceneScriptingFloatChannel, default=0, const [0,44]}
       ring_01_r_ctrl.Location.Y {type=MovieSceneScriptingFloatChannel, default=0, thinned_from=44, keys [[0,0],[43,0]]}
       ring_01_r_ctrl.Location.Z {type=MovieSceneScriptingFloatChannel, default=0, thinned_from=44, keys [[0,0],[43,0]]}
       ring_01_r_ctrl.Rotation.X {type=MovieSceneScriptingFloatChannel, default=-1.6664, const [-1.6664,44]}
       ring_01_r_ctrl.Rotation.Y {type=MovieSceneScriptingFloatChannel, default=-1.5704, const [-1.5704,44]}
       ring_01_r_ctrl.Rotation.Z {type=MovieSceneScriptingFloatChannel, default=-13.1407, const [-13.1407,44]}
       ring_01_r_ctrl.Scale.X {type=MovieSceneScriptingFloatChannel, default=1, const [1,44]}
       ring_01_r_ctrl.Scale.Y {type=MovieSceneScriptingFloatChannel, default=1, const [1,44]}
       ring_01_r_ctrl.Scale.Z {type=MovieSceneScriptingFloatChannel, default=1, const [1,44]}
       ring_02_r_ctrl.Location.X {type=MovieSceneScriptingFloatChannel, default=0, thinned_from=44, keys [[0,0],[43,0]]}
       ring_02_r_ctrl.Location.Y {type=MovieSceneScriptingFloatChannel, default=0, thinned_from=44, keys [[0,0],[43,0]]}
       ring_02_r_ctrl.Location.Z {type=MovieSceneScriptingFloatChannel, default=0, thinned_from=44, keys [[0,0],[43,0]]}
       ring_02_r_ctrl.Rotation.X {type=MovieSceneScriptingFloatChannel, default=0, thinned_from=44, keys [[0,0],[43,0]]}
       ring_02_r_ctrl.Rotation.Y {type=MovieSceneScriptingFloatChannel, default=0, const [0,44]}
       ring_02_r_ctrl.Rotation.Z {type=MovieSceneScriptingFloatChannel, default=-14.3916, const [-14.3916,44]}
       ring_02_r_ctrl.Scale.X {type=MovieSceneScriptingFloatChannel, default=1, const [1,44]}
       ring_02_r_ctrl.Scale.Y {type=MovieSceneScriptingFloatChannel, default=1, const [1,44]}
       ring_02_r_ctrl.Scale.Z {type=MovieSceneScriptingFloatChannel, default=1, const [1,44]}
       ring_03_r_ctrl.Location.X {type=MovieSceneScriptingFloatChannel, default=0, thinned_from=44, keys [[0,0],[43,0]]}
       ring_03_r_ctrl.Location.Y {type=MovieSceneScriptingFloatChannel, default=0, const [0,44]}
       ring_03_r_ctrl.Location.Z {type=MovieSceneScriptingFloatChannel, default=0, thinned_from=44, keys [[0,0],[43,0]]}
       ring_03_r_ctrl.Rotation.X {type=MovieSceneScriptingFloatChannel, default=0, thinned_from=44, keys [[0,0],[43,0]]}
       ring_03_r_ctrl.Rotation.Y {type=MovieSceneScriptingFloatChannel, default=0, const [0,44]}
       ring_03_r_ctrl.Rotation.Z {type=MovieSceneScriptingFloatChannel, default=-20.52, const [-20.52,44]}
       ring_03_r_ctrl.Scale.X {type=MovieSceneScriptingFloatChannel, default=1, const [1,44]}
       ring_03_r_ctrl.Scale.Y {type=MovieSceneScriptingFloatChannel, default=1, const [1,44]}
       ring_03_r_ctrl.Scale.Z {type=MovieSceneScriptingFloatChannel, default=1, const [1,44]}
       pinky_metacarpal_r_ctrl.Location.X {type=MovieSceneScriptingFloatChannel, default=0, const [0,44]}
       pinky_metacarpal_r_ctrl.Location.Y {type=MovieSceneScriptingFloatChannel, default=0, const [0,44]}
       pinky_metacarpal_r_ctrl.Location.Z {type=MovieSceneScriptingFloatChannel, default=0, const [0,44]}
       pinky_metacarpal_r_ctrl.Rotation.X {type=MovieSceneScriptingFloatChannel, default=0, const [0,44]}
       pinky_metacarpal_r_ctrl.Rotation.Y {type=MovieSceneScriptingFloatChannel, default=0, const [0,44]}
       pinky_metacarpal_r_ctrl.Rotation.Z {type=MovieSceneScriptingFloatChannel, default=0, const [0,44]}
       pinky_metacarpal_r_ctrl.Scale.X {type=MovieSceneScriptingFloatChannel, default=1, const [1,44]}
       pinky_metacarpal_r_ctrl.Scale.Y {type=MovieSceneScriptingFloatChannel, default=1, const [1,44]}
       pinky_metacarpal_r_ctrl.Scale.Z {type=MovieSceneScriptingFloatChannel, default=1, const [1,44]}
       pinky_01_r_ctrl.Location.X {type=MovieSceneScriptingFloatChannel, default=0, thinned_from=44, keys [[0,0],[43,0]]}
       pinky_01_r_ctrl.Location.Y {type=MovieSceneScriptingFloatChannel, default=0, const [0,44]}
       pinky_01_r_ctrl.Location.Z {type=MovieSceneScriptingFloatChannel, default=0, thinned_from=44, keys [[0,0],[43,0]]}
       pinky_01_r_ctrl.Rotation.X {type=MovieSceneScriptingFloatChannel, default=3.2669, const [3.2669,44]}
       pinky_01_r_ctrl.Rotation.Y {type=MovieSceneScriptingFloatChannel, default=-5.7306, const [-5.7306,44]}
       pinky_01_r_ctrl.Rotation.Z {type=MovieSceneScriptingFloatChannel, default=-4.1286, const [-4.1286,44]}
       pinky_01_r_ctrl.Scale.X {type=MovieSceneScriptingFloatChannel, default=1, const [1,44]}
       pinky_01_r_ctrl.Scale.Y {type=MovieSceneScriptingFloatChannel, default=1, const [1,44]}
       pinky_01_r_ctrl.Scale.Z {type=MovieSceneScriptingFloatChannel, default=1, const [1,44]}
       pinky_02_r_ctrl.Location.X {type=MovieSceneScriptingFloatChannel, default=0, thinned_from=44, keys [[0,0],[43,0]]}
       pinky_02_r_ctrl.Location.Y {type=MovieSceneScriptingFloatChannel, default=0, thinned_from=44, keys [[0,0],[43,0]]}
       pinky_02_r_ctrl.Location.Z {type=MovieSceneScriptingFloatChannel, default=0, thinned_from=44, keys [[0,0],[43,0]]}
       pinky_02_r_ctrl.Rotation.X {type=MovieSceneScriptingFloatChannel, default=0, thinned_from=44, keys [[0,0],[43,0]]}
       pinky_02_r_ctrl.Rotation.Y {type=MovieSceneScriptingFloatChannel, default=0, thinned_from=44, keys [[0,0],[43,0]]}
       pinky_02_r_ctrl.Rotation.Z {type=MovieSceneScriptingFloatChannel, default=-13.151, const [-13.151,44]}
       pinky_02_r_ctrl.Scale.X {type=MovieSceneScriptingFloatChannel, default=1, const [1,44]}
       pinky_02_r_ctrl.Scale.Y {type=MovieSceneScriptingFloatChannel, default=1, const [1,44]}
       pinky_02_r_ctrl.Scale.Z {type=MovieSceneScriptingFloatChannel, default=1, const [1,44]}
       pinky_03_r_ctrl.Location.X {type=MovieSceneScriptingFloatChannel, default=0, thinned_from=44, keys [[0,0],[43,0]]}
       pinky_03_r_ctrl.Location.Y {type=MovieSceneScriptingFloatChannel, default=0, thinned_from=44, keys [[0,0],[43,0]]}
       pinky_03_r_ctrl.Location.Z {type=MovieSceneScriptingFloatChannel, default=0, thinned_from=44, keys [[0,0],[43,0]]}
       pinky_03_r_ctrl.Rotation.X {type=MovieSceneScriptingFloatChannel, default=0, thinned_from=44, keys [[0,0],[43,0]]}
       pinky_03_r_ctrl.Rotation.Y {type=MovieSceneScriptingFloatChannel, default=0, const [0,44]}
       pinky_03_r_ctrl.Rotation.Z {type=MovieSceneScriptingFloatChannel, default=-22.5109, const [-22.5109,44]}
       pinky_03_r_ctrl.Scale.X {type=MovieSceneScriptingFloatChannel, default=1, const [1,44]}
       pinky_03_r_ctrl.Scale.Y {type=MovieSceneScriptingFloatChannel, default=1, const [1,44]}
       pinky_03_r_ctrl.Scale.Z {type=MovieSceneScriptingFloatChannel, default=1, const [1,44]}
       hand_r_ik_ctrl.Location.X {type=MovieSceneScriptingFloatChannel, default=-42.8068, thinned_from=44, keys [[0,-42.8068],[11,-43.3605],[30,-42.2648],[43,-42.8068]]}
       hand_r_ik_ctrl.Location.Y {
        type=MovieSceneScriptingFloatChannel
        default=9.1104
        thinned_from=44
        keys [[0,9.1104],[6,8.9795],[9,8.9646],[12,9.0057],[15,9.168],[21,9.5843],[24,9.6495],[27,9.6375],[35,9.412],[42,9.1366],[43,9.1104]]
       }
       hand_r_ik_ctrl.Location.Z {type=MovieSceneScriptingFloatChannel, default=-39.6889, thinned_from=44, keys [[0,-39.6889],[13,-38.6533],[42,-39.7074],[43,-39.6889]]}
       hand_r_ik_ctrl.Rotation.X {
        type=MovieSceneScriptingFloatChannel
        default=6.4265
        thinned_from=44
        keys [[0,6.4265],[1,6.3313],[2,6.2131],[4,5.9343],[7,5.4983],[8,5.3695],[9,5.2565],[10,5.1638],[11,5.0951],[12,5.0427],[13,5.0542],[14,5.0409],[16,5.0903],[18,5.2023],[19,5.2808],[21,5.4226],[23,5.6213],[28,6.213],[30,6.4219],[31,6.5047],[32,6.5673],[33,6.6055],[35,6.643],[37,6.6519],[38,6.6307],[39,6.6234],[41,6.5513],[43,6.4265]]
       }
       hand_r_ik_ctrl.Rotation.Y {
        type=MovieSceneScriptingFloatChannel
        default=-36.1341
        thinned_from=44
        keys [[0,-36.1341],[3,-35.3658],[7,-34.1624],[9,-33.6963],[11,-33.4474],[12,-33.423],[15,-33.6893],[18,-34.2215],[28,-36.7573],[30,-37.0994],[32,-37.2674],[36,-37.0984],[42,-36.3121],[43,-36.1341]]
       }
       hand_r_ik_ctrl.Rotation.Z {
        type=MovieSceneScriptingFloatChannel
        default=-86.2859
        thinned_from=44
        keys [[0,-86.2859],[11,-83.6428],[17,-83.5365],[33,-86.4368],[41,-86.5074],[43,-86.2859]]
       }
       hand_r_ik_ctrl.Scale.X {type=MovieSceneScriptingFloatChannel, default=1, const [1,44]}
       hand_r_ik_ctrl.Scale.Y {type=MovieSceneScriptingFloatChannel, default=1, const [1,44]}
       hand_r_ik_ctrl.Scale.Z {type=MovieSceneScriptingFloatChannel, default=1, const [1,44]}
       Softness_2 {type=MovieSceneScriptingFloatChannel, default=0, const [0,44]}
       hand_r_ikorient_ctrl.Rotation.Roll {type=MovieSceneScriptingFloatChannel, default=0, const [0,44]}
       hand_r_ikorient_ctrl.Rotation.Pitch {type=MovieSceneScriptingFloatChannel, default=0, const [0,44]}
       hand_r_ikorient_ctrl.Rotation.Yaw {type=MovieSceneScriptingFloatChannel, default=0, const [0,44]}
       arm_r_pv_ik_ctrl.Location.X {type=MovieSceneScriptingFloatChannel, default=-52.9543, thinned_from=44, keys [[0,-52.9543],[11,-53.9367],[35,-52.6159],[43,-52.9543]]}
       arm_r_pv_ik_ctrl.Location.Y {
        type=MovieSceneScriptingFloatChannel
        default=-25.8438
        thinned_from=44
        keys [[0,-25.8438],[10,-24.9319],[13,-24.8018],[18,-24.962],[32,-25.9313],[38,-26.0137],[43,-25.8438]]
       }
       arm_r_pv_ik_ctrl.Location.Z {
        type=MovieSceneScriptingFloatChannel
        default=19.2551
        thinned_from=44
        keys [[0,19.2551],[5,19.2695],[12,19.5604],[19,20.2161],[22,20.3742],[25,20.3743],[34,19.8667],[41,19.3228],[43,19.2551]]
       }
       arm_r_pv_ik_ctrl.Rotation.X {
        type=MovieSceneScriptingFloatChannel
        default=22.545
        thinned_from=44
        keys [[0,22.545],[4,22.2723],[7,22.1971],[10,22.278],[12,22.4566],[13,22.6111],[15,23.0826],[19,24.2731],[21,24.7543],[22,24.9019],[24,25.058],[26,25.0795],[29,24.8953],[33,24.3947],[36,23.826],[41,22.7709],[43,22.545]]
       }
       arm_r_pv_ik_ctrl.Rotation.Y {
        type=MovieSceneScriptingFloatChannel
        default=39.3922
        thinned_from=44
        keys [[0,39.3922],[11,40.2639],[15,40.2291],[32,39.007],[38,39.0647],[43,39.3922]]
       }
       arm_r_pv_ik_ctrl.Rotation.Z {
        type=MovieSceneScriptingFloatChannel
        default=-74.9958
        thinned_from=44
        keys [[0,-74.9958],[7,-76.7232],[10,-77.043],[13,-76.8229],[22,-73.5108],[27,-72.4954],[31,-72.3346],[35,-72.8139],[43,-74.9958]]
       }
       arm_r_pv_ik_ctrl.Scale.X {type=MovieSceneScriptingFloatChannel, default=1, const [1,44]}
       arm_r_pv_ik_ctrl.Scale.Y {type=MovieSceneScriptingFloatChannel, default=1, const [1,44]}
       arm_r_pv_ik_ctrl.Scale.Z {type=MovieSceneScriptingFloatChannel, default=1, const [1,44]}
       body_ctrl.Location.X {
        type=MovieSceneScriptingFloatChannel
        default=2.8561
        thinned_from=44
        keys [[0,2.8561],[4,2.8048],[6,2.792],[8,2.795],[11,2.8236],[23,3.0446],[26,3.0799],[28,3.0895],[30,3.0852],[32,3.0654],[42,2.8719],[43,2.8561]]
       }
       body_ctrl.Location.Y {
        type=MovieSceneScriptingFloatChannel
        default=-0.5418
        thinned_from=44
        keys [[0,-0.5418],[3,-0.5978],[5,-0.6321],[7,-0.6625],[8,-0.6757],[9,-0.6874],[10,-0.6974],[11,-0.7056],[12,-0.7118],[13,-0.7157],[14,-0.7172],[15,-0.716],[16,-0.7119],[17,-0.7047],[18,-0.6942],[19,-0.68],[20,-0.6626],[21,-0.6431],[24,-0.5803],[26,-0.5416],[27,-0.5239],[28,-0.5075],[29,-0.4927],[30,-0.4796],[31,-0.4684],[32,-0.4594],[33,-0.4527],[34,-0.4486],[35,-0.4473],[36,-0.4488],[37,-0.4536],[38,-0.4617],[39,-0.4734],[40,-0.4894],[42,-0.5268],[43,-0.5418]]
       }
       body_ctrl.Location.Z {type=MovieSceneScriptingFloatChannel, default=-5.3046, thinned_from=44, keys [[0,-5.3046],[24,-5.3809],[43,-5.3046]]}
       body_ctrl.Rotation.X {
        type=MovieSceneScriptingFloatChannel
        default=3.6931
        thinned_from=44
        keys [[0,3.6931],[1,3.7231],[2,3.7761],[3,3.8496],[4,3.941],[5,4.0479],[6,4.1677],[8,4.4361],[13,5.1582],[15,5.4186],[16,5.5339],[17,5.6364],[18,5.7239],[19,5.7941],[20,5.8449],[22,5.8794],[23,5.8576],[24,5.8088],[25,5.7358],[26,5.6418],[27,5.5296],[28,5.4022],[30,5.1138],[34,4.4858],[36,4.1938],[37,4.0637],[38,3.9481],[39,3.8501],[40,3.7726],[41,3.7186],[42,3.6911],[43,3.6931]]
       }
       body_ctrl.Rotation.Y {
        type=MovieSceneScriptingFloatChannel
        default=2.4702
        thinned_from=44
        keys [[0,2.4702],[1,2.4172],[2,2.3508],[3,2.2727],[4,2.185],[5,2.0892],[6,1.9875],[10,1.5565],[11,1.4522],[12,1.3525],[13,1.2582],[14,1.1709],[15,1.0916],[16,1.0213],[17,0.9613],[18,0.9129],[19,0.8772],[20,0.8554],[21,0.8391],[22,0.8581],[23,0.887],[24,0.9364],[25,1.0038],[26,1.0864],[27,1.1823],[28,1.2886],[29,1.403],[30,1.5231],[33,1.8918],[34,2.0095],[35,2.1205],[36,2.2222],[37,2.3121],[38,2.3883],[39,2.4475],[40,2.488],[41,2.5069],[42,2.5017],[43,2.4702]]
       }
       body_ctrl.Rotation.Z {type=MovieSceneScriptingFloatChannel, default=35.6319, thinned_from=44, keys [[0,35.6319],[43,35.6319]]}
       body_ctrl.Scale.X {type=MovieSceneScriptingFloatChannel, default=1, const [1,44]}
       body_ctrl.Scale.Y {type=MovieSceneScriptingFloatChannel, default=1, const [1,44]}
       body_ctrl.Scale.Z {type=MovieSceneScriptingFloatChannel, default=1, const [1,44]}
       hips_ctrl.Location.X {type=MovieSceneScriptingFloatChannel, default=0, thinned_from=44, keys [[0,0],[43,0]]}
       hips_ctrl.Location.Y {type=MovieSceneScriptingFloatChannel, default=0, thinned_from=44, keys [[0,0],[43,0]]}
       hips_ctrl.Location.Z {type=MovieSceneScriptingFloatChannel, default=0, thinned_from=44, keys [[0,0],[43,0]]}
       hips_ctrl.Rotation.X {
        type=MovieSceneScriptingFloatChannel
        default=-6.4387
        thinned_from=44
        keys [[0,-6.4387],[2,-6.5086],[4,-6.6197],[14,-7.3933],[17,-7.5698],[19,-7.6458],[20,-7.6681],[22,-7.676],[23,-7.6584],[25,-7.5786],[27,-7.4514],[35,-6.7535],[37,-6.6032],[39,-6.4912],[40,-6.4539],[41,-6.4316],[42,-6.4259],[43,-6.4387]]
       }
       hips_ctrl.Rotation.Y {
        type=MovieSceneScriptingFloatChannel
        default=1.1853
        thinned_from=44
        keys [[0,1.1853],[6,1.4581],[8,1.5424],[9,1.5806],[10,1.6151],[11,1.6453],[12,1.6714],[13,1.6934],[14,1.7112],[15,1.7248],[16,1.7344],[17,1.7398],[18,1.7412],[19,1.7385],[20,1.7317],[21,1.7317],[22,1.7061],[23,1.6855],[24,1.6579],[25,1.6244],[26,1.586],[28,1.4988],[32,1.3109],[33,1.2672],[34,1.2269],[35,1.1909],[36,1.1604],[37,1.1363],[38,1.1197],[39,1.1117],[40,1.1132],[41,1.1252],[42,1.1489],[43,1.1853]]
       }
       hips_ctrl.Rotation.Z {type=MovieSceneScriptingFloatChannel, default=-6.5486, thinned_from=44, keys [[0,-6.5486],[21,-6.6615],[41,-6.542],[43,-6.5486]]}
       hips_ctrl.Scale.X {type=MovieSceneScriptingFloatChannel, default=1, const [1,44]}
       hips_ctrl.Scale.Y {type=MovieSceneScriptingFloatChannel, default=1, const [1,44]}
       hips_ctrl.Scale.Z {type=MovieSceneScriptingFloatChannel, default=1, const [1,44]}
       hips_tan_ctrl.Location.X {
        type=MovieSceneScriptingFloatChannel
        default=-0.0149
        thinned_from=44
        keys [[0,-0.0149],[12,-0.0136],[13,-0.0144],[14,-0.0136],[15,-0.0141],[16,-0.0136],[18,-0.014],[21,-0.0138],[22,-0.013],[23,-0.0139],[37,-0.0153],[38,-0.0147],[39,-0.0152],[41,-0.0152],[43,-0.0149]]
       }
       hips_tan_ctrl.Location.Y {
        type=MovieSceneScriptingFloatChannel
        default=-0.0583
        thinned_from=44
        keys [[0,-0.0583],[3,-0.0586],[7,-0.06],[12,-0.0627],[13,-0.064],[14,-0.0639],[15,-0.0649],[16,-0.0649],[18,-0.0663],[19,-0.0663],[20,-0.0671],[22,-0.0671],[23,-0.0666],[26,-0.0659],[37,-0.0602],[38,-0.0592],[39,-0.0593],[40,-0.0587],[41,-0.0587],[43,-0.0583]]
       }
       hips_tan_ctrl.Location.Z {type=MovieSceneScriptingFloatChannel, default=-0.0132, thinned_from=44, keys [[0,-0.0132],[22,-0.0152],[43,-0.0132]]}
       hips_tan_ctrl.Rotation.X {type=MovieSceneScriptingFloatChannel, default=0, const [0,44]}
       hips_tan_ctrl.Rotation.Y {type=MovieSceneScriptingFloatChannel, default=0, const [0,44]}
       hips_tan_ctrl.Rotation.Z {type=MovieSceneScriptingFloatChannel, default=0, const [0,44]}
       hips_tan_ctrl.Scale.X {type=MovieSceneScriptingFloatChannel, default=1, const [1,44]}
       hips_tan_ctrl.Scale.Y {type=MovieSceneScriptingFloatChannel, default=1, const [1,44]}
       hips_tan_ctrl.Scale.Z {type=MovieSceneScriptingFloatChannel, default=1, const [1,44]}
       spine_01_ctrl.Location.X {type=MovieSceneScriptingFloatChannel, default=-0.0136, thinned_from=44, keys [[0,-0.0136],[43,-0.0136]]}
       spine_01_ctrl.Location.Y {type=MovieSceneScriptingFloatChannel, default=-0.0003, thinned_from=44, keys [[0,-0.0003],[43,-0.0003]]}
       spine_01_ctrl.Location.Z {type=MovieSceneScriptingFloatChannel, default=-0.0015, thinned_from=44, keys [[0,-0.0015],[43,-0.0015]]}
       spine_01_ctrl.Rotation.X {
        type=MovieSceneScriptingFloatChannel
        default=1.0049
        thinned_from=44
        keys [[0,1.0049],[3,0.9849],[11,0.9134],[12,0.9056],[13,0.9056],[14,0.8919],[15,0.8919],[16,0.8813],[22,0.8813],[23,0.8761],[25,0.8868],[28,0.9118],[35,0.9816],[37,0.9969],[38,0.9969],[39,1.007],[41,1.007],[42,1.0087],[43,1.0049]]
       }
       spine_01_ctrl.Rotation.Y {
        type=MovieSceneScriptingFloatChannel
        default=0.5413
        thinned_from=44
        keys [[0,0.5413],[2,0.5821],[6,0.668],[8,0.7082],[9,0.7266],[10,0.7433],[11,0.7582],[12,0.7712],[13,0.7712],[14,0.7913],[15,0.7913],[16,0.8038],[22,0.8038],[23,0.7846],[24,0.7719],[25,0.7563],[26,0.7383],[28,0.6972],[32,0.6078],[33,0.5867],[34,0.5672],[35,0.5498],[36,0.5348],[37,0.5228],[38,0.5228],[39,0.5098],[41,0.5098],[42,0.525],[43,0.5413]]
       }
       spine_01_ctrl.Rotation.Z {type=MovieSceneScriptingFloatChannel, default=6.4292, thinned_from=44, keys [[0,6.4292],[43,6.4292]]}
       spine_01_ctrl.Scale.X {type=MovieSceneScriptingFloatChannel, default=1, const [1,44]}
       spine_01_ctrl.Scale.Y {type=MovieSceneScriptingFloatChannel, default=1, const [1,44]}
       spine_01_ctrl.Scale.Z {type=MovieSceneScriptingFloatChannel, default=1, const [1,44]}
       spine_02_ctrl.Location.X {type=MovieSceneScriptingFloatChannel, default=-0.0093, thinned_from=44, keys [[0,-0.0093],[43,-0.0093]]}
       spine_02_ctrl.Location.Y {type=MovieSceneScriptingFloatChannel, default=-0.0005, thinned_from=44, keys [[0,-0.0005],[43,-0.0005]]}
       spine_02_ctrl.Location.Z {type=MovieSceneScriptingFloatChannel, default=-0.0004, thinned_from=44, keys [[0,-0.0004],[43,-0.0004]]}
       spine_02_ctrl.Rotation.X {
        type=MovieSceneScriptingFloatChannel
        default=0.2758
        thinned_from=44
        keys [[0,0.2758],[1,0.2693],[2,0.2591],[3,0.2455],[4,0.2291],[5,0.2102],[6,0.1894],[7,0.167],[8,0.1434],[11,0.0705],[12,0.0468],[13,0.0239],[14,0.0023],[15,-0.0178],[16,-0.036],[17,-0.052],[18,-0.0655],[19,-0.0655],[20,-0.0836],[21,-0.0836],[22,-0.0875],[23,-0.0831],[24,-0.0742],[25,-0.0613],[26,-0.045],[27,-0.0257],[28,-0.0039],[29,0.0198],[30,0.045],[31,0.0712],[33,0.1243],[34,0.1503],[35,0.1753],[36,0.1986],[37,0.2199],[38,0.2386],[39,0.2542],[40,0.2663],[41,0.2663],[42,0.2776],[43,0.2758]]
       }
       spine_02_ctrl.Rotation.Y {
        type=MovieSceneScriptingFloatChannel
        default=1.1316
        thinned_from=44
        keys [[0,1.1316],[2,1.197],[8,1.4156],[10,1.481],[12,1.5353],[13,1.5578],[14,1.5772],[15,1.5934],[16,1.6061],[17,1.6153],[18,1.621],[21,1.621],[23,1.5902],[24,1.5688],[25,1.5421],[26,1.5108],[28,1.4385],[32,1.2779],[34,1.2036],[35,1.1709],[36,1.1423],[37,1.1188],[38,1.1012],[39,1.0903],[40,1.087],[41,1.087],[42,1.1068],[43,1.1316]]
       }
       spine_02_ctrl.Rotation.Z {type=MovieSceneScriptingFloatChannel, default=6.3859, thinned_from=44, keys [[0,6.3859],[25,6.3325],[43,6.3859]]}
       spine_02_ctrl.Scale.X {type=MovieSceneScriptingFloatChannel, default=1, const [1,44]}
       spine_02_ctrl.Scale.Y {type=MovieSceneScriptingFloatChannel, default=1, const [1,44]}
       spine_02_ctrl.Scale.Z {type=MovieSceneScriptingFloatChannel, default=1, const [1,44]}
       spine_03_ctrl.Location.X {type=MovieSceneScriptingFloatChannel, default=0.005, thinned_from=44, keys [[0,0.005],[43,0.005]]}
       spine_03_ctrl.Location.Y {type=MovieSceneScriptingFloatChannel, default=0.0004, thinned_from=44, keys [[0,0.0004],[43,0.0004]]}
       spine_03_ctrl.Location.Z {type=MovieSceneScriptingFloatChannel, default=-0.0006, thinned_from=44, keys [[0,-0.0006],[43,-0.0006]]}
       spine_03_ctrl.Rotation.X {
        type=MovieSceneScriptingFloatChannel
        default=-0.7816
        thinned_from=44
        keys [[0,-0.7816],[1,-0.7871],[2,-0.7996],[3,-0.8185],[4,-0.8429],[5,-0.8722],[6,-0.9056],[7,-0.9424],[9,-1.0229],[13,-1.1914],[15,-1.2685],[16,-1.3029],[17,-1.3336],[18,-1.36],[19,-1.3815],[20,-1.3974],[21,-1.3974],[22,-1.4095],[23,-1.4043],[24,-1.3914],[25,-1.3716],[26,-1.3457],[27,-1.3147],[28,-1.2793],[29,-1.2403],[31,-1.1549],[34,-1.0209],[35,-0.9779],[36,-0.937],[37,-0.8992],[38,-0.8653],[39,-0.8359],[40,-0.8121],[41,-0.7945],[42,-0.7841],[43,-0.7816]]
       }
       spine_03_ctrl.Rotation.Y {
        type=MovieSceneScriptingFloatChannel
        default=1.7796
        thinned_from=44
        keys [[0,1.7796],[1,1.8305],[2,1.8862],[4,2.0083],[8,2.2664],[10,2.3859],[12,2.4883],[13,2.5321],[14,2.5705],[15,2.6033],[16,2.6303],[17,2.6509],[18,2.6651],[19,2.6724],[21,2.6726],[23,2.6255],[24,2.5891],[25,2.5431],[26,2.4888],[27,2.4277],[29,2.2916],[32,2.0751],[34,1.9402],[35,1.8803],[36,1.8272],[37,1.7827],[38,1.7482],[39,1.7253],[40,1.7154],[41,1.7202],[42,1.7411],[43,1.7796]]
       }
       spine_03_ctrl.Rotation.Z {type=MovieSceneScriptingFloatChannel, default=6.2964, thinned_from=44, keys [[0,6.2964],[21,6.1752],[40,6.3036],[43,6.2964]]}
       spine_03_ctrl.Scale.X {type=MovieSceneScriptingFloatChannel, default=1, const [1,44]}
       spine_03_ctrl.Scale.Y {type=MovieSceneScriptingFloatChannel, default=1, const [1,44]}
       spine_03_ctrl.Scale.Z {type=MovieSceneScriptingFloatChannel, default=1, const [1,44]}
       spine_04_ctrl.Location.X {type=MovieSceneScriptingFloatChannel, default=0}
       spine_04_ctrl.Location.Y {type=MovieSceneScriptingFloatChannel, default=0}
       spine_04_ctrl.Location.Z {type=MovieSceneScriptingFloatChannel, default=0}
       spine_04_ctrl.Rotation.X {type=MovieSceneScriptingFloatChannel, default=0}
       spine_04_ctrl.Rotation.Y {type=MovieSceneScriptingFloatChannel, default=0}
       spine_04_ctrl.Rotation.Z {type=MovieSceneScriptingFloatChannel, default=0}
       spine_04_ctrl.Scale.X {type=MovieSceneScriptingFloatChannel, default=1}
       spine_04_ctrl.Scale.Y {type=MovieSceneScriptingFloatChannel, default=1}
       spine_04_ctrl.Scale.Z {type=MovieSceneScriptingFloatChannel, default=1}
       head_ctrl.Location.X {type=MovieSceneScriptingFloatChannel, default=0, thinned_from=44, keys [[0,0],[43,0]]}
       head_ctrl.Location.Y {type=MovieSceneScriptingFloatChannel, default=0, thinned_from=44, keys [[0,0],[43,0]]}
       head_ctrl.Location.Z {type=MovieSceneScriptingFloatChannel, default=0, thinned_from=44, keys [[0,0],[43,0]]}
       head_ctrl.Rotation.X {
        type=MovieSceneScriptingFloatChannel
        default=4.2402
        thinned_from=44
        keys [[0,4.2402],[1,4.1783],[2,4.1049],[3,3.995],[4,3.8616],[5,3.7081],[6,3.538],[7,3.3548],[9,2.9626],[11,2.56],[12,2.3627],[13,2.1841],[14,1.9908],[16,1.6679],[17,1.5373],[18,1.4256],[19,1.3454],[20,1.2706],[21,1.2366],[22,1.2157],[23,1.2588],[24,1.3299],[25,1.4351],[26,1.5694],[27,1.7286],[28,1.9086],[29,2.105],[30,2.3139],[34,3.1888],[35,3.3964],[36,3.5911],[37,3.7688],[39,4.0561],[40,4.155],[41,4.2214],[42,4.2535],[43,4.2506]]
       }
       head_ctrl.Rotation.Y {
        type=MovieSceneScriptingFloatChannel
        default=-7.7167
        thinned_from=44
        keys [[0,-7.7167],[3,-7.5824],[12,-7.0058],[14,-6.8965],[17,-6.7735],[19,-6.7187],[20,-6.7097],[21,-6.6764],[24,-6.7716],[26,-6.8686],[35,-7.5194],[37,-7.6367],[39,-7.7169],[41,-7.7626],[43,-7.7155]]
       }
       head_ctrl.Rotation.Z {type=MovieSceneScriptingFloatChannel, default=-25.8805, thinned_from=44, keys [[0,-25.8805],[43,-25.8813]]}
       head_ctrl.Scale.X {type=MovieSceneScriptingFloatChannel, default=1, const [1,44]}
       head_ctrl.Scale.Y {type=MovieSceneScriptingFloatChannel, default=1, const [1,44]}
       head_ctrl.Scale.Z {type=MovieSceneScriptingFloatChannel, default=1, const [1,44]}
       clavicle_l_ctrl.Location.X {
        type=MovieSceneScriptingFloatChannel
        default=0.0233
        thinned_from=44
        keys [[0,0.0233],[1,0.0205],[2,0.0175],[4,0.0107],[8,-0.0036],[10,-0.0103],[11,-0.0133],[12,-0.0161],[13,-0.0186],[14,-0.0207],[15,-0.0226],[16,-0.0242],[17,-0.0254],[18,-0.0262],[19,-0.0266],[21,-0.0267],[23,-0.0241],[24,-0.0221],[25,-0.0195],[26,-0.0165],[27,-0.013],[28,-0.0093],[32,0.0066],[33,0.0105],[34,0.0142],[35,0.0175],[36,0.0205],[37,0.023],[38,0.0249],[39,0.0262],[40,0.0268],[41,0.0265],[42,0.0254],[43,0.0233]]
       }
       clavicle_l_ctrl.Location.Y {
        type=MovieSceneScriptingFloatChannel
        default=-0.1272
        thinned_from=44
        keys [[0,-0.1272],[1,-0.1277],[2,-0.1288],[3,-0.1303],[4,-0.1321],[6,-0.1366],[14,-0.159],[16,-0.1638],[18,-0.1676],[19,-0.169],[20,-0.17],[21,-0.17],[22,-0.1707],[23,-0.1702],[24,-0.1692],[25,-0.1678],[26,-0.1659],[28,-0.1611],[35,-0.1399],[37,-0.1345],[38,-0.1322],[39,-0.1303],[40,-0.1288],[41,-0.1277],[42,-0.1271],[43,-0.1272]]
       }
       clavicle_l_ctrl.Location.Z {
        type=MovieSceneScriptingFloatChannel
        default=-0.0205
        thinned_from=44
        keys [[0,-0.0205],[9,-0.0086],[11,-0.0064],[13,-0.0047],[15,-0.0035],[17,-0.0027],[19,-0.0026],[21,-0.0027],[23,-0.0039],[25,-0.0056],[27,-0.0081],[33,-0.0168],[35,-0.0193],[37,-0.0212],[38,-0.0218],[39,-0.0222],[40,-0.0223],[41,-0.0221],[42,-0.0215],[43,-0.0205]]
       }
       clavicle_l_ctrl.Rotation.X {
        type=MovieSceneScriptingFloatChannel
        default=-4.6636
        thinned_from=44
        keys [[0,-4.6636],[2,-4.673],[5,-4.7228],[17,-5.0706],[20,-5.1219],[21,-5.1217],[22,-5.1329],[24,-5.1211],[27,-5.0662],[38,-4.733],[41,-4.677],[43,-4.6636]]
       }
       clavicle_l_ctrl.Rotation.Y {
        type=MovieSceneScriptingFloatChannel
        default=5.1794
        thinned_from=44
        keys [[0,5.1794],[2,5.1722],[4,5.1894],[7,5.2528],[8,5.2823],[9,5.3304],[10,5.351],[13,5.4691],[14,5.5011],[15,5.5618],[21,5.8163],[24,5.9014],[26,5.9307],[28,5.9318],[30,5.9016],[32,5.8371],[34,5.7319],[39,5.3609],[40,5.2949],[41,5.2401],[42,5.2005],[43,5.1794]]
       }
       clavicle_l_ctrl.Rotation.Z {type=MovieSceneScriptingFloatChannel, default=7.1304, thinned_from=44, keys [[0,7.1304],[43,7.1304]]}
       clavicle_l_ctrl.Scale.X {type=MovieSceneScriptingFloatChannel, default=1, const [1,44]}
       clavicle_l_ctrl.Scale.Y {type=MovieSceneScriptingFloatChannel, default=1, const [1,44]}
       clavicle_l_ctrl.Scale.Z {type=MovieSceneScriptingFloatChannel, default=1, const [1,44]}
       clavicle_r_ctrl.Location.X {
        type=MovieSceneScriptingFloatChannel
        default=-0.0057
        thinned_from=44
        keys [[0,-0.0057],[1,-0.0028],[2,0.0003],[3,0.0037],[5,0.011],[9,0.0263],[10,0.0298],[11,0.0331],[12,0.0361],[13,0.0389],[14,0.0414],[15,0.0435],[16,0.0453],[17,0.0467],[18,0.0477],[19,0.0483],[21,0.0485],[22,0.0474],[23,0.046],[24,0.0439],[25,0.0412],[26,0.038],[27,0.0344],[28,0.0305],[33,0.0092],[34,0.0053],[35,0.0016],[36,-0.0016],[37,-0.0044],[38,-0.0066],[39,-0.0081],[40,-0.0088],[41,-0.0088],[42,-0.0077],[43,-0.0057]]
       }
       clavicle_r_ctrl.Location.Y {
        type=MovieSceneScriptingFloatChannel
        default=-0.6423
        thinned_from=44
        keys [[0,-0.6423],[5,-0.6461],[18,-0.6723],[20,-0.6745],[24,-0.6747],[40,-0.6445],[43,-0.6423]]
       }
       clavicle_r_ctrl.Location.Z {
        type=MovieSceneScriptingFloatChannel
        default=-0.0959
        thinned_from=44
        keys [[0,-0.0959],[1,-0.0972],[2,-0.0988],[4,-0.1025],[11,-0.118],[13,-0.1218],[15,-0.125],[17,-0.1275],[18,-0.1283],[19,-0.1289],[20,-0.1292],[21,-0.1292],[22,-0.1289],[23,-0.1282],[24,-0.1271],[25,-0.1256],[26,-0.1237],[28,-0.1193],[34,-0.1042],[35,-0.102],[36,-0.0999],[37,-0.0982],[38,-0.0967],[39,-0.0956],[40,-0.0949],[41,-0.0947],[42,-0.095],[43,-0.0959]]
       }
       clavicle_r_ctrl.Rotation.X {
        type=MovieSceneScriptingFloatChannel
        default=-5.4618
        thinned_from=44
        keys [[0,-5.4618],[3,-5.467],[8,-5.5483],[17,-5.7517],[20,-5.79],[21,-5.7875],[22,-5.7992],[26,-5.7725],[40,-5.4929],[43,-5.4618]]
       }
       clavicle_r_ctrl.Rotation.Y {
        type=MovieSceneScriptingFloatChannel
        default=1.3181
        thinned_from=44
        keys [[0,1.3181],[3,1.3229],[5,1.3195],[6,1.3135],[7,1.3037],[8,1.2894],[9,1.2697],[10,1.2443],[11,1.2123],[13,1.1223],[14,1.0619],[15,0.9942],[18,0.7771],[19,0.71],[20,0.6505],[21,0.5993],[22,0.5661],[23,0.545],[24,0.5365],[25,0.5394],[26,0.5529],[27,0.5758],[28,0.607],[29,0.6458],[30,0.6909],[31,0.7414],[32,0.7964],[33,0.8548],[36,1.0413],[37,1.1017],[38,1.1582],[39,1.209],[40,1.2522],[41,1.286],[42,1.3086],[43,1.3181]]
       }
       clavicle_r_ctrl.Rotation.Z {type=MovieSceneScriptingFloatChannel, default=-10.9427, thinned_from=44, keys [[0,-10.9427],[12,-10.985],[21,-11.1398],[43,-10.9427]]}
       clavicle_r_ctrl.Scale.X {type=MovieSceneScriptingFloatChannel, default=1, const [1,44]}
       clavicle_r_ctrl.Scale.Y {type=MovieSceneScriptingFloatChannel, default=1, const [1,44]}
       clavicle_r_ctrl.Scale.Z {type=MovieSceneScriptingFloatChannel, default=1, const [1,44]}
       neck_01_ctrl.Location.X {
        type=MovieSceneScriptingFloatChannel
        default=0.0182
        thinned_from=44
        keys [[0,0.0182],[1,0.0112],[2,0.0033],[3,-0.0051],[4,-0.014],[8,-0.0514],[9,-0.0605],[10,-0.069],[11,-0.077],[12,-0.0843],[13,-0.0909],[14,-0.0968],[15,-0.1018],[16,-0.106],[17,-0.1093],[18,-0.1116],[19,-0.113],[20,-0.1132],[21,-0.1132],[22,-0.1103],[23,-0.1068],[24,-0.1016],[25,-0.095],[26,-0.0871],[27,-0.0782],[28,-0.0686],[29,-0.0584],[32,-0.0268],[33,-0.0166],[34,-0.007],[35,0.0018],[36,0.0097],[37,0.0163],[38,0.0215],[39,0.025],[40,0.0267],[41,0.0263],[42,0.0235],[43,0.0182]]
       }
       neck_01_ctrl.Location.Y {
        type=MovieSceneScriptingFloatChannel
        default=-1.0072
        thinned_from=44
        keys [[0,-1.0072],[2,-1.01],[5,-1.0214],[16,-1.0902],[19,-1.1029],[20,-1.1055],[22,-1.1075],[24,-1.1047],[27,-1.0926],[37,-1.0262],[40,-1.0122],[42,-1.0076],[43,-1.0072]]
       }
       neck_01_ctrl.Location.Z {
        type=MovieSceneScriptingFloatChannel
        default=-0.1099
        thinned_from=44
        keys [[0,-0.1099],[2,-0.1103],[4,-0.1114],[7,-0.114],[15,-0.1229],[18,-0.1254],[20,-0.1264],[21,-0.1264],[22,-0.1267],[24,-0.1262],[26,-0.1249],[30,-0.1208],[36,-0.1139],[38,-0.1121],[40,-0.1107],[42,-0.11],[43,-0.1099]]
       }
       neck_01_ctrl.Rotation.X {
        type=MovieSceneScriptingFloatChannel
        default=-3.5738
        thinned_from=44
        keys [[0,-3.5738],[1,-3.5543],[2,-3.5688],[3,-3.6141],[4,-3.6874],[5,-3.7857],[6,-3.906],[7,-4.0454],[8,-4.2009],[9,-4.3696],[11,-4.7351],[14,-5.3028],[16,-5.6527],[17,-5.8083],[18,-5.9464],[19,-6.0633],[20,-6.1555],[21,-6.2169],[22,-6.2514],[23,-6.249],[24,-6.214],[25,-6.1496],[26,-6.059],[27,-5.9453],[28,-5.8117],[29,-5.6614],[31,-5.3233],[36,-4.4071],[37,-4.237],[38,-4.0786],[39,-3.9352],[40,-3.8098],[41,-3.7057],[42,-3.626],[43,-3.5738]]
       }
       neck_01_ctrl.Rotation.Y {
        type=MovieSceneScriptingFloatChannel
        default=-5.1142
        thinned_from=44
        keys [[0,-5.1142],[2,-5.1671],[4,-5.1929],[6,-5.1946],[9,-5.1581],[21,-4.8084],[25,-4.7781],[29,-4.7904],[34,-4.8643],[42,-5.0811],[43,-5.1142]]
       }
       neck_01_ctrl.Rotation.Z {type=MovieSceneScriptingFloatChannel, default=-46.0485, thinned_from=44, keys [[0,-46.0485],[43,-46.0485]]}
       neck_01_ctrl.Scale.X {type=MovieSceneScriptingFloatChannel, default=1, const [1,44]}
       neck_01_ctrl.Scale.Y {type=MovieSceneScriptingFloatChannel, default=1, const [1,44]}
       neck_01_ctrl.Scale.Z {type=MovieSceneScriptingFloatChannel, default=1, const [1,44]}
       neck_02_ctrl.Location.X {type=MovieSceneScriptingFloatChannel, default=0, const [0,44]}
       neck_02_ctrl.Location.Y {type=MovieSceneScriptingFloatChannel, default=0, const [0,44]}
       neck_02_ctrl.Location.Z {type=MovieSceneScriptingFloatChannel, default=0, const [0,44]}
       neck_02_ctrl.Rotation.X {type=MovieSceneScriptingFloatChannel, default=0, const [0,44]}
       neck_02_ctrl.Rotation.Y {type=MovieSceneScriptingFloatChannel, default=0, const [0,44]}
       neck_02_ctrl.Rotation.Z {type=MovieSceneScriptingFloatChannel, default=0, const [0,44]}
       neck_02_ctrl.Scale.X {type=MovieSceneScriptingFloatChannel, default=1, const [1,44]}
       neck_02_ctrl.Scale.Y {type=MovieSceneScriptingFloatChannel, default=1, const [1,44]}
       neck_02_ctrl.Scale.Z {type=MovieSceneScriptingFloatChannel, default=1, const [1,44]}
       head_ik_ctrl.Location.X {
        type=MovieSceneScriptingFloatChannel
        default=2.9024
        thinned_from=44
        keys [[0,2.9024],[2,2.8964],[4,2.8759],[7,2.8247],[16,2.6212],[18,2.5867],[21,2.554],[23,2.5561],[25,2.5725],[27,2.6025],[37,2.8302],[40,2.8807],[42,2.899],[43,2.9024]]
       }
       head_ik_ctrl.Location.Y {
        type=MovieSceneScriptingFloatChannel
        default=-2.0147
        thinned_from=44
        keys [[0,-2.0147],[1,-2.0177],[2,-2.0267],[3,-2.0411],[4,-2.0603],[6,-2.111],[9,-2.2083],[14,-2.3877],[16,-2.4512],[18,-2.5027],[20,-2.5373],[22,-2.5501],[23,-2.5466],[24,-2.5366],[25,-2.5206],[27,-2.4737],[29,-2.4115],[36,-2.154],[38,-2.0919],[39,-2.0661],[40,-2.0448],[41,-2.0287],[42,-2.0185],[43,-2.0147]]
       }
       head_ik_ctrl.Location.Z {
        type=MovieSceneScriptingFloatChannel
        default=0.1484
        thinned_from=44
        keys [[0,0.1484],[1,0.1504],[2,0.153],[3,0.156],[8,0.1733],[10,0.1792],[11,0.1816],[12,0.1836],[14,0.1863],[16,0.1876],[19,0.1884],[20,0.1878],[21,0.189],[23,0.1862],[26,0.1842],[28,0.1817],[29,0.1799],[30,0.1778],[32,0.1726],[37,0.1559],[39,0.1504],[41,0.1465],[43,0.1484]]
       }
       head_ik_ctrl.Rotation.X {
        type=MovieSceneScriptingFloatChannel
        default=4.2402
        thinned_from=44
        keys [[0,4.2402],[1,4.1783],[2,4.1049],[3,3.995],[4,3.8616],[5,3.7081],[6,3.538],[7,3.3548],[9,2.9626],[11,2.56],[12,2.3628],[13,2.1841],[14,1.9908],[16,1.6679],[17,1.5373],[18,1.4256],[19,1.3454],[20,1.2706],[21,1.2366],[22,1.2158],[23,1.2588],[24,1.3299],[25,1.4351],[26,1.5694],[27,1.7286],[28,1.9086],[29,2.105],[30,2.3139],[34,3.1888],[35,3.3964],[36,3.5911],[37,3.7688],[39,4.0561],[40,4.155],[41,4.2214],[42,4.2535],[43,4.2506]]
       }
       head_ik_ctrl.Rotation.Y {
        type=MovieSceneScriptingFloatChannel
        default=-7.7167
        thinned_from=44
        keys [[0,-7.7167],[3,-7.5825],[12,-7.0058],[14,-6.8965],[17,-6.7735],[19,-6.7187],[20,-6.7097],[21,-6.6764],[24,-6.7716],[26,-6.8686],[35,-7.5194],[37,-7.6367],[39,-7.717],[41,-7.7626],[43,-7.7155]]
       }
       head_ik_ctrl.Rotation.Z {type=MovieSceneScriptingFloatChannel, default=-25.8805, thinned_from=44, keys [[0,-25.8805],[43,-25.8813]]}
       head_ik_ctrl.Scale.X {type=MovieSceneScriptingFloatChannel, default=1, const [1,44]}
       head_ik_ctrl.Scale.Y {type=MovieSceneScriptingFloatChannel, default=1, const [1,44]}
       head_ik_ctrl.Scale.Z {type=MovieSceneScriptingFloatChannel, default=1, const [1,44]}
       chest_ctrl.Location.X {
        type=MovieSceneScriptingFloatChannel
        default=-0.0914
        thinned_from=44
        keys [[0,-0.0914],[1,-0.1004],[2,-0.1097],[6,-0.148],[7,-0.1572],[8,-0.1659],[9,-0.1742],[10,-0.1816],[11,-0.1883],[12,-0.1941],[13,-0.1965],[14,-0.2031],[15,-0.2047],[16,-0.2087],[17,-0.2094],[19,-0.2098],[20,-0.2092],[21,-0.2092],[22,-0.207],[23,-0.2001],[24,-0.1944],[25,-0.1874],[26,-0.1793],[27,-0.1704],[28,-0.1609],[31,-0.1308],[32,-0.1209],[33,-0.1116],[34,-0.1029],[35,-0.0951],[36,-0.0884],[37,-0.0831],[38,-0.0812],[39,-0.0773],[41,-0.0772],[42,-0.0841],[43,-0.0914]]
       }
       chest_ctrl.Location.Y {
        type=MovieSceneScriptingFloatChannel
        default=0.2574
        thinned_from=44
        keys [[0,0.2574],[1,0.2536],[2,0.2489],[3,0.2435],[4,0.2375],[6,0.2242],[10,0.1959],[12,0.1829],[13,0.1791],[14,0.1716],[15,0.1684],[16,0.1625],[17,0.1601],[18,0.158],[19,0.158],[20,0.1556],[22,0.1555],[23,0.1558],[24,0.1592],[25,0.1637],[26,0.1692],[27,0.1755],[28,0.1825],[30,0.1978],[33,0.2217],[34,0.2293],[35,0.2365],[36,0.243],[37,0.2487],[38,0.2518],[39,0.2571],[40,0.2589],[41,0.2589],[42,0.2598],[43,0.2574]]
       }
       chest_ctrl.Location.Z {
        type=MovieSceneScriptingFloatChannel
        default=0.0202
        thinned_from=44
        keys [[0,0.0202],[3,0.0189],[12,0.0128],[13,0.0123],[14,0.0116],[15,0.0112],[16,0.0107],[18,0.0102],[22,0.01],[23,0.0101],[25,0.011],[35,0.0184],[37,0.0195],[39,0.0203],[42,0.0205],[43,0.0202]]
       }
       chest_ctrl.Rotation.X {
        type=MovieSceneScriptingFloatChannel
        default=0.7181
        thinned_from=44
        keys [[0,0.7181],[1,0.7084],[2,0.6873],[3,0.6561],[4,0.6159],[5,0.5679],[6,0.5135],[7,0.4536],[8,0.3897],[9,0.3229],[11,0.1857],[12,0.1176],[13,0.0558],[14,-0.0124],[15,-0.0682],[16,-0.1272],[17,-0.173],[18,-0.2123],[19,-0.2339],[20,-0.2677],[21,-0.2677],[22,-0.2854],[23,-0.2866],[24,-0.2652],[25,-0.2328],[26,-0.1906],[27,-0.14],[28,-0.0824],[29,-0.0192],[30,0.0485],[31,0.1191],[33,0.264],[34,0.3356],[35,0.405],[36,0.4707],[37,0.5314],[39,0.6328],[40,0.6683],[41,0.6859],[42,0.7148],[43,0.7181]]
       }
       chest_ctrl.Rotation.Y {
        type=MovieSceneScriptingFloatChannel
        default=3.1807
        thinned_from=44
        keys [[0,3.1807],[1,3.2847],[3,3.5164],[8,4.1434],[10,4.372],[11,4.4733],[12,4.5648],[13,4.6337],[14,4.7166],[16,4.8239],[17,4.8556],[18,4.8769],[21,4.8865],[22,4.8492],[23,4.7909],[24,4.7187],[25,4.6279],[26,4.5215],[27,4.4023],[29,4.1375],[32,3.7188],[33,3.5853],[34,3.4598],[35,3.3453],[36,3.2447],[37,3.161],[39,3.0563],[40,3.0419],[41,3.0466],[42,3.1005],[43,3.1807]]
       }
       chest_ctrl.Rotation.Z {type=MovieSceneScriptingFloatChannel, default=19.1784, thinned_from=44, keys [[0,19.1784],[26,19.029],[43,19.1784]]}
       chest_ctrl.Scale.X {type=MovieSceneScriptingFloatChannel, default=1, const [1,44]}
       chest_ctrl.Scale.Y {type=MovieSceneScriptingFloatChannel, default=1, const [1,44]}
       chest_ctrl.Scale.Z {type=MovieSceneScriptingFloatChannel, default=1, const [1,44]}
       chest_tan_ctrl.Location.X {
        type=MovieSceneScriptingFloatChannel
        default=-0.167
        thinned_from=44
        keys [[0,-0.167],[1,-0.1741],[2,-0.182],[3,-0.1903],[8,-0.2354],[9,-0.2441],[10,-0.2522],[11,-0.2598],[12,-0.2666],[14,-0.2782],[16,-0.2866],[17,-0.2893],[18,-0.2912],[19,-0.2925],[21,-0.2924],[22,-0.2898],[23,-0.2859],[24,-0.2808],[25,-0.2743],[26,-0.2666],[27,-0.2581],[28,-0.2487],[32,-0.2084],[33,-0.1987],[34,-0.1895],[35,-0.181],[36,-0.1736],[37,-0.1673],[38,-0.1629],[39,-0.1593],[40,-0.1579],[41,-0.1588],[42,-0.1615],[43,-0.167]]
       }
       chest_tan_ctrl.Location.Y {
        type=MovieSceneScriptingFloatChannel
        default=-0.1496
        thinned_from=44
        keys [[0,-0.1496],[1,-0.1504],[2,-0.1521],[3,-0.1547],[4,-0.1581],[5,-0.1621],[6,-0.1667],[7,-0.1717],[9,-0.1828],[12,-0.2003],[14,-0.2114],[16,-0.2213],[17,-0.2254],[19,-0.2326],[20,-0.2339],[21,-0.2339],[22,-0.2355],[23,-0.2353],[24,-0.2335],[25,-0.2308],[26,-0.2272],[27,-0.223],[28,-0.2181],[30,-0.207],[35,-0.1767],[36,-0.1711],[37,-0.1659],[39,-0.1572],[41,-0.1509],[42,-0.15],[43,-0.1496]]
       }
       chest_tan_ctrl.Location.Z {type=MovieSceneScriptingFloatChannel, default=0.2522, thinned_from=44, keys [[0,0.2522],[24,0.2475],[43,0.2522]]}
       chest_tan_ctrl.Rotation.X {
        type=MovieSceneScriptingFloatChannel
        default=0.9724
        thinned_from=44
        keys [[0,0.9724],[1,0.9833],[2,1.0017],[3,1.0268],[4,1.0577],[5,1.0937],[6,1.1337],[7,1.1771],[13,1.4591],[15,1.543],[16,1.58],[17,1.6127],[18,1.6404],[19,1.6625],[20,1.6783],[21,1.6783],[22,1.6881],[23,1.6803],[24,1.6637],[25,1.6392],[26,1.6079],[27,1.5707],[28,1.5286],[29,1.4826],[34,1.2274],[35,1.1783],[36,1.1321],[37,1.0897],[38,1.0523],[39,1.0207],[40,0.9959],[41,0.979],[42,0.9708],[43,0.9724]]
       }
       chest_tan_ctrl.Rotation.Y {
        type=MovieSceneScriptingFloatChannel
        default=0
        thinned_from=44
        keys [[0,0],[1,-1.733],[3,-1.8445],[8,-2.1459],[10,-2.256],[11,-2.3047],[12,-2.3488],[13,-2.388],[14,-2.422],[15,-2.4507],[16,-2.4738],[17,-2.4911],[18,-2.5024],[19,-2.5074],[21,-2.5059],[23,-2.4581],[24,-2.4232],[25,-2.3794],[26,-2.328],[27,-2.2705],[29,-2.1429],[32,-1.9413],[33,-1.877],[34,-1.8167],[35,-1.7616],[36,-1.7133],[37,-1.6731],[38,-1.6425],[39,-1.6229],[40,-1.6157],[41,-1.6224],[42,-1.6443],[43,-1.683]]
       }
       chest_tan_ctrl.Rotation.Z {type=MovieSceneScriptingFloatChannel, default=0, thinned_from=44, keys [[0,0],[1,-6.3179],[23,-6.2496],[41,-6.3289],[43,-6.3228]]}
       chest_tan_ctrl.Scale.X {type=MovieSceneScriptingFloatChannel, default=1, const [1,44]}
       chest_tan_ctrl.Scale.Y {type=MovieSceneScriptingFloatChannel, default=1, const [1,44]}
       chest_tan_ctrl.Scale.Z {type=MovieSceneScriptingFloatChannel, default=1, const [1,44]}
      ]
     }
     {
      type=MovieSceneControlRigParameterSection
      end_frame=43
      start_frame=0
      channels [
       global_ctrl.Location.X {type=MovieSceneScriptingFloatChannel, default=0}
       global_ctrl.Location.Y {type=MovieSceneScriptingFloatChannel, default=0}
       global_ctrl.Location.Z {type=MovieSceneScriptingFloatChannel, default=0}
       global_ctrl.Rotation.X {type=MovieSceneScriptingFloatChannel, default=0}
       global_ctrl.Rotation.Y {type=MovieSceneScriptingFloatChannel, default=0}
       global_ctrl.Rotation.Z {type=MovieSceneScriptingFloatChannel, default=0}
       root_ctrl.Location.X {type=MovieSceneScriptingFloatChannel, default=0}
       root_ctrl.Location.Y {type=MovieSceneScriptingFloatChannel, default=0}
       root_ctrl.Location.Z {type=MovieSceneScriptingFloatChannel, default=0}
       root_ctrl.Rotation.X {type=MovieSceneScriptingFloatChannel, default=0}
       root_ctrl.Rotation.Y {type=MovieSceneScriptingFloatChannel, default=0}
       root_ctrl.Rotation.Z {type=MovieSceneScriptingFloatChannel, default=0}
       root_ctrl.Scale.X {type=MovieSceneScriptingFloatChannel, default=0}
       root_ctrl.Scale.Y {type=MovieSceneScriptingFloatChannel, default=0}
       root_ctrl.Scale.Z {type=MovieSceneScriptingFloatChannel, default=0}
       body_offset_ctrl.Location.X {type=MovieSceneScriptingFloatChannel, default=0}
       body_offset_ctrl.Location.Y {type=MovieSceneScriptingFloatChannel, default=0}
       body_offset_ctrl.Location.Z {type=MovieSceneScriptingFloatChannel, default=0}
       body_offset_ctrl.Rotation.X {type=MovieSceneScriptingFloatChannel, default=0}
       body_offset_ctrl.Rotation.Y {type=MovieSceneScriptingFloatChannel, default=0}
       body_offset_ctrl.Rotation.Z {type=MovieSceneScriptingFloatChannel, default=0}
       body_offset_ctrl.Scale.X {type=MovieSceneScriptingFloatChannel, default=0}
       body_offset_ctrl.Scale.Y {type=MovieSceneScriptingFloatChannel, default=0}
       body_offset_ctrl.Scale.Z {type=MovieSceneScriptingFloatChannel, default=0}
       thigh_l_fk_ctrl.Location.X {type=MovieSceneScriptingFloatChannel, default=0}
       thigh_l_fk_ctrl.Location.Y {type=MovieSceneScriptingFloatChannel, default=0}
       thigh_l_fk_ctrl.Location.Z {type=MovieSceneScriptingFloatChannel, default=0}
       thigh_l_fk_ctrl.Rotation.X {type=MovieSceneScriptingFloatChannel, default=0}
       thigh_l_fk_ctrl.Rotation.Y {type=MovieSceneScriptingFloatChannel, default=0}
       thigh_l_fk_ctrl.Rotation.Z {type=MovieSceneScriptingFloatChannel, default=0}
       thigh_l_fk_ctrl.Scale.X {type=MovieSceneScriptingFloatChannel, default=0}
       thigh_l_fk_ctrl.Scale.Y {type=MovieSceneScriptingFloatChannel, default=0}
       thigh_l_fk_ctrl.Scale.Z {type=MovieSceneScriptingFloatChannel, default=0}
       calf_l_fk_ctrl.Location.X {type=MovieSceneScriptingFloatChannel, default=0}
       calf_l_fk_ctrl.Location.Y {type=MovieSceneScriptingFloatChannel, default=0}
       calf_l_fk_ctrl.Location.Z {type=MovieSceneScriptingFloatChannel, default=0}
       calf_l_fk_ctrl.Rotation.X {type=MovieSceneScriptingFloatChannel, default=0}
       calf_l_fk_ctrl.Rotation.Y {type=MovieSceneScriptingFloatChannel, default=0}
       calf_l_fk_ctrl.Rotation.Z {type=MovieSceneScriptingFloatChannel, default=0}
       calf_l_fk_ctrl.Scale.X {type=MovieSceneScriptingFloatChannel, default=0}
       calf_l_fk_ctrl.Scale.Y {type=MovieSceneScriptingFloatChannel, default=0}
       calf_l_fk_ctrl.Scale.Z {type=MovieSceneScriptingFloatChannel, default=0}
       foot_l_fk_ctrl.Location.X {type=MovieSceneScriptingFloatChannel, default=0}
       foot_l_fk_ctrl.Location.Y {type=MovieSceneScriptingFloatChannel, default=0}
       foot_l_fk_ctrl.Location.Z {type=MovieSceneScriptingFloatChannel, default=0}
       foot_l_fk_ctrl.Rotation.X {type=MovieSceneScriptingFloatChannel, default=0}
       foot_l_fk_ctrl.Rotation.Y {type=MovieSceneScriptingFloatChannel, default=0}
       foot_l_fk_ctrl.Rotation.Z {type=MovieSceneScriptingFloatChannel, default=0}
       foot_l_fk_ctrl.Scale.X {type=MovieSceneScriptingFloatChannel, default=0}
       foot_l_fk_ctrl.Scale.Y {type=MovieSceneScriptingFloatChannel, default=0}
       foot_l_fk_ctrl.Scale.Z {type=MovieSceneScriptingFloatChannel, default=0}
       ball_l_fk_ctrl.Location.X {type=MovieSceneScriptingFloatChannel, default=0}
       ball_l_fk_ctrl.Location.Y {type=MovieSceneScriptingFloatChannel, default=0}
       ball_l_fk_ctrl.Location.Z {type=MovieSceneScriptingFloatChannel, default=0}
       ball_l_fk_ctrl.Rotation.X {type=MovieSceneScriptingFloatChannel, default=0}
       ball_l_fk_ctrl.Rotation.Y {type=MovieSceneScriptingFloatChannel, default=0}
       ball_l_fk_ctrl.Rotation.Z {type=MovieSceneScriptingFloatChannel, default=0}
       ball_l_fk_ctrl.Scale.X {type=MovieSceneScriptingFloatChannel, default=0}
       ball_l_fk_ctrl.Scale.Y {type=MovieSceneScriptingFloatChannel, default=0}
       ball_l_fk_ctrl.Scale.Z {type=MovieSceneScriptingFloatChannel, default=0}
       foot_l_ik_ctrl.Location.X {type=MovieSceneScriptingFloatChannel, default=0}
       foot_l_ik_ctrl.Location.Y {type=MovieSceneScriptingFloatChannel, default=0}
       foot_l_ik_ctrl.Location.Z {type=MovieSceneScriptingFloatChannel, default=0}
       foot_l_ik_ctrl.Rotation.X {type=MovieSceneScriptingFloatChannel, default=0}
       foot_l_ik_ctrl.Rotation.Y {type=MovieSceneScriptingFloatChannel, default=0}
       foot_l_ik_ctrl.Rotation.Z {type=MovieSceneScriptingFloatChannel, default=0}
       foot_l_ik_ctrl.Scale.X {type=MovieSceneScriptingFloatChannel, default=0}
       foot_l_ik_ctrl.Scale.Y {type=MovieSceneScriptingFloatChannel, default=0}
       foot_l_ik_ctrl.Scale.Z {type=MovieSceneScriptingFloatChannel, default=0}
       ball_l_ik_ctrl.Location.X {type=MovieSceneScriptingFloatChannel, default=0}
       ball_l_ik_ctrl.Location.Y {type=MovieSceneScriptingFloatChannel, default=0}
       ball_l_ik_ctrl.Location.Z {type=MovieSceneScriptingFloatChannel, default=0}
       ball_l_ik_ctrl.Rotation.X {type=MovieSceneScriptingFloatChannel, default=0}
       ball_l_ik_ctrl.Rotation.Y {type=MovieSceneScriptingFloatChannel, default=0}
       ball_l_ik_ctrl.Rotation.Z {type=MovieSceneScriptingFloatChannel, default=0}
       ball_l_ik_ctrl.Scale.X {type=MovieSceneScriptingFloatChannel, default=0}
       ball_l_ik_ctrl.Scale.Y {type=MovieSceneScriptingFloatChannel, default=0}
       ball_l_ik_ctrl.Scale.Z {type=MovieSceneScriptingFloatChannel, default=0}
       foot_bk1_l_ctrl.Location.X {type=MovieSceneScriptingFloatChannel, default=0}
       foot_bk1_l_ctrl.Location.Y {type=MovieSceneScriptingFloatChannel, default=0}
       foot_bk1_l_ctrl.Location.Z {type=MovieSceneScriptingFloatChannel, default=0}
       foot_bk1_l_ctrl.Rotation.X {type=MovieSceneScriptingFloatChannel, default=0}
       foot_bk1_l_ctrl.Rotation.Y {type=MovieSceneScriptingFloatChannel, default=0}
       foot_bk1_l_ctrl.Rotation.Z {type=MovieSceneScriptingFloatChannel, default=0}
       foot_bk1_l_ctrl.Scale.X {type=MovieSceneScriptingFloatChannel, default=0}
       foot_bk1_l_ctrl.Scale.Y {type=MovieSceneScriptingFloatChannel, default=0}
       foot_bk1_l_ctrl.Scale.Z {type=MovieSceneScriptingFloatChannel, default=0}
       foot_roll_l_ctrl.Location.X {type=MovieSceneScriptingFloatChannel, default=0}
       foot_roll_l_ctrl.Location.Y {type=MovieSceneScriptingFloatChannel, default=0}
       foot_roll_l_ctrl.Location.Z {type=MovieSceneScriptingFloatChannel, default=0}
       foot_roll_l_ctrl.Rotation.X {type=MovieSceneScriptingFloatChannel, default=0}
       foot_roll_l_ctrl.Rotation.Y {type=MovieSceneScriptingFloatChannel, default=0}
       foot_roll_l_ctrl.Rotation.Z {type=MovieSceneScriptingFloatChannel, default=0}
       foot_roll_l_ctrl.Scale.X {type=MovieSceneScriptingFloatChannel, default=0}
       foot_roll_l_ctrl.Scale.Y {type=MovieSceneScriptingFloatChannel, default=0}
       foot_roll_l_ctrl.Scale.Z {type=MovieSceneScriptingFloatChannel, default=0}
       roll_blend_l_ctrl {type=MovieSceneScriptingFloatChannel, default=0}
       tip_l_ctrl.Location.X {type=MovieSceneScriptingFloatChannel, default=0}
       tip_l_ctrl.Location.Y {type=MovieSceneScriptingFloatChannel, default=0}
       tip_l_ctrl.Location.Z {type=MovieSceneScriptingFloatChannel, default=0}
       tip_l_ctrl.Rotation.X {type=MovieSceneScriptingFloatChannel, default=0}
       tip_l_ctrl.Rotation.Y {type=MovieSceneScriptingFloatChannel, default=0}
       tip_l_ctrl.Rotation.Z {type=MovieSceneScriptingFloatChannel, default=0}
       tip_l_ctrl.Scale.X {type=MovieSceneScriptingFloatChannel, default=0}
       tip_l_ctrl.Scale.Y {type=MovieSceneScriptingFloatChannel, default=0}
       tip_l_ctrl.Scale.Z {type=MovieSceneScriptingFloatChannel, default=0}
       heel_l_ctrl.Location.X {type=MovieSceneScriptingFloatChannel, default=0}
       heel_l_ctrl.Location.Y {type=MovieSceneScriptingFloatChannel, default=0}
       heel_l_ctrl.Location.Z {type=MovieSceneScriptingFloatChannel, default=0}
       heel_l_ctrl.Rotation.X {type=MovieSceneScriptingFloatChannel, default=0}
       heel_l_ctrl.Rotation.Y {type=MovieSceneScriptingFloatChannel, default=0}
       heel_l_ctrl.Rotation.Z {type=MovieSceneScriptingFloatChannel, default=0}
       heel_l_ctrl.Scale.X {type=MovieSceneScriptingFloatChannel, default=0}
       heel_l_ctrl.Scale.Y {type=MovieSceneScriptingFloatChannel, default=0}
       heel_l_ctrl.Scale.Z {type=MovieSceneScriptingFloatChannel, default=0}
       leg_l_softness {type=MovieSceneScriptingFloatChannel, default=0}
       leg_l_pv_ik_ctrl.Location.X {type=MovieSceneScriptingFloatChannel, default=0}
       leg_l_pv_ik_ctrl.Location.Y {type=MovieSceneScriptingFloatChannel, default=0}
       leg_l_pv_ik_ctrl.Location.Z {type=MovieSceneScriptingFloatChannel, default=0}
       leg_l_pv_ik_ctrl.Rotation.X {type=MovieSceneScriptingFloatChannel, default=0}
       leg_l_pv_ik_ctrl.Rotation.Y {type=MovieSceneScriptingFloatChannel, default=0}
       leg_l_pv_ik_ctrl.Rotation.Z {type=MovieSceneScriptingFloatChannel, default=0}
       leg_l_pv_ik_ctrl.Scale.X {type=MovieSceneScriptingFloatChannel, default=0}
       leg_l_pv_ik_ctrl.Scale.Y {type=MovieSceneScriptingFloatChannel, default=0}
       leg_l_pv_ik_ctrl.Scale.Z {type=MovieSceneScriptingFloatChannel, default=0}
       thigh_r_fk_ctrl.Location.X {type=MovieSceneScriptingFloatChannel, default=0}
       thigh_r_fk_ctrl.Location.Y {type=MovieSceneScriptingFloatChannel, default=0}
       thigh_r_fk_ctrl.Location.Z {type=MovieSceneScriptingFloatChannel, default=0}
       thigh_r_fk_ctrl.Rotation.X {type=MovieSceneScriptingFloatChannel, default=0}
       thigh_r_fk_ctrl.Rotation.Y {type=MovieSceneScriptingFloatChannel, default=0}
       thigh_r_fk_ctrl.Rotation.Z {type=MovieSceneScriptingFloatChannel, default=0}
       thigh_r_fk_ctrl.Scale.X {type=MovieSceneScriptingFloatChannel, default=0}
       thigh_r_fk_ctrl.Scale.Y {type=MovieSceneScriptingFloatChannel, default=0}
       thigh_r_fk_ctrl.Scale.Z {type=MovieSceneScriptingFloatChannel, default=0}
       calf_r_fk_ctrl.Location.X {type=MovieSceneScriptingFloatChannel, default=0}
       calf_r_fk_ctrl.Location.Y {type=MovieSceneScriptingFloatChannel, default=0}
       calf_r_fk_ctrl.Location.Z {type=MovieSceneScriptingFloatChannel, default=0}
       calf_r_fk_ctrl.Rotation.X {type=MovieSceneScriptingFloatChannel, default=0}
       calf_r_fk_ctrl.Rotation.Y {type=MovieSceneScriptingFloatChannel, default=0}
       calf_r_fk_ctrl.Rotation.Z {type=MovieSceneScriptingFloatChannel, default=0}
       calf_r_fk_ctrl.Scale.X {type=MovieSceneScriptingFloatChannel, default=0}
       calf_r_fk_ctrl.Scale.Y {type=MovieSceneScriptingFloatChannel, default=0}
       calf_r_fk_ctrl.Scale.Z {type=MovieSceneScriptingFloatChannel, default=0}
       foot_r_fk_ctrl.Location.X {type=MovieSceneScriptingFloatChannel, default=0}
       foot_r_fk_ctrl.Location.Y {type=MovieSceneScriptingFloatChannel, default=0}
       foot_r_fk_ctrl.Location.Z {type=MovieSceneScriptingFloatChannel, default=0}
       foot_r_fk_ctrl.Rotation.X {type=MovieSceneScriptingFloatChannel, default=0}
       foot_r_fk_ctrl.Rotation.Y {type=MovieSceneScriptingFloatChannel, default=0}
       foot_r_fk_ctrl.Rotation.Z {type=MovieSceneScriptingFloatChannel, default=0}
       foot_r_fk_ctrl.Scale.X {type=MovieSceneScriptingFloatChannel, default=0}
       foot_r_fk_ctrl.Scale.Y {type=MovieSceneScriptingFloatChannel, default=0}
       foot_r_fk_ctrl.Scale.Z {type=MovieSceneScriptingFloatChannel, default=0}
       ball_r_fk_ctrl.Location.X {type=MovieSceneScriptingFloatChannel, default=0}
       ball_r_fk_ctrl.Location.Y {type=MovieSceneScriptingFloatChannel, default=0}
       ball_r_fk_ctrl.Location.Z {type=MovieSceneScriptingFloatChannel, default=0}
       ball_r_fk_ctrl.Rotation.X {type=MovieSceneScriptingFloatChannel, default=0}
       ball_r_fk_ctrl.Rotation.Y {type=MovieSceneScriptingFloatChannel, default=0}
       ball_r_fk_ctrl.Rotation.Z {type=MovieSceneScriptingFloatChannel, default=0}
       ball_r_fk_ctrl.Scale.X {type=MovieSceneScriptingFloatChannel, default=0}
       ball_r_fk_ctrl.Scale.Y {type=MovieSceneScriptingFloatChannel, default=0}
       ball_r_fk_ctrl.Scale.Z {type=MovieSceneScriptingFloatChannel, default=0}
       foot_r_ik_ctrl.Location.X {type=MovieSceneScriptingFloatChannel, default=0}
       foot_r_ik_ctrl.Location.Y {type=MovieSceneScriptingFloatChannel, default=0}
       foot_r_ik_ctrl.Location.Z {type=MovieSceneScriptingFloatChannel, default=0}
       foot_r_ik_ctrl.Rotation.X {type=MovieSceneScriptingFloatChannel, default=0}
       foot_r_ik_ctrl.Rotation.Y {type=MovieSceneScriptingFloatChannel, default=0}
       foot_r_ik_ctrl.Rotation.Z {type=MovieSceneScriptingFloatChannel, default=0}
       foot_r_ik_ctrl.Scale.X {type=MovieSceneScriptingFloatChannel, default=0}
       foot_r_ik_ctrl.Scale.Y {type=MovieSceneScriptingFloatChannel, default=0}
       foot_r_ik_ctrl.Scale.Z {type=MovieSceneScriptingFloatChannel, default=0}
       ball_r_ik_ctrl.Location.X {type=MovieSceneScriptingFloatChannel, default=0}
       ball_r_ik_ctrl.Location.Y {type=MovieSceneScriptingFloatChannel, default=0}
       ball_r_ik_ctrl.Location.Z {type=MovieSceneScriptingFloatChannel, default=0}
       ball_r_ik_ctrl.Rotation.X {type=MovieSceneScriptingFloatChannel, default=0}
       ball_r_ik_ctrl.Rotation.Y {type=MovieSceneScriptingFloatChannel, default=0}
       ball_r_ik_ctrl.Rotation.Z {type=MovieSceneScriptingFloatChannel, default=0}
       ball_r_ik_ctrl.Scale.X {type=MovieSceneScriptingFloatChannel, default=0}
       ball_r_ik_ctrl.Scale.Y {type=MovieSceneScriptingFloatChannel, default=0}
       ball_r_ik_ctrl.Scale.Z {type=MovieSceneScriptingFloatChannel, default=0}
       foot_bk1_r_ctrl.Location.X {type=MovieSceneScriptingFloatChannel, default=0}
       foot_bk1_r_ctrl.Location.Y {type=MovieSceneScriptingFloatChannel, default=0}
       foot_bk1_r_ctrl.Location.Z {type=MovieSceneScriptingFloatChannel, default=0}
       foot_bk1_r_ctrl.Rotation.X {type=MovieSceneScriptingFloatChannel, default=0}
       foot_bk1_r_ctrl.Rotation.Y {type=MovieSceneScriptingFloatChannel, default=0}
       foot_bk1_r_ctrl.Rotation.Z {type=MovieSceneScriptingFloatChannel, default=0}
       foot_bk1_r_ctrl.Scale.X {type=MovieSceneScriptingFloatChannel, default=0}
       foot_bk1_r_ctrl.Scale.Y {type=MovieSceneScriptingFloatChannel, default=0}
       foot_bk1_r_ctrl.Scale.Z {type=MovieSceneScriptingFloatChannel, default=0}
       foot_roll_r_ctrl.Location.X {type=MovieSceneScriptingFloatChannel, default=0}
       foot_roll_r_ctrl.Location.Y {type=MovieSceneScriptingFloatChannel, default=0}
       foot_roll_r_ctrl.Location.Z {type=MovieSceneScriptingFloatChannel, default=0}
       foot_roll_r_ctrl.Rotation.X {type=MovieSceneScriptingFloatChannel, default=0}
       foot_roll_r_ctrl.Rotation.Y {type=MovieSceneScriptingFloatChannel, default=0}
       foot_roll_r_ctrl.Rotation.Z {type=MovieSceneScriptingFloatChannel, default=0}
       foot_roll_r_ctrl.Scale.X {type=MovieSceneScriptingFloatChannel, default=0}
       foot_roll_r_ctrl.Scale.Y {type=MovieSceneScriptingFloatChannel, default=0}
       foot_roll_r_ctrl.Scale.Z {type=MovieSceneScriptingFloatChannel, default=0}
       roll_blend_r_ctrl {type=MovieSceneScriptingFloatChannel, default=0}
       tip_r_ctrl.Location.X {type=MovieSceneScriptingFloatChannel, default=0}
       tip_r_ctrl.Location.Y {type=MovieSceneScriptingFloatChannel, default=0}
       tip_r_ctrl.Location.Z {type=MovieSceneScriptingFloatChannel, default=0}
       tip_r_ctrl.Rotation.X {type=MovieSceneScriptingFloatChannel, default=0}
       tip_r_ctrl.Rotation.Y {type=MovieSceneScriptingFloatChannel, default=0}
       tip_r_ctrl.Rotation.Z {type=MovieSceneScriptingFloatChannel, default=0}
       tip_r_ctrl.Scale.X {type=MovieSceneScriptingFloatChannel, default=0}
       tip_r_ctrl.Scale.Y {type=MovieSceneScriptingFloatChannel, default=0}
       tip_r_ctrl.Scale.Z {type=MovieSceneScriptingFloatChannel, default=0}
       heel_r_ctrl.Location.X {type=MovieSceneScriptingFloatChannel, default=0}
       heel_r_ctrl.Location.Y {type=MovieSceneScriptingFloatChannel, default=0}
       heel_r_ctrl.Location.Z {type=MovieSceneScriptingFloatChannel, default=0}
       heel_r_ctrl.Rotation.X {type=MovieSceneScriptingFloatChannel, default=0}
       heel_r_ctrl.Rotation.Y {type=MovieSceneScriptingFloatChannel, default=0}
       heel_r_ctrl.Rotation.Z {type=MovieSceneScriptingFloatChannel, default=0}
       heel_r_ctrl.Scale.X {type=MovieSceneScriptingFloatChannel, default=0}
       heel_r_ctrl.Scale.Y {type=MovieSceneScriptingFloatChannel, default=0}
       heel_r_ctrl.Scale.Z {type=MovieSceneScriptingFloatChannel, default=0}
       leg_r_softness {type=MovieSceneScriptingFloatChannel, default=0}
       leg_r_pv_ik_ctrl.Location.X {type=MovieSceneScriptingFloatChannel, default=0}
       leg_r_pv_ik_ctrl.Location.Y {type=MovieSceneScriptingFloatChannel, default=0}
       leg_r_pv_ik_ctrl.Location.Z {type=MovieSceneScriptingFloatChannel, default=0}
       leg_r_pv_ik_ctrl.Rotation.X {type=MovieSceneScriptingFloatChannel, default=0}
       leg_r_pv_ik_ctrl.Rotation.Y {type=MovieSceneScriptingFloatChannel, default=0}
       leg_r_pv_ik_ctrl.Rotation.Z {type=MovieSceneScriptingFloatChannel, default=0}
       leg_r_pv_ik_ctrl.Scale.X {type=MovieSceneScriptingFloatChannel, default=0}
       leg_r_pv_ik_ctrl.Scale.Y {type=MovieSceneScriptingFloatChannel, default=0}
       leg_r_pv_ik_ctrl.Scale.Z {type=MovieSceneScriptingFloatChannel, default=0}
       upperarm_l_fk_ctrl.Location.X {type=MovieSceneScriptingFloatChannel, default=0, const [0,2]}
       upperarm_l_fk_ctrl.Location.Y {type=MovieSceneScriptingFloatChannel, default=0, const [0,2]}
       upperarm_l_fk_ctrl.Location.Z {type=MovieSceneScriptingFloatChannel, default=0, const [0,2]}
       upperarm_l_fk_ctrl.Rotation.X {type=MovieSceneScriptingFloatChannel, default=0, const [0.9246,2]}
       upperarm_l_fk_ctrl.Rotation.Y {type=MovieSceneScriptingFloatChannel, default=0, const [34.8229,2]}
       upperarm_l_fk_ctrl.Rotation.Z {type=MovieSceneScriptingFloatChannel, default=0, const [32.8927,2]}
       upperarm_l_fk_ctrl.Scale.X {type=MovieSceneScriptingFloatChannel, default=0, const [0,2]}
       upperarm_l_fk_ctrl.Scale.Y {type=MovieSceneScriptingFloatChannel, default=0, const [0,2]}
       upperarm_l_fk_ctrl.Scale.Z {type=MovieSceneScriptingFloatChannel, default=0, const [0,2]}
       lowerarm_l_fk_ctrl.Location.X {type=MovieSceneScriptingFloatChannel, default=0, const [0,2]}
       lowerarm_l_fk_ctrl.Location.Y {type=MovieSceneScriptingFloatChannel, default=0, const [0,2]}
       lowerarm_l_fk_ctrl.Location.Z {type=MovieSceneScriptingFloatChannel, default=0, const [0,2]}
       lowerarm_l_fk_ctrl.Rotation.X {type=MovieSceneScriptingFloatChannel, default=0, const [0,2]}
       lowerarm_l_fk_ctrl.Rotation.Y {type=MovieSceneScriptingFloatChannel, default=0, const [0,2]}
       lowerarm_l_fk_ctrl.Rotation.Z {type=MovieSceneScriptingFloatChannel, default=0, const [0,2]}
       lowerarm_l_fk_ctrl.Scale.X {type=MovieSceneScriptingFloatChannel, default=0, const [0,2]}
       lowerarm_l_fk_ctrl.Scale.Y {type=MovieSceneScriptingFloatChannel, default=0, const [0,2]}
       lowerarm_l_fk_ctrl.Scale.Z {type=MovieSceneScriptingFloatChannel, default=0, const [0,2]}
       hand_l_fk_ctrl.Location.X {type=MovieSceneScriptingFloatChannel, default=0, const [0,2]}
       hand_l_fk_ctrl.Location.Y {type=MovieSceneScriptingFloatChannel, default=0, const [0,2]}
       hand_l_fk_ctrl.Location.Z {type=MovieSceneScriptingFloatChannel, default=0, const [0,2]}
       hand_l_fk_ctrl.Rotation.X {type=MovieSceneScriptingFloatChannel, default=0, const [0,2]}
       hand_l_fk_ctrl.Rotation.Y {type=MovieSceneScriptingFloatChannel, default=0, const [0,2]}
       hand_l_fk_ctrl.Rotation.Z {type=MovieSceneScriptingFloatChannel, default=0, const [0,2]}
       hand_l_fk_ctrl.Scale.X {type=MovieSceneScriptingFloatChannel, default=0, const [0,2]}
       hand_l_fk_ctrl.Scale.Y {type=MovieSceneScriptingFloatChannel, default=0, const [0,2]}
       hand_l_fk_ctrl.Scale.Z {type=MovieSceneScriptingFloatChannel, default=0, const [0,2]}
       thumb_01_l_ctrl.Location.X {type=MovieSceneScriptingFloatChannel, default=0}
       thumb_01_l_ctrl.Location.Y {type=MovieSceneScriptingFloatChannel, default=0}
       thumb_01_l_ctrl.Location.Z {type=MovieSceneScriptingFloatChannel, default=0}
       thumb_01_l_ctrl.Rotation.X {type=MovieSceneScriptingFloatChannel, default=0}
       thumb_01_l_ctrl.Rotation.Y {type=MovieSceneScriptingFloatChannel, default=0}
       thumb_01_l_ctrl.Rotation.Z {type=MovieSceneScriptingFloatChannel, default=0}
       thumb_01_l_ctrl.Scale.X {type=MovieSceneScriptingFloatChannel, default=0}
       thumb_01_l_ctrl.Scale.Y {type=MovieSceneScriptingFloatChannel, default=0}
       thumb_01_l_ctrl.Scale.Z {type=MovieSceneScriptingFloatChannel, default=0}
       thumb_02_l_ctrl.Location.X {type=MovieSceneScriptingFloatChannel, default=0}
       thumb_02_l_ctrl.Location.Y {type=MovieSceneScriptingFloatChannel, default=0}
       thumb_02_l_ctrl.Location.Z {type=MovieSceneScriptingFloatChannel, default=0}
       thumb_02_l_ctrl.Rotation.X {type=MovieSceneScriptingFloatChannel, default=0}
       thumb_02_l_ctrl.Rotation.Y {type=MovieSceneScriptingFloatChannel, default=0}
       thumb_02_l_ctrl.Rotation.Z {type=MovieSceneScriptingFloatChannel, default=0}
       thumb_02_l_ctrl.Scale.X {type=MovieSceneScriptingFloatChannel, default=0}
       thumb_02_l_ctrl.Scale.Y {type=MovieSceneScriptingFloatChannel, default=0}
       thumb_02_l_ctrl.Scale.Z {type=MovieSceneScriptingFloatChannel, default=0}
       thumb_03_l_ctrl.Location.X {type=MovieSceneScriptingFloatChannel, default=0}
       thumb_03_l_ctrl.Location.Y {type=MovieSceneScriptingFloatChannel, default=0}
       thumb_03_l_ctrl.Location.Z {type=MovieSceneScriptingFloatChannel, default=0}
       thumb_03_l_ctrl.Rotation.X {type=MovieSceneScriptingFloatChannel, default=0}
       thumb_03_l_ctrl.Rotation.Y {type=MovieSceneScriptingFloatChannel, default=0}
       thumb_03_l_ctrl.Rotation.Z {type=MovieSceneScriptingFloatChannel, default=0}
       thumb_03_l_ctrl.Scale.X {type=MovieSceneScriptingFloatChannel, default=0}
       thumb_03_l_ctrl.Scale.Y {type=MovieSceneScriptingFloatChannel, default=0}
       thumb_03_l_ctrl.Scale.Z {type=MovieSceneScriptingFloatChannel, default=0}
       index_metacarpal_l_ctrl.Location.X {type=MovieSceneScriptingFloatChannel, default=0}
       index_metacarpal_l_ctrl.Location.Y {type=MovieSceneScriptingFloatChannel, default=0}
       index_metacarpal_l_ctrl.Location.Z {type=MovieSceneScriptingFloatChannel, default=0}
       index_metacarpal_l_ctrl.Rotation.X {type=MovieSceneScriptingFloatChannel, default=0}
       index_metacarpal_l_ctrl.Rotation.Y {type=MovieSceneScriptingFloatChannel, default=0}
       index_metacarpal_l_ctrl.Rotation.Z {type=MovieSceneScriptingFloatChannel, default=0}
       index_metacarpal_l_ctrl.Scale.X {type=MovieSceneScriptingFloatChannel, default=0}
       index_metacarpal_l_ctrl.Scale.Y {type=MovieSceneScriptingFloatChannel, default=0}
       index_metacarpal_l_ctrl.Scale.Z {type=MovieSceneScriptingFloatChannel, default=0}
       index_01_l_ctrl.Location.X {type=MovieSceneScriptingFloatChannel, default=0}
       index_01_l_ctrl.Location.Y {type=MovieSceneScriptingFloatChannel, default=0}
       index_01_l_ctrl.Location.Z {type=MovieSceneScriptingFloatChannel, default=0}
       index_01_l_ctrl.Rotation.X {type=MovieSceneScriptingFloatChannel, default=0}
       index_01_l_ctrl.Rotation.Y {type=MovieSceneScriptingFloatChannel, default=0}
       index_01_l_ctrl.Rotation.Z {type=MovieSceneScriptingFloatChannel, default=0}
       index_01_l_ctrl.Scale.X {type=MovieSceneScriptingFloatChannel, default=0}
       index_01_l_ctrl.Scale.Y {type=MovieSceneScriptingFloatChannel, default=0}
       index_01_l_ctrl.Scale.Z {type=MovieSceneScriptingFloatChannel, default=0}
       index_02_l_ctrl.Location.X {type=MovieSceneScriptingFloatChannel, default=0}
       index_02_l_ctrl.Location.Y {type=MovieSceneScriptingFloatChannel, default=0}
       index_02_l_ctrl.Location.Z {type=MovieSceneScriptingFloatChannel, default=0}
       index_02_l_ctrl.Rotation.X {type=MovieSceneScriptingFloatChannel, default=0}
       index_02_l_ctrl.Rotation.Y {type=MovieSceneScriptingFloatChannel, default=0}
       index_02_l_ctrl.Rotation.Z {type=MovieSceneScriptingFloatChannel, default=0}
       index_02_l_ctrl.Scale.X {type=MovieSceneScriptingFloatChannel, default=0}
       index_02_l_ctrl.Scale.Y {type=MovieSceneScriptingFloatChannel, default=0}
       index_02_l_ctrl.Scale.Z {type=MovieSceneScriptingFloatChannel, default=0}
       index_03_l_ctrl.Location.X {type=MovieSceneScriptingFloatChannel, default=0}
       index_03_l_ctrl.Location.Y {type=MovieSceneScriptingFloatChannel, default=0}
       index_03_l_ctrl.Location.Z {type=MovieSceneScriptingFloatChannel, default=0}
       index_03_l_ctrl.Rotation.X {type=MovieSceneScriptingFloatChannel, default=0}
       index_03_l_ctrl.Rotation.Y {type=MovieSceneScriptingFloatChannel, default=0}
       index_03_l_ctrl.Rotation.Z {type=MovieSceneScriptingFloatChannel, default=0}
       index_03_l_ctrl.Scale.X {type=MovieSceneScriptingFloatChannel, default=0}
       index_03_l_ctrl.Scale.Y {type=MovieSceneScriptingFloatChannel, default=0}
       index_03_l_ctrl.Scale.Z {type=MovieSceneScriptingFloatChannel, default=0}
       middle_metacarpal_l_ctrl.Location.X {type=MovieSceneScriptingFloatChannel, default=0}
       middle_metacarpal_l_ctrl.Location.Y {type=MovieSceneScriptingFloatChannel, default=0}
       middle_metacarpal_l_ctrl.Location.Z {type=MovieSceneScriptingFloatChannel, default=0}
       middle_metacarpal_l_ctrl.Rotation.X {type=MovieSceneScriptingFloatChannel, default=0}
       middle_metacarpal_l_ctrl.Rotation.Y {type=MovieSceneScriptingFloatChannel, default=0}
       middle_metacarpal_l_ctrl.Rotation.Z {type=MovieSceneScriptingFloatChannel, default=0}
       middle_metacarpal_l_ctrl.Scale.X {type=MovieSceneScriptingFloatChannel, default=0}
       middle_metacarpal_l_ctrl.Scale.Y {type=MovieSceneScriptingFloatChannel, default=0}
       middle_metacarpal_l_ctrl.Scale.Z {type=MovieSceneScriptingFloatChannel, default=0}
       middle_01_l_ctrl.Location.X {type=MovieSceneScriptingFloatChannel, default=0}
       middle_01_l_ctrl.Location.Y {type=MovieSceneScriptingFloatChannel, default=0}
       middle_01_l_ctrl.Location.Z {type=MovieSceneScriptingFloatChannel, default=0}
       middle_01_l_ctrl.Rotation.X {type=MovieSceneScriptingFloatChannel, default=0}
       middle_01_l_ctrl.Rotation.Y {type=MovieSceneScriptingFloatChannel, default=0}
       middle_01_l_ctrl.Rotation.Z {type=MovieSceneScriptingFloatChannel, default=0}
       middle_01_l_ctrl.Scale.X {type=MovieSceneScriptingFloatChannel, default=0}
       middle_01_l_ctrl.Scale.Y {type=MovieSceneScriptingFloatChannel, default=0}
       middle_01_l_ctrl.Scale.Z {type=MovieSceneScriptingFloatChannel, default=0}
       middle_02_l_ctrl.Location.X {type=MovieSceneScriptingFloatChannel, default=0}
       middle_02_l_ctrl.Location.Y {type=MovieSceneScriptingFloatChannel, default=0}
       middle_02_l_ctrl.Location.Z {type=MovieSceneScriptingFloatChannel, default=0}
       middle_02_l_ctrl.Rotation.X {type=MovieSceneScriptingFloatChannel, default=0}
       middle_02_l_ctrl.Rotation.Y {type=MovieSceneScriptingFloatChannel, default=0}
       middle_02_l_ctrl.Rotation.Z {type=MovieSceneScriptingFloatChannel, default=0}
       middle_02_l_ctrl.Scale.X {type=MovieSceneScriptingFloatChannel, default=0}
       middle_02_l_ctrl.Scale.Y {type=MovieSceneScriptingFloatChannel, default=0}
       middle_02_l_ctrl.Scale.Z {type=MovieSceneScriptingFloatChannel, default=0}
       middle_03_l_ctrl.Location.X {type=MovieSceneScriptingFloatChannel, default=0}
       middle_03_l_ctrl.Location.Y {type=MovieSceneScriptingFloatChannel, default=0}
       middle_03_l_ctrl.Location.Z {type=MovieSceneScriptingFloatChannel, default=0}
       middle_03_l_ctrl.Rotation.X {type=MovieSceneScriptingFloatChannel, default=0}
       middle_03_l_ctrl.Rotation.Y {type=MovieSceneScriptingFloatChannel, default=0}
       middle_03_l_ctrl.Rotation.Z {type=MovieSceneScriptingFloatChannel, default=0}
       middle_03_l_ctrl.Scale.X {type=MovieSceneScriptingFloatChannel, default=0}
       middle_03_l_ctrl.Scale.Y {type=MovieSceneScriptingFloatChannel, default=0}
       middle_03_l_ctrl.Scale.Z {type=MovieSceneScriptingFloatChannel, default=0}
       ring_metacarpal_l_ctrl.Location.X {type=MovieSceneScriptingFloatChannel, default=0}
       ring_metacarpal_l_ctrl.Location.Y {type=MovieSceneScriptingFloatChannel, default=0}
       ring_metacarpal_l_ctrl.Location.Z {type=MovieSceneScriptingFloatChannel, default=0}
       ring_metacarpal_l_ctrl.Rotation.X {type=MovieSceneScriptingFloatChannel, default=0}
       ring_metacarpal_l_ctrl.Rotation.Y {type=MovieSceneScriptingFloatChannel, default=0}
       ring_metacarpal_l_ctrl.Rotation.Z {type=MovieSceneScriptingFloatChannel, default=0}
       ring_metacarpal_l_ctrl.Scale.X {type=MovieSceneScriptingFloatChannel, default=0}
       ring_metacarpal_l_ctrl.Scale.Y {type=MovieSceneScriptingFloatChannel, default=0}
       ring_metacarpal_l_ctrl.Scale.Z {type=MovieSceneScriptingFloatChannel, default=0}
       ring_01_l_ctrl.Location.X {type=MovieSceneScriptingFloatChannel, default=0}
       ring_01_l_ctrl.Location.Y {type=MovieSceneScriptingFloatChannel, default=0}
       ring_01_l_ctrl.Location.Z {type=MovieSceneScriptingFloatChannel, default=0}
       ring_01_l_ctrl.Rotation.X {type=MovieSceneScriptingFloatChannel, default=0}
       ring_01_l_ctrl.Rotation.Y {type=MovieSceneScriptingFloatChannel, default=0}
       ring_01_l_ctrl.Rotation.Z {type=MovieSceneScriptingFloatChannel, default=0}
       ring_01_l_ctrl.Scale.X {type=MovieSceneScriptingFloatChannel, default=0}
       ring_01_l_ctrl.Scale.Y {type=MovieSceneScriptingFloatChannel, default=0}
       ring_01_l_ctrl.Scale.Z {type=MovieSceneScriptingFloatChannel, default=0}
       ring_02_l_ctrl.Location.X {type=MovieSceneScriptingFloatChannel, default=0}
       ring_02_l_ctrl.Location.Y {type=MovieSceneScriptingFloatChannel, default=0}
       ring_02_l_ctrl.Location.Z {type=MovieSceneScriptingFloatChannel, default=0}
       ring_02_l_ctrl.Rotation.X {type=MovieSceneScriptingFloatChannel, default=0}
       ring_02_l_ctrl.Rotation.Y {type=MovieSceneScriptingFloatChannel, default=0}
       ring_02_l_ctrl.Rotation.Z {type=MovieSceneScriptingFloatChannel, default=0}
       ring_02_l_ctrl.Scale.X {type=MovieSceneScriptingFloatChannel, default=0}
       ring_02_l_ctrl.Scale.Y {type=MovieSceneScriptingFloatChannel, default=0}
       ring_02_l_ctrl.Scale.Z {type=MovieSceneScriptingFloatChannel, default=0}
       ring_03_l_ctrl.Location.X {type=MovieSceneScriptingFloatChannel, default=0}
       ring_03_l_ctrl.Location.Y {type=MovieSceneScriptingFloatChannel, default=0}
       ring_03_l_ctrl.Location.Z {type=MovieSceneScriptingFloatChannel, default=0}
       ring_03_l_ctrl.Rotation.X {type=MovieSceneScriptingFloatChannel, default=0}
       ring_03_l_ctrl.Rotation.Y {type=MovieSceneScriptingFloatChannel, default=0}
       ring_03_l_ctrl.Rotation.Z {type=MovieSceneScriptingFloatChannel, default=0}
       ring_03_l_ctrl.Scale.X {type=MovieSceneScriptingFloatChannel, default=0}
       ring_03_l_ctrl.Scale.Y {type=MovieSceneScriptingFloatChannel, default=0}
       ring_03_l_ctrl.Scale.Z {type=MovieSceneScriptingFloatChannel, default=0}
       pinky_metacarpal_l_ctrl.Location.X {type=MovieSceneScriptingFloatChannel, default=0}
       pinky_metacarpal_l_ctrl.Location.Y {type=MovieSceneScriptingFloatChannel, default=0}
       pinky_metacarpal_l_ctrl.Location.Z {type=MovieSceneScriptingFloatChannel, default=0}
       pinky_metacarpal_l_ctrl.Rotation.X {type=MovieSceneScriptingFloatChannel, default=0}
       pinky_metacarpal_l_ctrl.Rotation.Y {type=MovieSceneScriptingFloatChannel, default=0}
       pinky_metacarpal_l_ctrl.Rotation.Z {type=MovieSceneScriptingFloatChannel, default=0}
       pinky_metacarpal_l_ctrl.Scale.X {type=MovieSceneScriptingFloatChannel, default=0}
       pinky_metacarpal_l_ctrl.Scale.Y {type=MovieSceneScriptingFloatChannel, default=0}
       pinky_metacarpal_l_ctrl.Scale.Z {type=MovieSceneScriptingFloatChannel, default=0}
       pinky_01_l_ctrl.Location.X {type=MovieSceneScriptingFloatChannel, default=0}
       pinky_01_l_ctrl.Location.Y {type=MovieSceneScriptingFloatChannel, default=0}
       pinky_01_l_ctrl.Location.Z {type=MovieSceneScriptingFloatChannel, default=0}
       pinky_01_l_ctrl.Rotation.X {type=MovieSceneScriptingFloatChannel, default=0}
       pinky_01_l_ctrl.Rotation.Y {type=MovieSceneScriptingFloatChannel, default=0}
       pinky_01_l_ctrl.Rotation.Z {type=MovieSceneScriptingFloatChannel, default=0}
       pinky_01_l_ctrl.Scale.X {type=MovieSceneScriptingFloatChannel, default=0}
       pinky_01_l_ctrl.Scale.Y {type=MovieSceneScriptingFloatChannel, default=0}
       pinky_01_l_ctrl.Scale.Z {type=MovieSceneScriptingFloatChannel, default=0}
       pinky_02_l_ctrl.Location.X {type=MovieSceneScriptingFloatChannel, default=0}
       pinky_02_l_ctrl.Location.Y {type=MovieSceneScriptingFloatChannel, default=0}
       pinky_02_l_ctrl.Location.Z {type=MovieSceneScriptingFloatChannel, default=0}
       pinky_02_l_ctrl.Rotation.X {type=MovieSceneScriptingFloatChannel, default=0}
       pinky_02_l_ctrl.Rotation.Y {type=MovieSceneScriptingFloatChannel, default=0}
       pinky_02_l_ctrl.Rotation.Z {type=MovieSceneScriptingFloatChannel, default=0}
       pinky_02_l_ctrl.Scale.X {type=MovieSceneScriptingFloatChannel, default=0}
       pinky_02_l_ctrl.Scale.Y {type=MovieSceneScriptingFloatChannel, default=0}
       pinky_02_l_ctrl.Scale.Z {type=MovieSceneScriptingFloatChannel, default=0}
       pinky_03_l_ctrl.Location.X {type=MovieSceneScriptingFloatChannel, default=0}
       pinky_03_l_ctrl.Location.Y {type=MovieSceneScriptingFloatChannel, default=0}
       pinky_03_l_ctrl.Location.Z {type=MovieSceneScriptingFloatChannel, default=0}
       pinky_03_l_ctrl.Rotation.X {type=MovieSceneScriptingFloatChannel, default=0}
       pinky_03_l_ctrl.Rotation.Y {type=MovieSceneScriptingFloatChannel, default=0}
       pinky_03_l_ctrl.Rotation.Z {type=MovieSceneScriptingFloatChannel, default=0}
       pinky_03_l_ctrl.Scale.X {type=MovieSceneScriptingFloatChannel, default=0}
       pinky_03_l_ctrl.Scale.Y {type=MovieSceneScriptingFloatChannel, default=0}
       pinky_03_l_ctrl.Scale.Z {type=MovieSceneScriptingFloatChannel, default=0}
       hand_l_ik_ctrl.Location.X {type=MovieSceneScriptingFloatChannel, default=0}
       hand_l_ik_ctrl.Location.Y {type=MovieSceneScriptingFloatChannel, default=0}
       hand_l_ik_ctrl.Location.Z {type=MovieSceneScriptingFloatChannel, default=0}
       hand_l_ik_ctrl.Rotation.X {type=MovieSceneScriptingFloatChannel, default=0}
       hand_l_ik_ctrl.Rotation.Y {type=MovieSceneScriptingFloatChannel, default=0}
       hand_l_ik_ctrl.Rotation.Z {type=MovieSceneScriptingFloatChannel, default=0}
       hand_l_ik_ctrl.Scale.X {type=MovieSceneScriptingFloatChannel, default=0}
       hand_l_ik_ctrl.Scale.Y {type=MovieSceneScriptingFloatChannel, default=0}
       hand_l_ik_ctrl.Scale.Z {type=MovieSceneScriptingFloatChannel, default=0}
       Softness {type=MovieSceneScriptingFloatChannel, default=0}
       hand_l_ikorient_ctrl.Rotation.Roll {type=MovieSceneScriptingFloatChannel, default=0}
       hand_l_ikorient_ctrl.Rotation.Pitch {type=MovieSceneScriptingFloatChannel, default=0}
       hand_l_ikorient_ctrl.Rotation.Yaw {type=MovieSceneScriptingFloatChannel, default=0}
       arm_l_pv_ik_ctrl.Location.X {type=MovieSceneScriptingFloatChannel, default=0}
       arm_l_pv_ik_ctrl.Location.Y {type=MovieSceneScriptingFloatChannel, default=0}
       arm_l_pv_ik_ctrl.Location.Z {type=MovieSceneScriptingFloatChannel, default=0}
       arm_l_pv_ik_ctrl.Rotation.X {type=MovieSceneScriptingFloatChannel, default=0}
       arm_l_pv_ik_ctrl.Rotation.Y {type=MovieSceneScriptingFloatChannel, default=0}
       arm_l_pv_ik_ctrl.Rotation.Z {type=MovieSceneScriptingFloatChannel, default=0}
       arm_l_pv_ik_ctrl.Scale.X {type=MovieSceneScriptingFloatChannel, default=0}
       arm_l_pv_ik_ctrl.Scale.Y {type=MovieSceneScriptingFloatChannel, default=0}
       arm_l_pv_ik_ctrl.Scale.Z {type=MovieSceneScriptingFloatChannel, default=0}
       upperarm_r_fk_ctrl.Location.X {type=MovieSceneScriptingFloatChannel, default=0, const [0,2]}
       upperarm_r_fk_ctrl.Location.Y {type=MovieSceneScriptingFloatChannel, default=0, const [0,2]}
       upperarm_r_fk_ctrl.Location.Z {type=MovieSceneScriptingFloatChannel, default=0, const [0,2]}
       upperarm_r_fk_ctrl.Rotation.X {type=MovieSceneScriptingFloatChannel, default=0, const [0,2]}
       upperarm_r_fk_ctrl.Rotation.Y {type=MovieSceneScriptingFloatChannel, default=0, const [0,2]}
       upperarm_r_fk_ctrl.Rotation.Z {type=MovieSceneScriptingFloatChannel, default=0, const [0,2]}
       upperarm_r_fk_ctrl.Scale.X {type=MovieSceneScriptingFloatChannel, default=0, const [0,2]}
       upperarm_r_fk_ctrl.Scale.Y {type=MovieSceneScriptingFloatChannel, default=0, const [0,2]}
       upperarm_r_fk_ctrl.Scale.Z {type=MovieSceneScriptingFloatChannel, default=0, const [0,2]}
       lowerarm_r_fk_ctrl.Location.X {type=MovieSceneScriptingFloatChannel, default=0, const [0,2]}
       lowerarm_r_fk_ctrl.Location.Y {type=MovieSceneScriptingFloatChannel, default=0, const [0,2]}
       lowerarm_r_fk_ctrl.Location.Z {type=MovieSceneScriptingFloatChannel, default=0, const [0,2]}
       lowerarm_r_fk_ctrl.Rotation.X {type=MovieSceneScriptingFloatChannel, default=0, const [0,2]}
       lowerarm_r_fk_ctrl.Rotation.Y {type=MovieSceneScriptingFloatChannel, default=0, const [0,2]}
       lowerarm_r_fk_ctrl.Rotation.Z {type=MovieSceneScriptingFloatChannel, default=0, const [0,2]}
       lowerarm_r_fk_ctrl.Scale.X {type=MovieSceneScriptingFloatChannel, default=0, const [0,2]}
       lowerarm_r_fk_ctrl.Scale.Y {type=MovieSceneScriptingFloatChannel, default=0, const [0,2]}
       lowerarm_r_fk_ctrl.Scale.Z {type=MovieSceneScriptingFloatChannel, default=0, const [0,2]}
       hand_r_fk_ctrl.Location.X {type=MovieSceneScriptingFloatChannel, default=0, const [0,2]}
       hand_r_fk_ctrl.Location.Y {type=MovieSceneScriptingFloatChannel, default=0, const [0,2]}
       hand_r_fk_ctrl.Location.Z {type=MovieSceneScriptingFloatChannel, default=0, const [0,2]}
       hand_r_fk_ctrl.Rotation.X {type=MovieSceneScriptingFloatChannel, default=0, const [0,2]}
       hand_r_fk_ctrl.Rotation.Y {type=MovieSceneScriptingFloatChannel, default=0, const [0,2]}
       hand_r_fk_ctrl.Rotation.Z {type=MovieSceneScriptingFloatChannel, default=0, const [0,2]}
       hand_r_fk_ctrl.Scale.X {type=MovieSceneScriptingFloatChannel, default=0, const [0,2]}
       hand_r_fk_ctrl.Scale.Y {type=MovieSceneScriptingFloatChannel, default=0, const [0,2]}
       hand_r_fk_ctrl.Scale.Z {type=MovieSceneScriptingFloatChannel, default=0, const [0,2]}
       thumb_01_r_ctrl.Location.X {type=MovieSceneScriptingFloatChannel, default=0}
       thumb_01_r_ctrl.Location.Y {type=MovieSceneScriptingFloatChannel, default=0}
       thumb_01_r_ctrl.Location.Z {type=MovieSceneScriptingFloatChannel, default=0}
       thumb_01_r_ctrl.Rotation.X {type=MovieSceneScriptingFloatChannel, default=0}
       thumb_01_r_ctrl.Rotation.Y {type=MovieSceneScriptingFloatChannel, default=0}
       thumb_01_r_ctrl.Rotation.Z {type=MovieSceneScriptingFloatChannel, default=0}
       thumb_01_r_ctrl.Scale.X {type=MovieSceneScriptingFloatChannel, default=0}
       thumb_01_r_ctrl.Scale.Y {type=MovieSceneScriptingFloatChannel, default=0}
       thumb_01_r_ctrl.Scale.Z {type=MovieSceneScriptingFloatChannel, default=0}
       thumb_02_r_ctrl.Location.X {type=MovieSceneScriptingFloatChannel, default=0}
       thumb_02_r_ctrl.Location.Y {type=MovieSceneScriptingFloatChannel, default=0}
       thumb_02_r_ctrl.Location.Z {type=MovieSceneScriptingFloatChannel, default=0}
       thumb_02_r_ctrl.Rotation.X {type=MovieSceneScriptingFloatChannel, default=0}
       thumb_02_r_ctrl.Rotation.Y {type=MovieSceneScriptingFloatChannel, default=0}
       thumb_02_r_ctrl.Rotation.Z {type=MovieSceneScriptingFloatChannel, default=0}
       thumb_02_r_ctrl.Scale.X {type=MovieSceneScriptingFloatChannel, default=0}
       thumb_02_r_ctrl.Scale.Y {type=MovieSceneScriptingFloatChannel, default=0}
       thumb_02_r_ctrl.Scale.Z {type=MovieSceneScriptingFloatChannel, default=0}
       thumb_03_r_ctrl.Location.X {type=MovieSceneScriptingFloatChannel, default=0}
       thumb_03_r_ctrl.Location.Y {type=MovieSceneScriptingFloatChannel, default=0}
       thumb_03_r_ctrl.Location.Z {type=MovieSceneScriptingFloatChannel, default=0}
       thumb_03_r_ctrl.Rotation.X {type=MovieSceneScriptingFloatChannel, default=0}
       thumb_03_r_ctrl.Rotation.Y {type=MovieSceneScriptingFloatChannel, default=0}
       thumb_03_r_ctrl.Rotation.Z {type=MovieSceneScriptingFloatChannel, default=0}
       thumb_03_r_ctrl.Scale.X {type=MovieSceneScriptingFloatChannel, default=0}
       thumb_03_r_ctrl.Scale.Y {type=MovieSceneScriptingFloatChannel, default=0}
       thumb_03_r_ctrl.Scale.Z {type=MovieSceneScriptingFloatChannel, default=0}
       index_metacarpal_r_ctrl.Location.X {type=MovieSceneScriptingFloatChannel, default=0}
       index_metacarpal_r_ctrl.Location.Y {type=MovieSceneScriptingFloatChannel, default=0}
       index_metacarpal_r_ctrl.Location.Z {type=MovieSceneScriptingFloatChannel, default=0}
       index_metacarpal_r_ctrl.Rotation.X {type=MovieSceneScriptingFloatChannel, default=0}
       index_metacarpal_r_ctrl.Rotation.Y {type=MovieSceneScriptingFloatChannel, default=0}
       index_metacarpal_r_ctrl.Rotation.Z {type=MovieSceneScriptingFloatChannel, default=0}
       index_metacarpal_r_ctrl.Scale.X {type=MovieSceneScriptingFloatChannel, default=0}
       index_metacarpal_r_ctrl.Scale.Y {type=MovieSceneScriptingFloatChannel, default=0}
       index_metacarpal_r_ctrl.Scale.Z {type=MovieSceneScriptingFloatChannel, default=0}
       index_01_r_ctrl.Location.X {type=MovieSceneScriptingFloatChannel, default=0}
       index_01_r_ctrl.Location.Y {type=MovieSceneScriptingFloatChannel, default=0}
       index_01_r_ctrl.Location.Z {type=MovieSceneScriptingFloatChannel, default=0}
       index_01_r_ctrl.Rotation.X {type=MovieSceneScriptingFloatChannel, default=0}
       index_01_r_ctrl.Rotation.Y {type=MovieSceneScriptingFloatChannel, default=0}
       index_01_r_ctrl.Rotation.Z {type=MovieSceneScriptingFloatChannel, default=0}
       index_01_r_ctrl.Scale.X {type=MovieSceneScriptingFloatChannel, default=0}
       index_01_r_ctrl.Scale.Y {type=MovieSceneScriptingFloatChannel, default=0}
       index_01_r_ctrl.Scale.Z {type=MovieSceneScriptingFloatChannel, default=0}
       index_02_r_ctrl.Location.X {type=MovieSceneScriptingFloatChannel, default=0}
       index_02_r_ctrl.Location.Y {type=MovieSceneScriptingFloatChannel, default=0}
       index_02_r_ctrl.Location.Z {type=MovieSceneScriptingFloatChannel, default=0}
       index_02_r_ctrl.Rotation.X {type=MovieSceneScriptingFloatChannel, default=0}
       index_02_r_ctrl.Rotation.Y {type=MovieSceneScriptingFloatChannel, default=0}
       index_02_r_ctrl.Rotation.Z {type=MovieSceneScriptingFloatChannel, default=0}
       index_02_r_ctrl.Scale.X {type=MovieSceneScriptingFloatChannel, default=0}
       index_02_r_ctrl.Scale.Y {type=MovieSceneScriptingFloatChannel, default=0}
       index_02_r_ctrl.Scale.Z {type=MovieSceneScriptingFloatChannel, default=0}
       index_03_r_ctrl.Location.X {type=MovieSceneScriptingFloatChannel, default=0}
       index_03_r_ctrl.Location.Y {type=MovieSceneScriptingFloatChannel, default=0}
       index_03_r_ctrl.Location.Z {type=MovieSceneScriptingFloatChannel, default=0}
       index_03_r_ctrl.Rotation.X {type=MovieSceneScriptingFloatChannel, default=0}
       index_03_r_ctrl.Rotation.Y {type=MovieSceneScriptingFloatChannel, default=0}
       index_03_r_ctrl.Rotation.Z {type=MovieSceneScriptingFloatChannel, default=0}
       index_03_r_ctrl.Scale.X {type=MovieSceneScriptingFloatChannel, default=0}
       index_03_r_ctrl.Scale.Y {type=MovieSceneScriptingFloatChannel, default=0}
       index_03_r_ctrl.Scale.Z {type=MovieSceneScriptingFloatChannel, default=0}
       middle_metacarpal_r_ctrl.Location.X {type=MovieSceneScriptingFloatChannel, default=0}
       middle_metacarpal_r_ctrl.Location.Y {type=MovieSceneScriptingFloatChannel, default=0}
       middle_metacarpal_r_ctrl.Location.Z {type=MovieSceneScriptingFloatChannel, default=0}
       middle_metacarpal_r_ctrl.Rotation.X {type=MovieSceneScriptingFloatChannel, default=0}
       middle_metacarpal_r_ctrl.Rotation.Y {type=MovieSceneScriptingFloatChannel, default=0}
       middle_metacarpal_r_ctrl.Rotation.Z {type=MovieSceneScriptingFloatChannel, default=0}
       middle_metacarpal_r_ctrl.Scale.X {type=MovieSceneScriptingFloatChannel, default=0}
       middle_metacarpal_r_ctrl.Scale.Y {type=MovieSceneScriptingFloatChannel, default=0}
       middle_metacarpal_r_ctrl.Scale.Z {type=MovieSceneScriptingFloatChannel, default=0}
       middle_01_r_ctrl.Location.X {type=MovieSceneScriptingFloatChannel, default=0}
       middle_01_r_ctrl.Location.Y {type=MovieSceneScriptingFloatChannel, default=0}
       middle_01_r_ctrl.Location.Z {type=MovieSceneScriptingFloatChannel, default=0}
       middle_01_r_ctrl.Rotation.X {type=MovieSceneScriptingFloatChannel, default=0}
       middle_01_r_ctrl.Rotation.Y {type=MovieSceneScriptingFloatChannel, default=0}
       middle_01_r_ctrl.Rotation.Z {type=MovieSceneScriptingFloatChannel, default=0}
       middle_01_r_ctrl.Scale.X {type=MovieSceneScriptingFloatChannel, default=0}
       middle_01_r_ctrl.Scale.Y {type=MovieSceneScriptingFloatChannel, default=0}
       middle_01_r_ctrl.Scale.Z {type=MovieSceneScriptingFloatChannel, default=0}
       middle_02_r_ctrl.Location.X {type=MovieSceneScriptingFloatChannel, default=0}
       middle_02_r_ctrl.Location.Y {type=MovieSceneScriptingFloatChannel, default=0}
       middle_02_r_ctrl.Location.Z {type=MovieSceneScriptingFloatChannel, default=0}
       middle_02_r_ctrl.Rotation.X {type=MovieSceneScriptingFloatChannel, default=0}
       middle_02_r_ctrl.Rotation.Y {type=MovieSceneScriptingFloatChannel, default=0}
       middle_02_r_ctrl.Rotation.Z {type=MovieSceneScriptingFloatChannel, default=0}
       middle_02_r_ctrl.Scale.X {type=MovieSceneScriptingFloatChannel, default=0}
       middle_02_r_ctrl.Scale.Y {type=MovieSceneScriptingFloatChannel, default=0}
       middle_02_r_ctrl.Scale.Z {type=MovieSceneScriptingFloatChannel, default=0}
       middle_03_r_ctrl.Location.X {type=MovieSceneScriptingFloatChannel, default=0}
       middle_03_r_ctrl.Location.Y {type=MovieSceneScriptingFloatChannel, default=0}
       middle_03_r_ctrl.Location.Z {type=MovieSceneScriptingFloatChannel, default=0}
       middle_03_r_ctrl.Rotation.X {type=MovieSceneScriptingFloatChannel, default=0}
       middle_03_r_ctrl.Rotation.Y {type=MovieSceneScriptingFloatChannel, default=0}
       middle_03_r_ctrl.Rotation.Z {type=MovieSceneScriptingFloatChannel, default=0}
       middle_03_r_ctrl.Scale.X {type=MovieSceneScriptingFloatChannel, default=0}
       middle_03_r_ctrl.Scale.Y {type=MovieSceneScriptingFloatChannel, default=0}
       middle_03_r_ctrl.Scale.Z {type=MovieSceneScriptingFloatChannel, default=0}
       ring_metacarpal_r_ctrl.Location.X {type=MovieSceneScriptingFloatChannel, default=0}
       ring_metacarpal_r_ctrl.Location.Y {type=MovieSceneScriptingFloatChannel, default=0}
       ring_metacarpal_r_ctrl.Location.Z {type=MovieSceneScriptingFloatChannel, default=0}
       ring_metacarpal_r_ctrl.Rotation.X {type=MovieSceneScriptingFloatChannel, default=0}
       ring_metacarpal_r_ctrl.Rotation.Y {type=MovieSceneScriptingFloatChannel, default=0}
       ring_metacarpal_r_ctrl.Rotation.Z {type=MovieSceneScriptingFloatChannel, default=0}
       ring_metacarpal_r_ctrl.Scale.X {type=MovieSceneScriptingFloatChannel, default=0}
       ring_metacarpal_r_ctrl.Scale.Y {type=MovieSceneScriptingFloatChannel, default=0}
       ring_metacarpal_r_ctrl.Scale.Z {type=MovieSceneScriptingFloatChannel, default=0}
       ring_01_r_ctrl.Location.X {type=MovieSceneScriptingFloatChannel, default=0}
       ring_01_r_ctrl.Location.Y {type=MovieSceneScriptingFloatChannel, default=0}
       ring_01_r_ctrl.Location.Z {type=MovieSceneScriptingFloatChannel, default=0}
       ring_01_r_ctrl.Rotation.X {type=MovieSceneScriptingFloatChannel, default=0}
       ring_01_r_ctrl.Rotation.Y {type=MovieSceneScriptingFloatChannel, default=0}
       ring_01_r_ctrl.Rotation.Z {type=MovieSceneScriptingFloatChannel, default=0}
       ring_01_r_ctrl.Scale.X {type=MovieSceneScriptingFloatChannel, default=0}
       ring_01_r_ctrl.Scale.Y {type=MovieSceneScriptingFloatChannel, default=0}
       ring_01_r_ctrl.Scale.Z {type=MovieSceneScriptingFloatChannel, default=0}
       ring_02_r_ctrl.Location.X {type=MovieSceneScriptingFloatChannel, default=0}
       ring_02_r_ctrl.Location.Y {type=MovieSceneScriptingFloatChannel, default=0}
       ring_02_r_ctrl.Location.Z {type=MovieSceneScriptingFloatChannel, default=0}
       ring_02_r_ctrl.Rotation.X {type=MovieSceneScriptingFloatChannel, default=0}
       ring_02_r_ctrl.Rotation.Y {type=MovieSceneScriptingFloatChannel, default=0}
       ring_02_r_ctrl.Rotation.Z {type=MovieSceneScriptingFloatChannel, default=0}
       ring_02_r_ctrl.Scale.X {type=MovieSceneScriptingFloatChannel, default=0}
       ring_02_r_ctrl.Scale.Y {type=MovieSceneScriptingFloatChannel, default=0}
       ring_02_r_ctrl.Scale.Z {type=MovieSceneScriptingFloatChannel, default=0}
       ring_03_r_ctrl.Location.X {type=MovieSceneScriptingFloatChannel, default=0}
       ring_03_r_ctrl.Location.Y {type=MovieSceneScriptingFloatChannel, default=0}
       ring_03_r_ctrl.Location.Z {type=MovieSceneScriptingFloatChannel, default=0}
       ring_03_r_ctrl.Rotation.X {type=MovieSceneScriptingFloatChannel, default=0}
       ring_03_r_ctrl.Rotation.Y {type=MovieSceneScriptingFloatChannel, default=0}
       ring_03_r_ctrl.Rotation.Z {type=MovieSceneScriptingFloatChannel, default=0}
       ring_03_r_ctrl.Scale.X {type=MovieSceneScriptingFloatChannel, default=0}
       ring_03_r_ctrl.Scale.Y {type=MovieSceneScriptingFloatChannel, default=0}
       ring_03_r_ctrl.Scale.Z {type=MovieSceneScriptingFloatChannel, default=0}
       pinky_metacarpal_r_ctrl.Location.X {type=MovieSceneScriptingFloatChannel, default=0}
       pinky_metacarpal_r_ctrl.Location.Y {type=MovieSceneScriptingFloatChannel, default=0}
       pinky_metacarpal_r_ctrl.Location.Z {type=MovieSceneScriptingFloatChannel, default=0}
       pinky_metacarpal_r_ctrl.Rotation.X {type=MovieSceneScriptingFloatChannel, default=0}
       pinky_metacarpal_r_ctrl.Rotation.Y {type=MovieSceneScriptingFloatChannel, default=0}
       pinky_metacarpal_r_ctrl.Rotation.Z {type=MovieSceneScriptingFloatChannel, default=0}
       pinky_metacarpal_r_ctrl.Scale.X {type=MovieSceneScriptingFloatChannel, default=0}
       pinky_metacarpal_r_ctrl.Scale.Y {type=MovieSceneScriptingFloatChannel, default=0}
       pinky_metacarpal_r_ctrl.Scale.Z {type=MovieSceneScriptingFloatChannel, default=0}
       pinky_01_r_ctrl.Location.X {type=MovieSceneScriptingFloatChannel, default=0}
       pinky_01_r_ctrl.Location.Y {type=MovieSceneScriptingFloatChannel, default=0}
       pinky_01_r_ctrl.Location.Z {type=MovieSceneScriptingFloatChannel, default=0}
       pinky_01_r_ctrl.Rotation.X {type=MovieSceneScriptingFloatChannel, default=0}
       pinky_01_r_ctrl.Rotation.Y {type=MovieSceneScriptingFloatChannel, default=0}
       pinky_01_r_ctrl.Rotation.Z {type=MovieSceneScriptingFloatChannel, default=0}
       pinky_01_r_ctrl.Scale.X {type=MovieSceneScriptingFloatChannel, default=0}
       pinky_01_r_ctrl.Scale.Y {type=MovieSceneScriptingFloatChannel, default=0}
       pinky_01_r_ctrl.Scale.Z {type=MovieSceneScriptingFloatChannel, default=0}
       pinky_02_r_ctrl.Location.X {type=MovieSceneScriptingFloatChannel, default=0}
       pinky_02_r_ctrl.Location.Y {type=MovieSceneScriptingFloatChannel, default=0}
       pinky_02_r_ctrl.Location.Z {type=MovieSceneScriptingFloatChannel, default=0}
       pinky_02_r_ctrl.Rotation.X {type=MovieSceneScriptingFloatChannel, default=0}
       pinky_02_r_ctrl.Rotation.Y {type=MovieSceneScriptingFloatChannel, default=0}
       pinky_02_r_ctrl.Rotation.Z {type=MovieSceneScriptingFloatChannel, default=0}
       pinky_02_r_ctrl.Scale.X {type=MovieSceneScriptingFloatChannel, default=0}
       pinky_02_r_ctrl.Scale.Y {type=MovieSceneScriptingFloatChannel, default=0}
       pinky_02_r_ctrl.Scale.Z {type=MovieSceneScriptingFloatChannel, default=0}
       pinky_03_r_ctrl.Location.X {type=MovieSceneScriptingFloatChannel, default=0}
       pinky_03_r_ctrl.Location.Y {type=MovieSceneScriptingFloatChannel, default=0}
       pinky_03_r_ctrl.Location.Z {type=MovieSceneScriptingFloatChannel, default=0}
       pinky_03_r_ctrl.Rotation.X {type=MovieSceneScriptingFloatChannel, default=0}
       pinky_03_r_ctrl.Rotation.Y {type=MovieSceneScriptingFloatChannel, default=0}
       pinky_03_r_ctrl.Rotation.Z {type=MovieSceneScriptingFloatChannel, default=0}
       pinky_03_r_ctrl.Scale.X {type=MovieSceneScriptingFloatChannel, default=0}
       pinky_03_r_ctrl.Scale.Y {type=MovieSceneScriptingFloatChannel, default=0}
       pinky_03_r_ctrl.Scale.Z {type=MovieSceneScriptingFloatChannel, default=0}
       hand_r_ik_ctrl.Location.X {type=MovieSceneScriptingFloatChannel, default=0}
       hand_r_ik_ctrl.Location.Y {type=MovieSceneScriptingFloatChannel, default=0}
       hand_r_ik_ctrl.Location.Z {type=MovieSceneScriptingFloatChannel, default=0}
       hand_r_ik_ctrl.Rotation.X {type=MovieSceneScriptingFloatChannel, default=0}
       hand_r_ik_ctrl.Rotation.Y {type=MovieSceneScriptingFloatChannel, default=0}
       hand_r_ik_ctrl.Rotation.Z {type=MovieSceneScriptingFloatChannel, default=0}
       hand_r_ik_ctrl.Scale.X {type=MovieSceneScriptingFloatChannel, default=0}
       hand_r_ik_ctrl.Scale.Y {type=MovieSceneScriptingFloatChannel, default=0}
       hand_r_ik_ctrl.Scale.Z {type=MovieSceneScriptingFloatChannel, default=0}
       Softness_2 {type=MovieSceneScriptingFloatChannel, default=0}
       hand_r_ikorient_ctrl.Rotation.Roll {type=MovieSceneScriptingFloatChannel, default=0}
       hand_r_ikorient_ctrl.Rotation.Pitch {type=MovieSceneScriptingFloatChannel, default=0}
       hand_r_ikorient_ctrl.Rotation.Yaw {type=MovieSceneScriptingFloatChannel, default=0}
       arm_r_pv_ik_ctrl.Location.X {type=MovieSceneScriptingFloatChannel, default=0}
       arm_r_pv_ik_ctrl.Location.Y {type=MovieSceneScriptingFloatChannel, default=0}
       arm_r_pv_ik_ctrl.Location.Z {type=MovieSceneScriptingFloatChannel, default=0}
       arm_r_pv_ik_ctrl.Rotation.X {type=MovieSceneScriptingFloatChannel, default=0}
       arm_r_pv_ik_ctrl.Rotation.Y {type=MovieSceneScriptingFloatChannel, default=0}
       arm_r_pv_ik_ctrl.Rotation.Z {type=MovieSceneScriptingFloatChannel, default=0}
       arm_r_pv_ik_ctrl.Scale.X {type=MovieSceneScriptingFloatChannel, default=0}
       arm_r_pv_ik_ctrl.Scale.Y {type=MovieSceneScriptingFloatChannel, default=0}
       arm_r_pv_ik_ctrl.Scale.Z {type=MovieSceneScriptingFloatChannel, default=0}
       body_ctrl.Location.X {type=MovieSceneScriptingFloatChannel, default=0}
       body_ctrl.Location.Y {type=MovieSceneScriptingFloatChannel, default=0}
       body_ctrl.Location.Z {type=MovieSceneScriptingFloatChannel, default=0}
       body_ctrl.Rotation.X {type=MovieSceneScriptingFloatChannel, default=0}
       body_ctrl.Rotation.Y {type=MovieSceneScriptingFloatChannel, default=0}
       body_ctrl.Rotation.Z {type=MovieSceneScriptingFloatChannel, default=0}
       body_ctrl.Scale.X {type=MovieSceneScriptingFloatChannel, default=0}
       body_ctrl.Scale.Y {type=MovieSceneScriptingFloatChannel, default=0}
       body_ctrl.Scale.Z {type=MovieSceneScriptingFloatChannel, default=0}
       hips_ctrl.Location.X {type=MovieSceneScriptingFloatChannel, default=0}
       hips_ctrl.Location.Y {type=MovieSceneScriptingFloatChannel, default=0}
       hips_ctrl.Location.Z {type=MovieSceneScriptingFloatChannel, default=0}
       hips_ctrl.Rotation.X {type=MovieSceneScriptingFloatChannel, default=0}
       hips_ctrl.Rotation.Y {type=MovieSceneScriptingFloatChannel, default=0}
       hips_ctrl.Rotation.Z {type=MovieSceneScriptingFloatChannel, default=0}
       hips_ctrl.Scale.X {type=MovieSceneScriptingFloatChannel, default=0}
       hips_ctrl.Scale.Y {type=MovieSceneScriptingFloatChannel, default=0}
       hips_ctrl.Scale.Z {type=MovieSceneScriptingFloatChannel, default=0}
       hips_tan_ctrl.Location.X {type=MovieSceneScriptingFloatChannel, default=0}
       hips_tan_ctrl.Location.Y {type=MovieSceneScriptingFloatChannel, default=0}
       hips_tan_ctrl.Location.Z {type=MovieSceneScriptingFloatChannel, default=0}
       hips_tan_ctrl.Rotation.X {type=MovieSceneScriptingFloatChannel, default=0}
       hips_tan_ctrl.Rotation.Y {type=MovieSceneScriptingFloatChannel, default=0}
       hips_tan_ctrl.Rotation.Z {type=MovieSceneScriptingFloatChannel, default=0}
       hips_tan_ctrl.Scale.X {type=MovieSceneScriptingFloatChannel, default=0}
       hips_tan_ctrl.Scale.Y {type=MovieSceneScriptingFloatChannel, default=0}
       hips_tan_ctrl.Scale.Z {type=MovieSceneScriptingFloatChannel, default=0}
       spine_01_ctrl.Location.X {type=MovieSceneScriptingFloatChannel, default=0}
       spine_01_ctrl.Location.Y {type=MovieSceneScriptingFloatChannel, default=0}
       spine_01_ctrl.Location.Z {type=MovieSceneScriptingFloatChannel, default=0}
       spine_01_ctrl.Rotation.X {type=MovieSceneScriptingFloatChannel, default=0}
       spine_01_ctrl.Rotation.Y {type=MovieSceneScriptingFloatChannel, default=0}
       spine_01_ctrl.Rotation.Z {type=MovieSceneScriptingFloatChannel, default=0}
       spine_01_ctrl.Scale.X {type=MovieSceneScriptingFloatChannel, default=0}
       spine_01_ctrl.Scale.Y {type=MovieSceneScriptingFloatChannel, default=0}
       spine_01_ctrl.Scale.Z {type=MovieSceneScriptingFloatChannel, default=0}
       spine_02_ctrl.Location.X {type=MovieSceneScriptingFloatChannel, default=0}
       spine_02_ctrl.Location.Y {type=MovieSceneScriptingFloatChannel, default=0}
       spine_02_ctrl.Location.Z {type=MovieSceneScriptingFloatChannel, default=0}
       spine_02_ctrl.Rotation.X {type=MovieSceneScriptingFloatChannel, default=0}
       spine_02_ctrl.Rotation.Y {type=MovieSceneScriptingFloatChannel, default=0}
       spine_02_ctrl.Rotation.Z {type=MovieSceneScriptingFloatChannel, default=0}
       spine_02_ctrl.Scale.X {type=MovieSceneScriptingFloatChannel, default=0}
       spine_02_ctrl.Scale.Y {type=MovieSceneScriptingFloatChannel, default=0}
       spine_02_ctrl.Scale.Z {type=MovieSceneScriptingFloatChannel, default=0}
       spine_03_ctrl.Location.X {type=MovieSceneScriptingFloatChannel, default=0}
       spine_03_ctrl.Location.Y {type=MovieSceneScriptingFloatChannel, default=0}
       spine_03_ctrl.Location.Z {type=MovieSceneScriptingFloatChannel, default=0}
       spine_03_ctrl.Rotation.X {type=MovieSceneScriptingFloatChannel, default=0}
       spine_03_ctrl.Rotation.Y {type=MovieSceneScriptingFloatChannel, default=0}
       spine_03_ctrl.Rotation.Z {type=MovieSceneScriptingFloatChannel, default=0}
       spine_03_ctrl.Scale.X {type=MovieSceneScriptingFloatChannel, default=0}
       spine_03_ctrl.Scale.Y {type=MovieSceneScriptingFloatChannel, default=0}
       spine_03_ctrl.Scale.Z {type=MovieSceneScriptingFloatChannel, default=0}
       spine_04_ctrl.Location.X {type=MovieSceneScriptingFloatChannel, default=0}
       spine_04_ctrl.Location.Y {type=MovieSceneScriptingFloatChannel, default=0}
       spine_04_ctrl.Location.Z {type=MovieSceneScriptingFloatChannel, default=0}
       spine_04_ctrl.Rotation.X {type=MovieSceneScriptingFloatChannel, default=0}
       spine_04_ctrl.Rotation.Y {type=MovieSceneScriptingFloatChannel, default=0}
       spine_04_ctrl.Rotation.Z {type=MovieSceneScriptingFloatChannel, default=0}
       spine_04_ctrl.Scale.X {type=MovieSceneScriptingFloatChannel, default=0}
       spine_04_ctrl.Scale.Y {type=MovieSceneScriptingFloatChannel, default=0}
       spine_04_ctrl.Scale.Z {type=MovieSceneScriptingFloatChannel, default=0}
       head_ctrl.Location.X {type=MovieSceneScriptingFloatChannel, default=0}
       head_ctrl.Location.Y {type=MovieSceneScriptingFloatChannel, default=0}
       head_ctrl.Location.Z {type=MovieSceneScriptingFloatChannel, default=0}
       head_ctrl.Rotation.X {type=MovieSceneScriptingFloatChannel, default=0}
       head_ctrl.Rotation.Y {type=MovieSceneScriptingFloatChannel, default=0}
       head_ctrl.Rotation.Z {type=MovieSceneScriptingFloatChannel, default=0}
       head_ctrl.Scale.X {type=MovieSceneScriptingFloatChannel, default=0}
       head_ctrl.Scale.Y {type=MovieSceneScriptingFloatChannel, default=0}
       head_ctrl.Scale.Z {type=MovieSceneScriptingFloatChannel, default=0}
       clavicle_l_ctrl.Location.X {type=MovieSceneScriptingFloatChannel, default=0}
       clavicle_l_ctrl.Location.Y {type=MovieSceneScriptingFloatChannel, default=0}
       clavicle_l_ctrl.Location.Z {type=MovieSceneScriptingFloatChannel, default=0}
       clavicle_l_ctrl.Rotation.X {type=MovieSceneScriptingFloatChannel, default=0}
       clavicle_l_ctrl.Rotation.Y {type=MovieSceneScriptingFloatChannel, default=0}
       clavicle_l_ctrl.Rotation.Z {type=MovieSceneScriptingFloatChannel, default=0}
       clavicle_l_ctrl.Scale.X {type=MovieSceneScriptingFloatChannel, default=0}
       clavicle_l_ctrl.Scale.Y {type=MovieSceneScriptingFloatChannel, default=0}
       clavicle_l_ctrl.Scale.Z {type=MovieSceneScriptingFloatChannel, default=0}
       clavicle_r_ctrl.Location.X {type=MovieSceneScriptingFloatChannel, default=0}
       clavicle_r_ctrl.Location.Y {type=MovieSceneScriptingFloatChannel, default=0}
       clavicle_r_ctrl.Location.Z {type=MovieSceneScriptingFloatChannel, default=0}
       clavicle_r_ctrl.Rotation.X {type=MovieSceneScriptingFloatChannel, default=0}
       clavicle_r_ctrl.Rotation.Y {type=MovieSceneScriptingFloatChannel, default=0}
       clavicle_r_ctrl.Rotation.Z {type=MovieSceneScriptingFloatChannel, default=0}
       clavicle_r_ctrl.Scale.X {type=MovieSceneScriptingFloatChannel, default=0}
       clavicle_r_ctrl.Scale.Y {type=MovieSceneScriptingFloatChannel, default=0}
       clavicle_r_ctrl.Scale.Z {type=MovieSceneScriptingFloatChannel, default=0}
       neck_01_ctrl.Location.X {type=MovieSceneScriptingFloatChannel, default=0}
       neck_01_ctrl.Location.Y {type=MovieSceneScriptingFloatChannel, default=0}
       neck_01_ctrl.Location.Z {type=MovieSceneScriptingFloatChannel, default=0}
       neck_01_ctrl.Rotation.X {type=MovieSceneScriptingFloatChannel, default=0}
       neck_01_ctrl.Rotation.Y {type=MovieSceneScriptingFloatChannel, default=0}
       neck_01_ctrl.Rotation.Z {type=MovieSceneScriptingFloatChannel, default=0}
       neck_01_ctrl.Scale.X {type=MovieSceneScriptingFloatChannel, default=0}
       neck_01_ctrl.Scale.Y {type=MovieSceneScriptingFloatChannel, default=0}
       neck_01_ctrl.Scale.Z {type=MovieSceneScriptingFloatChannel, default=0}
       neck_02_ctrl.Location.X {type=MovieSceneScriptingFloatChannel, default=0}
       neck_02_ctrl.Location.Y {type=MovieSceneScriptingFloatChannel, default=0}
       neck_02_ctrl.Location.Z {type=MovieSceneScriptingFloatChannel, default=0}
       neck_02_ctrl.Rotation.X {type=MovieSceneScriptingFloatChannel, default=0}
       neck_02_ctrl.Rotation.Y {type=MovieSceneScriptingFloatChannel, default=0}
       neck_02_ctrl.Rotation.Z {type=MovieSceneScriptingFloatChannel, default=0}
       neck_02_ctrl.Scale.X {type=MovieSceneScriptingFloatChannel, default=0}
       neck_02_ctrl.Scale.Y {type=MovieSceneScriptingFloatChannel, default=0}
       neck_02_ctrl.Scale.Z {type=MovieSceneScriptingFloatChannel, default=0}
       head_ik_ctrl.Location.X {type=MovieSceneScriptingFloatChannel, default=0}
       head_ik_ctrl.Location.Y {type=MovieSceneScriptingFloatChannel, default=0}
       head_ik_ctrl.Location.Z {type=MovieSceneScriptingFloatChannel, default=0}
       head_ik_ctrl.Rotation.X {type=MovieSceneScriptingFloatChannel, default=0}
       head_ik_ctrl.Rotation.Y {type=MovieSceneScriptingFloatChannel, default=0}
       head_ik_ctrl.Rotation.Z {type=MovieSceneScriptingFloatChannel, default=0}
       head_ik_ctrl.Scale.X {type=MovieSceneScriptingFloatChannel, default=0}
       head_ik_ctrl.Scale.Y {type=MovieSceneScriptingFloatChannel, default=0}
       head_ik_ctrl.Scale.Z {type=MovieSceneScriptingFloatChannel, default=0}
       chest_ctrl.Location.X {type=MovieSceneScriptingFloatChannel, default=0}
       chest_ctrl.Location.Y {type=MovieSceneScriptingFloatChannel, default=0}
       chest_ctrl.Location.Z {type=MovieSceneScriptingFloatChannel, default=0}
       chest_ctrl.Rotation.X {type=MovieSceneScriptingFloatChannel, default=0}
       chest_ctrl.Rotation.Y {type=MovieSceneScriptingFloatChannel, default=0}
       chest_ctrl.Rotation.Z {type=MovieSceneScriptingFloatChannel, default=0}
       chest_ctrl.Scale.X {type=MovieSceneScriptingFloatChannel, default=0}
       chest_ctrl.Scale.Y {type=MovieSceneScriptingFloatChannel, default=0}
       chest_ctrl.Scale.Z {type=MovieSceneScriptingFloatChannel, default=0}
       chest_tan_ctrl.Location.X {type=MovieSceneScriptingFloatChannel, default=0}
       chest_tan_ctrl.Location.Y {type=MovieSceneScriptingFloatChannel, default=0}
       chest_tan_ctrl.Location.Z {type=MovieSceneScriptingFloatChannel, default=0}
       chest_tan_ctrl.Rotation.X {type=MovieSceneScriptingFloatChannel, default=0}
       chest_tan_ctrl.Rotation.Y {type=MovieSceneScriptingFloatChannel, default=0}
       chest_tan_ctrl.Rotation.Z {type=MovieSceneScriptingFloatChannel, default=0}
       chest_tan_ctrl.Scale.X {type=MovieSceneScriptingFloatChannel, default=0}
       chest_tan_ctrl.Scale.Y {type=MovieSceneScriptingFloatChannel, default=0}
       chest_tan_ctrl.Scale.Z {type=MovieSceneScriptingFloatChannel, default=0}
      ]
     }
    ]
   }
  ]
 }]
display_rate {denominator=1, numerator=30}
tick_resolution {denominator=1, numerator=24000}
tracks []
